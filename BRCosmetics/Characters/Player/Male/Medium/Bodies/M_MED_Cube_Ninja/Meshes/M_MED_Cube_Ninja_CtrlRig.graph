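# Control Rig for the Cube Ninja medium male cosmetic body that procedurally adjusts belt and crystal bones as the legs move.
# It reads each thigh's rotation relative to the pelvis as ZYX Euler and feeds the Z angle into clamped float remaps.
# These drive a chain of Modify Transforms: each thigh rotates its side's two upper belt bones and shifts the first along Z.
# The left thigh also pitches dyn_crystal_03.
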ControlRigBlueprint /BRCosmetics/Characters/Player/Male/Medium/Bodies/M_MED_Cube_Ninja/Meshes/M_MED_Cube_Ninja_CtrlRig
graphs [
 RigVMFunctionLibrary {links [], nodes []}
 RigVMModel {nodes [
   RigUnit_BeginExecution = &RigUnit_BeginExecution()
   n1 = &RigVMFunction_MathQuaternionToEuler(RotationOrder=ZYX, Value=&RigUnit_GetRelativeBoneTransform(Bone=thigh_r, Space=pelvis).Transform.Rotation)
   n2 = &RigVMFunction_MathFloatRemap(SourceMaximum=90, SourceMinimum=130, TargetMaximum=7, TargetMinimum=0, bClamp=true, Value=@n1.Result.Z)
   n3 = &RigVMFunction_MathFloatRemap(SourceMaximum=90, SourceMinimum=130, TargetMaximum=5, TargetMinimum=0, bClamp=true, Value=@n1.Result.Z)
   n4 = &RigVMFunction_MathFloatRemap(SourceMaximum=90, SourceMinimum=130, TargetMaximum=5, TargetMinimum=0, bClamp=true, Value=@n1.Result.Z)
   &RigUnit_ModifyTransforms(
    ItemToModify=((Item=(Type=Bone,Name="dyn_r_belt_01"),Transform=(Rotation=(X=0,Y=0,Z=0,W=1),Translation=(X=0,Y=0,Z=0),Scale3D=(X=1,Y=1,Z=1))))
    Weight=1
    ItemToModify.0.Transform.Translation.Z=&RigVMFunction_MathFloatRemap(SourceMaximum=90, SourceMinimum=130, TargetMaximum=1.5, TargetMinimum=0, bClamp=true, Value=@n1.Result.Z).Result
    ItemToModify.0.Transform.Rotation=/Script/RigVM.RigVMFunction_MathQuaternionFromRotator(Rotator=(Pitch=0,Yaw=0,Roll=0), Rotator.Pitch=@n2.Result, Rotator.Roll=@n3.Result, Rotator.Yaw=@n4.Result) {is=struct RigVMUnitNode}.Result
    ExecutePin=@RigUnit_BeginExecution.ExecutePin
   )
   > &RigUnit_ModifyTransforms(
    ItemToModify=((Item=(Type=Bone,Name="dyn_r_belt_02"),Transform=(Rotation=(X=0,Y=0,Z=0,W=1),Translation=(X=0,Y=0,Z=0),Scale3D=(X=1,Y=1,Z=1))))
    Weight=1
    ItemToModify.0.Transform.Rotation=&RigVMFunction_MathQuaternionFromRotator(Rotator=(Pitch=0,Yaw=0,Roll=0), Rotator.Pitch=@n2.Result, Rotator.Roll=@n3.Result, Rotator.Yaw=@n4.Result).Result
   )
   n5 = &RigVMFunction_MathQuaternionToEuler(RotationOrder=ZYX, Value=&RigUnit_GetRelativeBoneTransform(Bone=thigh_l, Space=pelvis).Transform.Rotation)
   n6 = &RigVMFunction_MathFloatRemap(SourceMaximum=-100, SourceMinimum=-70, TargetMaximum=10, TargetMinimum=0, bClamp=true, Value=@n5.Result.Z)
   n7 = &RigVMFunction_MathFloatRemap(SourceMaximum=-100, SourceMinimum=-70, TargetMaximum=0, TargetMinimum=0, bClamp=true, Value=@n5.Result.Z)
   n8 = &RigVMFunction_MathFloatRemap(SourceMaximum=-100, SourceMinimum=-70, TargetMaximum=5, TargetMinimum=0, bClamp=true, Value=@n5.Result.Z)
   > &RigUnit_ModifyTransforms(
    ItemToModify=((Item=(Type=Bone,Name="dyn_l_belt_01"),Transform=(Rotation=(X=0,Y=0,Z=0,W=1),Translation=(X=0,Y=0,Z=0),Scale3D=(X=1,Y=1,Z=1))))
    Weight=1
    ItemToModify.0.Transform.Translation.Z=&RigVMFunction_MathFloatRemap(SourceMaximum=-100, SourceMinimum=-70, TargetMaximum=-1.5, TargetMinimum=0, bClamp=true, Value=@n5.Result.Z).Result
    ItemToModify.0.Transform.Rotation=&RigVMFunction_MathQuaternionFromRotator(Rotator=(Pitch=0,Yaw=0,Roll=0), Rotator.Roll=@n6.Result, Rotator.Yaw=@n7.Result, Rotator.Pitch=@n8.Result).Result
   )
   > &RigUnit_ModifyTransforms(
    ItemToModify=((Item=(Type=Bone,Name="dyn_l_belt_02"),Transform=(Rotation=(X=0,Y=0,Z=0,W=1),Translation=(X=0,Y=0,Z=0),Scale3D=(X=1,Y=1,Z=1))))
    Weight=1
    ItemToModify.0.Transform.Rotation=&RigVMFunction_MathQuaternionFromRotator(Rotator=(Pitch=0,Yaw=0,Roll=0), Rotator.Roll=@n6.Result, Rotator.Yaw=@n7.Result, Rotator.Pitch=@n8.Result).Result
   )
   > &RigUnit_ModifyTransforms(
    ItemToModify=((Item=(Type=Bone,Name="dyn_crystal_03"),Transform=(Rotation=(X=0,Y=0,Z=0,W=1),Translation=(X=0,Y=0,Z=0),Scale3D=(X=1,Y=1,Z=1))))
    Weight=1
    ItemToModify.0.Transform.Rotation=&RigVMFunction_MathQuaternionFromRotator(
     Rotator=(Pitch=8,Yaw=0,Roll=0)
     Rotator.Pitch=&RigVMFunction_MathFloatRemap(SourceMaximum=-100, SourceMinimum=-70, TargetMaximum=8, TargetMinimum=0, bClamp=true, Value=@n5.Result.Z).Result
    ).Result
   )
   &RigVMFunction_MathQuaternionFromRotator(Rotator=(Pitch=0,Yaw=0,Roll=30))
  ]}
]
hierarchy [
 root {loc=<0 0 0>, rot=<0 0 0 1>}
 pelvis {parent=root, loc=<0 -0.7071 93.1191>, rot=<0.0127 0.707 -0.0127 -0.707>}
 spine_01 {parent=pelvis, loc=<6.7494 -0.6388 0>, rot=<0 0 0.034 -0.9994>}
 spine_02 {parent=spine_01, loc=<7.2518 -0.035 0>, rot=<0 0 0.0238 -0.9997>}
 spine_03 {parent=spine_02, loc=<6.0104 0 0>, rot=<0 0 0.0149 -0.9999>}
 spine_04 {parent=spine_03, loc=<8.1024 0 0>, rot=<0 0 -0.0132 -0.9999>}
 spine_05 {parent=spine_04, loc=<10.0146 0 0>, rot=<0 0 -0.0015 -1>}
 clavicle_l {parent=spine_05, loc=<4.3575 -0.1466 -2.3223>, rot=<0.0465 -0.7289 -0.009 -0.6829>}
 upperarm_l {parent=clavicle_l, loc=<15.5111 0 0>, rot=<0.0226 -0.3234 -0.0088 -0.9459>}
 lowerarm_l {parent=upperarm_l, loc=<26.6514 0 0>, rot=<0.0391 -0.0099 -0.2453 -0.9686>}
 hand_l {parent=lowerarm_l, loc=<22.3491 0 0>, rot=<-0.7081 -0.0507 0.0975 -0.6975>}
 index_metacarpal_l {parent=hand_l, loc=<1.6535 0.0183 -2.96>, rot=<0.0158 -0.0124 -0.0173 -0.9996>}
 index_01_l {parent=index_metacarpal_l, loc=<8.3008 0.0357 -0.0524>, rot=<-0.0044 0.0485 0.0924 -0.9945>}
 index_02_l {parent=index_01_l, loc=<3.3587 -0.3004 0.0452>, rot=<-0.002 0.0027 0.3225 -0.9466>}
 index_03_l {parent=index_02_l, loc=<2.5578 -0.123 -0.0252>, rot=<-0.002 0.0067 0.2558 -0.9667>}
 middle_metacarpal_l {parent=hand_l, loc=<2.1856 -0.1725 -0.5984>, rot=<0.006 0.0168 -0.0329 -0.9993>}
 middle_01_l {parent=middle_metacarpal_l, loc=<7.3275 -0.2447 0.0129>, rot=<-0.0208 0.0042 0.0858 -0.9961>}
 middle_02_l {parent=middle_01_l, loc=<4.0123 0.1401 0.1175>, rot=<-0.0218 0.0258 0.3512 -0.9357>}
 middle_03_l {parent=middle_02_l, loc=<2.6077 0.0534 0.1373>, rot=<0.0082 0.0188 0.1708 -0.9851>}
 pinky_metacarpal_l {parent=hand_l, loc=<1.7616 -0.9753 3.3814>, rot=<-0.0397 0.0814 -0.069 -0.9935>}
 pinky_01_l {parent=pinky_metacarpal_l, loc=<7.1709 -0.4831 -0.1094>, rot=<-0.0654 -0.0035 0.2242 -0.9723>}
 pinky_02_l {parent=pinky_01_l, loc=<2.8274 0.0893 -0.1043>, rot=<0.0037 0.0011 0.2854 -0.9584>}
 pinky_03_l {parent=pinky_02_l, loc=<2.2046 0.0307 -0.0078>, rot=<-0.0002 0.0098 0.1445 -0.9895>}
 ring_metacarpal_l {parent=hand_l, loc=<2.2074 -0.3625 1.3638>, rot=<-0.0284 0.0722 -0.0458 -0.9959>}
 ring_01_l {parent=ring_metacarpal_l, loc=<7.0295 -0.6197 -0.1446>, rot=<-0.0489 -0.0252 0.1892 -0.9804>}
 ring_02_l {parent=ring_01_l, loc=<3.6174 0.4647 -0.1027>, rot=<0 0 0.2806 -0.9598>}
 ring_03_l {parent=ring_02_l, loc=<2.4152 0.114 0.0199>, rot=<0 0 0.1189 -0.9929>}
 thumb_01_l {parent=hand_l, loc=<1.3118 -0.3191 -2.459>, rot=<0.6546 -0.4042 0.0664 -0.6354>}
 thumb_02_l {parent=thumb_01_l, loc=<5.9336 -0.0448 0.0401>, rot=<-0.0037 0.0213 0.298 -0.9543>}
 thumb_03_l {parent=thumb_02_l, loc=<2.2223 0.3682 -0.0907>, rot=<0.0118 -0.013 0.0066 -0.9998>}
 lowerarm_twist_01_l {parent=lowerarm_l, loc=<17.8549 0 0>, rot=<0 0 0 -1>}
 lowerarm_twist_02_l {parent=lowerarm_l, loc=<9.6199 0 0>, rot=<0 0 0 -1>}
 dyn_l_arm_ribbon_master {parent=lowerarm_l, loc=<17.878 -3.0284 0.423>, rot=<-0.4966 0.3185 0.6187 -0.5188>}
 dyn_l_arm_ribbon_01 {parent=dyn_l_arm_ribbon_master, loc=<0.7918 0 0>, rot=<0 0 0 1>}
 dyn_l_arm_ribbon_02 {parent=dyn_l_arm_ribbon_01, loc=<4.9938 0 0>, rot=<-0.0143 -0.0224 -0.0086 -0.9996>}
 dyn_l_arm_ribbon_03 {parent=dyn_l_arm_ribbon_02, loc=<7.3369 0 0>, rot=<0.0169 0.0262 0.0101 -0.9995>}
 upperarm_twist_01_l {parent=upperarm_l, loc=<0.0041 0 0>, rot=<0 0 0 -1>}
 upperarm_twist_02_l {parent=upperarm_l, loc=<13.7941 0 0>, rot=<0 0 0 -1>}
 clavicle_r {parent=spine_05, loc=<4.3575 -0.1466 2.3223>, rot=<0.7289 0.0465 -0.6829 0.009>}
 upperarm_r {parent=clavicle_r, loc=<-15.5111 0 0>, rot=<0.0226 -0.3234 -0.0088 -0.9459>}
 lowerarm_r {parent=upperarm_r, loc=<-26.6515 0 0>, rot=<0.0391 -0.0099 -0.2453 -0.9686>}
 hand_r {parent=lowerarm_r, loc=<-22.3492 0 0>, rot=<-0.7081 -0.0507 0.0975 -0.6975>}
 index_metacarpal_r {parent=hand_r, loc=<-1.6534 -0.0181 2.9601>, rot=<0.0158 -0.0124 -0.0173 -0.9996>}
 index_01_r {parent=index_metacarpal_r, loc=<-8.3008 -0.0357 0.0524>, rot=<-0.0044 0.0485 0.0924 -0.9945>}
 index_02_r {parent=index_01_r, loc=<-3.3589 0.3003 -0.0453>, rot=<-0.002 0.0027 0.3225 -0.9466>}
 index_03_r {parent=index_02_r, loc=<-2.5578 0.1231 0.0251>, rot=<-0.002 0.0067 0.2558 -0.9667>}
 middle_metacarpal_r {parent=hand_r, loc=<-2.1856 0.1726 0.5984>, rot=<0.006 0.0168 -0.0329 -0.9993>}
 middle_01_r {parent=middle_metacarpal_r, loc=<-7.3273 0.2447 -0.0129>, rot=<-0.0208 0.0042 0.0858 -0.9961>}
 middle_02_r {parent=middle_01_r, loc=<-4.0142 -0.1411 -0.07>, rot=<-0.0217 0.0259 0.3512 -0.9357>}
 middle_03_r {parent=middle_02_r, loc=<-2.6074 -0.0534 -0.1312>, rot=<0.0083 0.0188 0.1708 -0.9851>}
 pinky_metacarpal_r {parent=hand_r, loc=<-1.7617 0.9753 -3.3814>, rot=<-0.0397 0.0814 -0.069 -0.9935>}
 pinky_01_r {parent=pinky_metacarpal_r, loc=<-7.1707 0.4832 0.1094>, rot=<-0.0654 -0.0035 0.2242 -0.9723>}
 pinky_02_r {parent=pinky_01_r, loc=<-2.8274 -0.0892 0.1043>, rot=<0.0037 0.0011 0.2854 -0.9584>}
 pinky_03_r {parent=pinky_02_r, loc=<-2.2045 -0.031 0.0078>, rot=<-0.0002 0.0098 0.1445 -0.9895>}
 ring_metacarpal_r {parent=hand_r, loc=<-2.2073 0.3627 -1.3638>, rot=<-0.0284 0.0722 -0.0458 -0.9959>}
 ring_01_r {parent=ring_metacarpal_r, loc=<-7.0297 0.6197 0.1446>, rot=<-0.0489 -0.0252 0.1892 -0.9804>}
 ring_02_r {parent=ring_01_r, loc=<-3.6171 -0.4648 0.1027>, rot=<0 0 0.2806 -0.9598>}
 ring_03_r {parent=ring_02_r, loc=<-2.4153 -0.1141 -0.0199>, rot=<0 0 0.1189 -0.9929>}
 thumb_01_r {parent=hand_r, loc=<-1.3119 0.3192 2.459>, rot=<0.6546 -0.4042 0.0664 -0.6354>}
 thumb_02_r {parent=thumb_01_r, loc=<-5.9336 0.0447 -0.0401>, rot=<-0.0037 0.0213 0.298 -0.9543>}
 thumb_03_r {parent=thumb_02_r, loc=<-2.2223 -0.3682 0.0907>, rot=<0.0118 -0.013 0.0066 -0.9998>}
 lowerarm_twist_01_r {parent=lowerarm_r, loc=<-17.855 0 0>, rot=<0 0 0 1>}
 lowerarm_twist_02_r {parent=lowerarm_r, loc=<-9.6199 0 0>, rot=<0 0 0 1>}
 upperarm_twist_01_r {parent=upperarm_r, loc=<-0.0041 0 0>, rot=<0 0 0 1>}
 upperarm_twist_02_r {parent=upperarm_r, loc=<-13.7942 0 0>, rot=<0 0 0 1>}
 neck_01 {parent=spine_05, loc=<11.3138 0.1529 0>, rot=<0 0 -0.2003 -0.9797>}
 neck_02 {parent=neck_01, loc=<4.5627 0 0>, rot=<0 0 0.0504 -0.9987>}
 thigh_l {parent=pelvis, loc=<-2.5693 -0.0109 -7.7501>, rot=<0.0012 0.0572 0.0155 -0.9982>}
 calf_l {parent=thigh_l, loc=<-39.0484 0 0>, rot=<0.0006 -0.0161 -0.0391 -0.9991>}
 calf_twist_01_l {parent=calf_l, loc=<-35.813 -0.0967 0.0397>, rot=<0.0237 0.0095 -0.0035 -0.9997>}
 calf_twist_02_l {parent=calf_l, loc=<-18.8064 -0.0967 0.0397>, rot=<0.0237 0.0095 -0.0035 -0.9997>}
 foot_l {parent=calf_l, loc=<-42.561 0.1077 0.4292>, rot=<0.0418 -0.0455 0.0343 -0.9975>}
 ball_l {parent=foot_l, loc=<-6.1599 13.8264 0.6806>, rot=<0.0001 -0.0001 -0.7063 -0.7079>}
 dyn_pant_leg_back_r_master {parent=calf_l, loc=<-12.2296 -9.046 3.3858>, rot=<0.1725 -0.5348 0.8252 -0.0568>}
 dyn_pant_leg_back_r_01 {parent=dyn_pant_leg_back_r_master, loc=<0 0 0>, rot=<0 0 0 1>}
 dyn_pant_leg_back_r_02 {parent=dyn_pant_leg_back_r_01, loc=<5.488 0 0>, rot=<-0.0068 0.0508 -0.011 -0.9986>}
 dyn_pant_leg_back_r_03 {parent=dyn_pant_leg_back_r_02, loc=<6.2581 0 0>, rot=<-0.0145 0.0099 -0.0159 -0.9997>}
 dyn_pant_leg_back_l_master {parent=calf_l, loc=<-12.8345 -8.762 -4.2008>, rot=<0.0524 -0.886 0.4469 -0.112>}
 dyn_pant_leg_back_l_01 {parent=dyn_pant_leg_back_l_master, loc=<0 0 0>, rot=<0 0 0 -1>}
 dyn_pant_leg_back_l_02 {parent=dyn_pant_leg_back_l_01, loc=<5.488 0 0>, rot=<0.0038 0.0505 -0.0123 -0.9986>}
 dyn_pant_leg_back_l_03 {parent=dyn_pant_leg_back_l_02, loc=<6.2581 0 0>, rot=<-0.0151 0.0099 0.0033 -0.9998>}
 dyn_pant_leg_left_master {parent=calf_l, loc=<-13.4721 -1.5037 -7.4534>, rot=<-0.0664 -0.9838 0.0032 -0.1663>}
 dyn_pant_leg_left_01 {parent=dyn_pant_leg_left_master, loc=<0 0 0>, rot=<0 0 0 -1>}
 dyn_pant_leg_left_02 {parent=dyn_pant_leg_left_01, loc=<5.4522 0 0>, rot=<0.0006 0.0305 -0.0129 -0.9995>}
 dyn_pant_leg_left_03 {parent=dyn_pant_leg_left_02, loc=<6.1444 0 0>, rot=<0.0004 0.0254 0.0159 -0.9996>}
 dyn_pant_leg_right_master {parent=calf_l, loc=<-11.5626 -0.9546 5.971>, rot=<0.1706 -0.0065 0.9839 0.0521>}
 dyn_pant_leg_right_01 {parent=dyn_pant_leg_right_master, loc=<0 0 0>, rot=<0 0 0 1>}
 dyn_pant_leg_right_02 {parent=dyn_pant_leg_right_01, loc=<5.5514 0 0>, rot=<-0.0002 0.0337 -0.0171 -0.9993>}
 dyn_pant_leg_right_03 {parent=dyn_pant_leg_right_02, loc=<5.8972 0 0>, rot=<0 0.0009 0.0147 -0.9999>}
 dyn_pant_leg_front_master {parent=calf_l, loc=<-12.4588 4.0541 -0.7994>, rot=<-0.1677 -0.6901 -0.6889 -0.1452>}
 dyn_pant_leg_front_01 {parent=dyn_pant_leg_front_master, loc=<0 0 0>, rot=<0.0029 0.0002 0.0075 -1>}
 dyn_pant_leg_front_02 {parent=dyn_pant_leg_front_01, loc=<5.373 0 0>, rot=<-0.0031 0.0006 -0.0082 -1>}
 dyn_pant_leg_front_03 {parent=dyn_pant_leg_front_02, loc=<5.9 0 0>, rot=<-0.0004 0.003 -0.0011 -1>}
 thigh_twist_01_l {parent=thigh_l, loc=<-21.6303 0 0>, rot=<0.0013 0 0 -1>}
 thigh_r {parent=pelvis, loc=<-2.5691 -0.0109 7.7501>, rot=<0.0572 -0.0012 0.9982 0.0155>}
 calf_r {parent=thigh_r, loc=<39.0487 0 0>, rot=<0.0006 -0.0161 -0.0391 -0.9991>}
 calf_twist_01_r {parent=calf_r, loc=<35.8132 0.0967 -0.0397>, rot=<0.0237 0.0095 -0.0035 -0.9997>}
 calf_twist_02_r {parent=calf_r, loc=<18.8064 0.0967 -0.0397>, rot=<0.0237 0.0095 -0.0035 -0.9997>}
 foot_r {parent=calf_r, loc=<42.5612 -0.1077 -0.4292>, rot=<0.0418 -0.0455 0.0343 -0.9975>}
 ball_r {parent=foot_r, loc=<6.1598 -13.8264 -0.6806>, rot=<0.0001 -0.0001 -0.7063 -0.7079>}
 dyn_leg_ribbon_ball {parent=calf_r, loc=<26.106 3.0982 5.457>, rot=<-0.6769 0.027 0.7349 0.0315>}
 dyn_leg_ribbon_top_master {parent=dyn_leg_ribbon_ball, loc=<-0.0219 -0.0837 1.2181>, rot=<0.1604 -0.0857 0.9795 -0.0872>}
 dyn_leg_ribbon_top_01 {parent=dyn_leg_ribbon_top_master, loc=<0 0 0>, rot=<0 0 0 1>}
 dyn_leg_ribbon_top_02 {parent=dyn_leg_ribbon_top_01, loc=<2.2511 0 0>, rot=<0.068 -0.3316 -0.1197 -0.9333>}
 dyn_leg_ribbon_top_03 {parent=dyn_leg_ribbon_top_02, loc=<2.5743 0 0>, rot=<0.0074 -0.2714 0.016 -0.9623>}
 dyn_leg_ribbon_master {parent=dyn_leg_ribbon_ball, loc=<-0.9954 -0.2359 -0.2476>, rot=<-0.2011 -0.1468 0.9675 -0.0441>}
 dyn_leg_ribbon_01 {parent=dyn_leg_ribbon_master, loc=<0 0 0>, rot=<0 0 0 -1>}
 dyn_leg_ribbon_02 {parent=dyn_leg_ribbon_01, loc=<1.6223 0 0>, rot=<0.1031 -0.1556 -0.295 -0.9371>}
 dyn_leg_ribbon_03 {parent=dyn_leg_ribbon_02, loc=<2.2462 0 0>, rot=<0.0493 -0.0024 -0.0354 -0.9982>}
 dyn_knee_adjustment_r {parent=calf_r, loc=<-1.5924 -0.8983 9.6246>, rot=<-0.6769 0.027 0.7349 0.0315>}
 thigh_twist_01_r {parent=thigh_r, loc=<21.6303 0 0>, rot=<0.0013 0 0 -1>}
 dyn_crystal_master {parent=pelvis, loc=<7.1696 -10.081 -9.0014>, rot=<0.1703 -0.7509 0.4291 -0.4723>}
 dyn_crystal_01 {parent=dyn_crystal_master, loc=<0 0 0>, rot=<0 0 0 1>}
 dyn_crystal_02 {parent=dyn_crystal_01, loc=<3.2994 0 0>, rot=<0.0448 -0.0389 0.0392 -0.9975>}
 dyn_crystal_03 {parent=dyn_crystal_02, loc=<0.6769 0 -0.1534>, rot=<0.0143 -0.3156 0.0115 -0.9487>}
 dyn_belt_base {parent=pelvis, loc=<0 0 0>, rot=<-0.0127 -0.707 0.0127 -0.707>}
 dyn_l_belt_master {parent=dyn_belt_base, loc=<0 0.7071 2.9232>, rot=<0 -0.9887 0 0.1499>}
 dyn_l_belt_01 {parent=dyn_l_belt_master, loc=<0 0 0>, rot=<0 0 0 1>}
 dyn_l_belt_02 {parent=dyn_l_belt_01, loc=<-7.6579 0 0>, rot=<0 -0.026 0 -0.9997>}
 dyn_l_belt_03 {parent=dyn_l_belt_02, loc=<-9.192 0 0>, rot=<0 -0.0323 0 -0.9995>}
 dyn_mask_master {parent=dyn_l_belt_03, loc=<-2.3881 0 -0.9156>, rot=<0 -0.9783 0 -0.2073>}
 dyn_mask_01 {parent=dyn_mask_master, loc=<0 0 0>, rot=<0 0 0 1>}
 dyn_r_belt_master {parent=dyn_belt_base, loc=<0 0.7071 2.9232>, rot=<-0.1499 0 0.9887 0>}
 dyn_r_belt_01 {parent=dyn_r_belt_master, loc=<0 0 0>, rot=<0 0 0 1>}
 dyn_r_belt_02 {parent=dyn_r_belt_01, loc=<7.6579 0 0>, rot=<0 -0.026 0 -0.9997>}
 dyn_r_belt_03 {parent=dyn_r_belt_02, loc=<9.1919 0 0>, rot=<0 -0.0323 0 -0.9995>}
 dyn_scrolls_master {parent=pelvis, loc=<-2.6901 0.6174 15.1483>, rot=<0.707 -0.0127 0.707 0.0127>}
 dyn_scroll_02_master {parent=dyn_scrolls_master, loc=<5.0155 -0.2351 -1.4696>, rot=<1 0 0 0>}
 dyn_scroll_02 {parent=dyn_scroll_02_master, loc=<0 0 0>, rot=<0 0 0 1>}
 dyn_scroll_02_02 {parent=dyn_scroll_02, loc=<0 0 0>, rot=<0 0 0 1>}
 dyn_scroll_01_master {parent=dyn_scrolls_master, loc=<3.3696 -0.1074 3.7366>, rot=<0.9943 -0.1056 0.0018 -0.0169>}
 dyn_scroll_01 {parent=dyn_scroll_01_master, loc=<0 0 0>, rot=<0 0 0 1>}
 dyn_sash_back_master {parent=dyn_scrolls_master, loc=<-0.1498 4.3196 -1.4537>, rot=<-0.0255 -0.5327 -0.0405 -0.845>}
 dyn_sash_back_01 {parent=dyn_sash_back_master, loc=<0 0 0>, rot=<0 0 0 1>}
 dyn_sash_back_01_05 {parent=dyn_sash_back_01, loc=<4.766 -0.0382 0.139>, rot=<0.0669 0.0027 -0.0164 -0.9976>}
 dyn_sash_back_02 {parent=dyn_sash_back_01_05, loc=<7.221 -0.1771 -0.187>, rot=<-0.0758 -0.1015 0.0265 -0.9916>}
 dyn_sash_back_03 {parent=dyn_sash_back_02, loc=<14.8031 0 0>, rot=<0.0005 0.0042 -0.0461 -0.9989>}
 dyn_sash_back_04 {parent=dyn_sash_back_03, loc=<13.426 0 0>, rot=<0.0065 0.0413 -0.0008 -0.9991>}
 dyn_sash_back_05 {parent=dyn_sash_back_04, loc=<8.8585 0 0>, rot=<0 0 0.0144 -0.9999>}
 dyn_sash_front_master {parent=dyn_scrolls_master, loc=<0.1567 -0.1865 -1.4223>, rot=<0 -0.5192 0 -0.8546>}
 dyn_sash_front_01 {parent=dyn_sash_front_master, loc=<0 0 0>, rot=<0 0 0 1>}
 dyn_sash_front_02 {parent=dyn_sash_front_01, loc=<5.375 0 0>, rot=<-0.0009 -0.0208 -0.045 -0.9988>}
 dyn_sash_front_03 {parent=dyn_sash_front_02, loc=<7.8293 0 0>, rot=<-0.0029 -0.034 0.0047 -0.9994>}
 dyn_sash_front_04 {parent=dyn_sash_front_03, loc=<10.4601 0 0>, rot=<-0.0019 -0.0181 -0.0249 -0.9995>}
 dyn_sash_front_05 {parent=dyn_sash_front_04, loc=<13.3134 0 0>, rot=<-0.0016 -0.0151 0.0275 -0.9995>}
 dyn_sash_front_06 {parent=dyn_sash_front_05, loc=<12.0466 0 0>, rot=<0.0042 0.0374 -0.0373 -0.9986>}
 dyn_sash_front_07 {parent=dyn_sash_front_06, loc=<12.0778 0 0>, rot=<-0.0033 -0.0215 -0.0002 -0.9998>}
 dyn_sash_front_08 {parent=dyn_sash_front_07, loc=<11.8814 0 0>, rot=<-0.0011 -0.0082 0.0183 -0.9998>}
 ik_hand_root {parent=root, loc=<0 0 0>, rot=<0 0 0 1>}
 ik_hand_gun {parent=ik_hand_root, loc=<-51.9729 4.0884 102.6935>, rot=<-0.6084 -0.1793 -0.4613 0.6205>}
]
