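# Physics asset for the Cube Ninja male medium cosmetic body, Frontend variant, covering its dangling costume parts.
# Angle-limited chains form the left arm ribbon, belts with a mask, sashes, scrolls, and five pant-leg strips, all five on the left calf.
# Ribbon, belt and scroll masters join forearm or pelvis with free rotation; sashes hang from the scroll master.
# Legs, pelvis, left arm and a floor-like root box add shapes; belt ends link to thighs.
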PhysicsAsset /BRCosmetics/Characters/Player/Male/Medium/Bodies/M_MED_Cube_Ninja/Meshes/M_MED_Cube_Ninja_Physics_Frontend
bodies {
 calf_l [{shapeType=Capsule, length=30.0048, radius=8.7886, center=<-20.978 0.0003 0>, rotation=<rot 90 0 0>}]
 calf_r [{shapeType=Capsule, length=30.0656, radius=11.8446, center=<20.9218 1.6597 1.093>, rotation=<rot 90 0 0>}]
 dyn_l_arm_ribbon_01 [{shapeType=Capsule, length=2.1226, radius=1.3752, center=<3.0395 0 -0.0001>, rotation=<rot 0 -90 -90>}]
 dyn_l_arm_ribbon_02 [{shapeType=Box, extentX=6.6455, extentY=0.9987, extentZ=4.316, center=<3.7181 0.3373 0>, rotation=<rot 0 0 90>}]
 dyn_l_arm_ribbon_03 [{shapeType=Box, extentX=4.1649, extentY=0.926, extentZ=3.7271, center=<2.3289 0 0>, rotation=<rot 0 0 90>}]
 dyn_l_arm_ribbon_master [{shapeType=Box, extentX=1.01, extentY=1.01, extentZ=1.01, rotation=<rot 0 -90 0>}]
 dyn_l_belt_01 [{shapeType=Capsule, length=2, radius=3.1453, center=<-4.3065 -0.0002 0>, rotation=<rot 90 0 0>}]
 dyn_l_belt_02 [{shapeType=Capsule, length=18.1505, radius=3.1453, center=<-4.3065 -0.0002 0>, rotation=<rot 0 0 90>}]
 dyn_l_belt_03 [{shapeType=Capsule, length=18.1505, radius=3.1453, center=<-4.3065 -0.0002 0>, rotation=<rot 0 0 90>}]
 dyn_l_belt_master [{shapeType=Capsule, length=0.505, radius=0.505, rotation=<rot 0 -90 0>}]
 dyn_mask_01 [{shapeType=Capsule, length=18.2463, radius=5.7534, center=<3.7069 0 0.1697>, rotation=<rot 13.6832 0 0>}]
 dyn_pant_leg_back_l_01 [{shapeType=Capsule, length=12.4292, radius=2.6012, center=<3.0507 -0.0002 1.7447>, rotation=<rot 0 0 90>}]
 dyn_pant_leg_back_l_02 [{shapeType=Capsule, length=12.4292, radius=2.6012, center=<3.0507 -0.0002 1.7447>, rotation=<rot 0 0 90>}]
 dyn_pant_leg_back_l_03 [{shapeType=Capsule, length=12.4292, radius=2.6012, center=<3.0507 -0.0002 1.7447>, rotation=<rot 0 0 90>}]
 dyn_pant_leg_back_l_master [{shapeType=Capsule, length=3.0424, radius=2.6012, rotation=<rot 0 0 90>}]
 dyn_pant_leg_back_r_01 [{shapeType=Capsule, length=12.4292, radius=2.6012, center=<3.0507 -0.0002 1.7447>, rotation=<rot 0 0 90>}]
 dyn_pant_leg_back_r_02 [{shapeType=Capsule, length=12.4292, radius=2.6012, center=<3.0507 -0.0002 1.7447>, rotation=<rot 0 0 90>}]
 dyn_pant_leg_back_r_03 [{shapeType=Capsule, length=12.4292, radius=2.6012, center=<3.0507 -0.0002 1.7447>, rotation=<rot 0 0 90>}]
 dyn_pant_leg_back_r_master [{shapeType=Capsule, length=3.0424, radius=2.6012, rotation=<rot 0 0 90>}]
 dyn_pant_leg_front_01 [{shapeType=Capsule, length=12.4292, radius=2.6012, center=<3.0507 -0.0002 1.7447>, rotation=<rot 0 0 90>}]
 dyn_pant_leg_front_02 [{shapeType=Capsule, length=12.4292, radius=2.6012, center=<3.0507 -0.0002 1.7447>, rotation=<rot 0 0 90>}]
 dyn_pant_leg_front_03 [{shapeType=Capsule, length=12.4292, radius=2.6012, center=<3.0507 -0.0002 1.7447>, rotation=<rot 0 0 90>}]
 dyn_pant_leg_front_master [{shapeType=Capsule, length=3.0424, radius=2.6012, rotation=<rot 0 0 90>}]
 dyn_pant_leg_left_01 [{shapeType=Capsule, length=12.4292, radius=2.6012, center=<3.0507 -0.0002 1.7447>, rotation=<rot 0 0 90>}]
 dyn_pant_leg_left_02 [{shapeType=Capsule, length=12.4292, radius=2.6012, center=<3.0507 -0.0002 1.7447>, rotation=<rot 0 0 90>}]
 dyn_pant_leg_left_03 [{shapeType=Capsule, length=12.4292, radius=2.6012, center=<3.0507 -0.0002 1.7447>, rotation=<rot 0 0 90>}]
 dyn_pant_leg_left_master [{shapeType=Capsule, length=3.0424, radius=2.6012, rotation=<rot 0 0 90>}]
 dyn_pant_leg_right_01 [{shapeType=Capsule, length=12.4292, radius=2.6012, center=<3.0507 -0.0002 1.7447>, rotation=<rot 0 0 90>}]
 dyn_pant_leg_right_02 [{shapeType=Capsule, length=12.4292, radius=2.6012, center=<3.0507 -0.0002 1.7447>, rotation=<rot 0 0 90>}]
 dyn_pant_leg_right_03 [{shapeType=Capsule, length=12.4292, radius=2.6012, center=<3.0507 -0.0002 1.7447>, rotation=<rot 0 0 90>}]
 dyn_pant_leg_right_master [{shapeType=Capsule, length=3.0424, radius=2.6012, rotation=<rot 0 0 90>}]
 dyn_r_belt_01 [{shapeType=Capsule, length=2, radius=3.1453, center=<3.3981 0 0>, rotation=<rot 90 0 0>}]
 dyn_r_belt_02 [{shapeType=Capsule, length=18.1505, radius=3.1453, center=<3.3981 0 0>, rotation=<rot 0 0 90>}]
 dyn_r_belt_03 [{shapeType=Capsule, length=18.1505, radius=3.1453, center=<3.3981 0 0>, rotation=<rot 0 0 90>}]
 dyn_r_belt_master [{shapeType=Capsule, length=0.505, radius=0.505, rotation=<rot 0 -90 0>}]
 dyn_sash_back_01 [{shapeType=Capsule, length=4.1133, radius=2, center=<7.4803 0.0003 0>, rotation=<rot 90 0 0>}]
 dyn_sash_back_02 [{shapeType=Capsule, length=11.5734, radius=2, center=<7.7306 0.0002 0>, rotation=<rot 90 0 0>}]
 dyn_sash_back_03 [{shapeType=Capsule, length=9.6863, radius=2, center=<6.9218 -0.0001 0>, rotation=<rot 90 0 0>}]
 dyn_sash_back_04 [{shapeType=Capsule, length=4.8426, radius=2, center=<4.6607 -0.0001 0>, rotation=<rot 90 0 0>}]
 dyn_sash_back_05 [{shapeType=Capsule, length=2, radius=2, center=<3.3098 0 0>, rotation=<rot 90 0 0>}]
 dyn_sash_front_01 [{shapeType=Capsule, length=5.377, radius=3.338, center=<3.6216 0.3171 -0.0773>, rotation=<rot 0 -84.9962 -88.782>}]
 dyn_sash_front_02 [{shapeType=Capsule, length=2.4818, radius=2.374, center=<4.1379 0 0.157>, rotation=<rot 90 0 0>}]
 dyn_sash_front_03 [{shapeType=Capsule, length=0, radius=4.3275, center=<5.6719 -0.3877 0.4134>, rotation=<rot 0 175.748 -72.7401>}]
 dyn_sash_front_04 [{shapeType=Capsule, length=6.7604, radius=3.1017, center=<6.7892 -0.3698 0.3197>, rotation=<rot 0 88.6327 -89.6766>}]
 dyn_sash_front_05 [{shapeType=Capsule, length=5.4203, radius=2.7543, center=<6.2818 -0.1674 0.4816>, rotation=<rot 0 93.8958 -88.1457>}]
 dyn_sash_front_06 [{shapeType=Capsule, length=6.9674, radius=2.374, center=<5.9687 -0.0136 0.4239>, rotation=<rot 0 -91.6411 -87.8661>}]
 dyn_sash_front_07 [{shapeType=Capsule, length=6.9351, radius=2.3517, center=<5.9405 -0.0478 0.0691>, rotation=<rot 0 86.3821 -89.1753>}]
 dyn_sash_front_08 [{shapeType=Capsule, length=3.469, radius=2.4816, center=<2.1814 -0.0002 0>, rotation=<rot 0 0 90>}]
 dyn_scroll_01 [{shapeType=Capsule, length=12.3671, radius=4.9605, center=<0.2673 0 -0.2316>, rotation=<rot 0 0 90>}]
 dyn_scroll_02 [{shapeType=Capsule, length=13.8541, radius=3.9942, rotation=<rot 0 0 90>}]
 dyn_scroll_02_02 [{shapeType=Capsule, length=14.0093, radius=3.839, rotation=<rot 0 0 90>}]
 dyn_scrolls_master [{shapeType=Capsule, length=4.8646, radius=2.8317}]
 hand_l [{shapeType=Capsule, length=5.0105, radius=4.6395, center=<4.2225 0.0001 0>, rotation=<rot 90 0 0>}]
 lowerarm_l [{shapeType=Capsule, length=13.0254, radius=3.8728, center=<9.5567 0.5413 0>, rotation=<rot 90 -90 -89.0154>}]
 pelvis [{shapeType=Capsule, length=6.6712, radius=11.1366, rotation=<rot 0 -90 0>}]
 root [{shapeType=Box, extentX=200, extentY=200, extentZ=50, center=<0 0 -25>}]
 thigh_l [{shapeType=Capsule, length=22.1578, radius=10.1311, center=<-19.8846 -0.0003 0>, rotation=<rot 90 0 0>}]
 thigh_r [{shapeType=Capsule, length=30.7964, radius=10.5923, center=<26.0501 1.166 0.0159>, rotation=<rot 87.6693 -180 -180>}]
}
constraints [
 {
  bone1Name=thigh_r
  bone2Name=pelvis
  swing1LimitDegrees=45
  swing1Motion=Free
  swing2LimitDegrees=45
  swing2Motion=Free
  twistLimitDegrees=45
  twistMotion=Free
 }
 {
  bone1Name=calf_r
  bone2Name=thigh_r
  swing1LimitDegrees=45
  swing1Motion=Free
  swing2LimitDegrees=45
  swing2Motion=Free
  twistLimitDegrees=45
  twistMotion=Free
 }
 {bone1Name=dyn_l_arm_ribbon_03, bone2Name=dyn_l_arm_ribbon_02, swing1LimitDegrees=45, swing2LimitDegrees=45, twistLimitDegrees=20}
 {bone1Name=dyn_l_arm_ribbon_01, bone2Name=dyn_l_arm_ribbon_master, swing1LimitDegrees=60, swing2LimitDegrees=60, twistLimitDegrees=10}
 {bone1Name=dyn_l_arm_ribbon_02, bone2Name=dyn_l_arm_ribbon_01, swing1LimitDegrees=45, swing2LimitDegrees=45, twistLimitDegrees=20}
 {
  bone1Name=dyn_l_arm_ribbon_master
  bone2Name=lowerarm_l
  swing1LimitDegrees=45
  swing1Motion=Free
  swing2LimitDegrees=45
  swing2Motion=Free
  twistLimitDegrees=45
  twistMotion=Free
 }
 {
  bone1Name=lowerarm_l
  bone2Name=pelvis
  swing1LimitDegrees=45
  swing1Motion=Free
  swing2LimitDegrees=45
  swing2Motion=Free
  twistLimitDegrees=45
  twistMotion=Free
 }
 {
  bone1Name=hand_l
  bone2Name=lowerarm_l
  swing1LimitDegrees=45
  swing1Motion=Free
  swing2LimitDegrees=45
  swing2Motion=Free
  twistLimitDegrees=45
  twistMotion=Free
 }
 {bone1Name=dyn_sash_front_08, bone2Name=dyn_sash_front_07, swing1LimitDegrees=20, swing2LimitDegrees=40, twistLimitDegrees=20}
 {bone1Name=dyn_sash_front_02, bone2Name=dyn_sash_front_01, swing1LimitDegrees=20, swing2LimitDegrees=20, twistLimitDegrees=20}
 {bone1Name=dyn_sash_front_03, bone2Name=dyn_sash_front_02, swing1LimitDegrees=20, swing2LimitDegrees=20, twistLimitDegrees=20}
 {bone1Name=dyn_sash_front_04, bone2Name=dyn_sash_front_03, swing1LimitDegrees=20, swing2LimitDegrees=40, twistLimitDegrees=20}
 {bone1Name=dyn_sash_front_05, bone2Name=dyn_sash_front_04, swing1LimitDegrees=20, swing2LimitDegrees=40, twistLimitDegrees=20}
 {bone1Name=dyn_sash_front_06, bone2Name=dyn_sash_front_05, swing1LimitDegrees=20, swing2LimitDegrees=40, twistLimitDegrees=20}
 {bone1Name=dyn_sash_front_07, bone2Name=dyn_sash_front_06, swing1LimitDegrees=20, swing2LimitDegrees=40, twistLimitDegrees=20}
 {bone1Name=dyn_sash_back_02, bone2Name=dyn_sash_back_01, swing1LimitDegrees=20, swing2LimitDegrees=20, twistLimitDegrees=20}
 {bone1Name=dyn_sash_back_03, bone2Name=dyn_sash_back_02, swing1LimitDegrees=20, swing2LimitDegrees=40, twistLimitDegrees=20}
 {bone1Name=dyn_sash_back_04, bone2Name=dyn_sash_back_03, swing1LimitDegrees=20, swing2LimitDegrees=40, twistLimitDegrees=20}
 {bone1Name=dyn_sash_back_05, bone2Name=dyn_sash_back_04, swing1LimitDegrees=20, swing2LimitDegrees=40, twistLimitDegrees=20}
 {bone1Name=dyn_scroll_02, bone2Name=dyn_scrolls_master, swing1LimitDegrees=3.5, swing2LimitDegrees=5, twistLimitDegrees=5}
 {bone1Name=dyn_scroll_01, bone2Name=dyn_scrolls_master, swing1LimitDegrees=5, swing2LimitDegrees=5, twistLimitDegrees=10}
 {
  bone1Name=dyn_scroll_02_02
  bone2Name=dyn_scroll_02
  swing1LimitDegrees=45
  swing1Motion=Free
  swing2LimitDegrees=45
  swing2Motion=Free
  twistLimitDegrees=45
  twistMotion=Free
 }
 {
  bone1Name=pelvis
  bone2Name=root
  swing1LimitDegrees=45
  swing1Motion=Free
  swing2LimitDegrees=45
  swing2Motion=Free
  twistLimitDegrees=45
  twistMotion=Free
 }
 {bone1Name=dyn_sash_back_01, bone2Name=dyn_scrolls_master, swing1LimitDegrees=15, swing2LimitDegrees=15, twistLimitDegrees=15}
 {bone1Name=dyn_sash_front_01, bone2Name=dyn_scrolls_master, swing1LimitDegrees=15, swing2LimitDegrees=15, twistLimitDegrees=15}
 {
  bone1Name=dyn_scrolls_master
  bone2Name=pelvis
  swing1LimitDegrees=45
  swing1Motion=Free
  swing2LimitDegrees=45
  swing2Motion=Free
  twistLimitDegrees=45
  twistMotion=Free
 }
 {
  bone1Name=calf_l
  bone2Name=thigh_l
  swing1LimitDegrees=45
  swing1Motion=Free
  swing2LimitDegrees=45
  swing2Motion=Free
  twistLimitDegrees=45
  twistMotion=Free
 }
 {bone1Name=dyn_l_belt_01, bone2Name=dyn_l_belt_master, swing1LimitDegrees=5, swing2LimitDegrees=10, twistLimitDegrees=3}
 {bone1Name=dyn_l_belt_02, bone2Name=dyn_l_belt_01, swing1LimitDegrees=5, swing2LimitDegrees=10, twistLimitDegrees=3}
 {bone1Name=dyn_l_belt_03, bone2Name=dyn_l_belt_02, swing1LimitDegrees=5, swing2LimitDegrees=10, twistLimitDegrees=3}
 {bone1Name=dyn_r_belt_01, bone2Name=dyn_r_belt_master, swing1LimitDegrees=5, swing2LimitDegrees=10, twistLimitDegrees=3}
 {bone1Name=dyn_r_belt_02, bone2Name=dyn_r_belt_01, swing1LimitDegrees=5, swing2LimitDegrees=10, twistLimitDegrees=3}
 {bone1Name=dyn_r_belt_03, bone2Name=dyn_r_belt_02, swing1LimitDegrees=5, swing2LimitDegrees=10, twistLimitDegrees=3}
 {bone1Name=dyn_mask_01, bone2Name=dyn_l_belt_03, swing1LimitDegrees=5, swing2LimitDegrees=5, twistLimitDegrees=5}
 {
  bone1Name=dyn_l_belt_master
  bone2Name=pelvis
  swing1LimitDegrees=45
  swing1Motion=Free
  swing2LimitDegrees=45
  swing2Motion=Free
  twistLimitDegrees=45
  twistMotion=Free
 }
 {
  bone1Name=dyn_r_belt_master
  bone2Name=pelvis
  swing1LimitDegrees=45
  swing1Motion=Free
  swing2LimitDegrees=45
  swing2Motion=Free
  twistLimitDegrees=45
  twistMotion=Free
 }
 {bone1Name=dyn_pant_leg_back_r_master, bone2Name=calf_l, swing1LimitDegrees=35, swing2LimitDegrees=35, twistLimitDegrees=5}
 {bone1Name=dyn_pant_leg_back_r_01, bone2Name=dyn_pant_leg_back_r_master, swing1LimitDegrees=35, swing2LimitDegrees=35, twistLimitDegrees=5}
 {bone1Name=dyn_pant_leg_back_r_02, bone2Name=dyn_pant_leg_back_r_01, swing1LimitDegrees=35, swing2LimitDegrees=35, twistLimitDegrees=5}
 {bone1Name=dyn_pant_leg_back_r_03, bone2Name=dyn_pant_leg_back_r_02, swing1LimitDegrees=35, swing2LimitDegrees=35, twistLimitDegrees=5}
 {bone1Name=dyn_pant_leg_back_l_master, bone2Name=calf_l, swing1LimitDegrees=35, swing2LimitDegrees=35, twistLimitDegrees=5}
 {bone1Name=dyn_pant_leg_back_l_01, bone2Name=dyn_pant_leg_back_l_master, swing1LimitDegrees=35, swing2LimitDegrees=35, twistLimitDegrees=5}
 {bone1Name=dyn_pant_leg_back_l_02, bone2Name=dyn_pant_leg_back_l_01, swing1LimitDegrees=35, swing2LimitDegrees=35, twistLimitDegrees=5}
 {bone1Name=dyn_pant_leg_back_l_03, bone2Name=dyn_pant_leg_back_l_02, swing1LimitDegrees=35, swing2LimitDegrees=35, twistLimitDegrees=5}
 {bone1Name=dyn_pant_leg_left_master, bone2Name=calf_l, swing1LimitDegrees=35, swing2LimitDegrees=35, twistLimitDegrees=5}
 {bone1Name=dyn_pant_leg_left_01, bone2Name=dyn_pant_leg_left_master, swing1LimitDegrees=35, swing2LimitDegrees=35, twistLimitDegrees=5}
 {bone1Name=dyn_pant_leg_left_02, bone2Name=dyn_pant_leg_left_01, swing1LimitDegrees=35, swing2LimitDegrees=35, twistLimitDegrees=5}
 {bone1Name=dyn_pant_leg_left_03, bone2Name=dyn_pant_leg_left_02, swing1LimitDegrees=35, swing2LimitDegrees=35, twistLimitDegrees=5}
 {bone1Name=dyn_pant_leg_right_master, bone2Name=calf_l, swing1LimitDegrees=35, swing2LimitDegrees=35, twistLimitDegrees=5}
 {bone1Name=dyn_pant_leg_right_01, bone2Name=dyn_pant_leg_right_master, swing1LimitDegrees=35, swing2LimitDegrees=35, twistLimitDegrees=5}
 {bone1Name=dyn_pant_leg_right_02, bone2Name=dyn_pant_leg_right_01, swing1LimitDegrees=35, swing2LimitDegrees=35, twistLimitDegrees=5}
 {bone1Name=dyn_pant_leg_right_03, bone2Name=dyn_pant_leg_right_02, swing1LimitDegrees=35, swing2LimitDegrees=35, twistLimitDegrees=5}
 {bone1Name=dyn_pant_leg_front_master, bone2Name=calf_l, swing1LimitDegrees=35, swing2LimitDegrees=35, twistLimitDegrees=5}
 {bone1Name=dyn_pant_leg_front_01, bone2Name=dyn_pant_leg_front_master, swing1LimitDegrees=35, swing2LimitDegrees=35, twistLimitDegrees=5}
 {bone1Name=dyn_pant_leg_front_02, bone2Name=dyn_pant_leg_front_01, swing1LimitDegrees=35, swing2LimitDegrees=35, twistLimitDegrees=5}
 {bone1Name=dyn_pant_leg_front_03, bone2Name=dyn_pant_leg_front_02, swing1LimitDegrees=35, swing2LimitDegrees=35, twistLimitDegrees=5}
 {
  bone1Name=thigh_l
  bone2Name=pelvis
  swing1LimitDegrees=45
  swing1Motion=Free
  swing2LimitDegrees=45
  swing2Motion=Free
  twistLimitDegrees=45
  twistMotion=Free
 }
 {
  bone1Name=thigh_l
  bone2Name=dyn_l_belt_03
  swing1LimitDegrees=45
  swing1Motion=Free
  swing2LimitDegrees=45
  swing2Motion=Free
  twistLimitDegrees=45
  twistMotion=Free
 }
 {
  bone1Name=thigh_r
  bone2Name=dyn_r_belt_03
  swing1LimitDegrees=45
  swing1Motion=Free
  swing2LimitDegrees=45
  swing2Motion=Free
  twistLimitDegrees=45
  twistMotion=Free
 }
]
