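# Crouch physics asset for the medium male Cube Ninja player body, with collision and joints for its dangling costume parts.
# Capsules on the spine, pelvis, legs and left arm link by mostly free constraints, while a large root box forms a floor slab below the origin.
# Limited swing and twist chains shape the pant flaps, ribbons, sashes, scrolls and crystal.
# The mask and first sash bodies have no constraint to the body.
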PhysicsAsset /Game/Characters/Player/Male/Medium/Bodies/M_MED_Cube_Ninja/Meshes/M_MED_Cube_Ninja_Physics_Crouch
bodies {
 calf_l [{shapeType=Capsule, length=9.7147, radius=6.046, center=<-16.3497 -2.6658 -0.4358>, rotation=<rot 90 0 0>}]
 calf_r [
  {shapeType=Capsule, length=26.4517, radius=10.6586, center=<20.9218 1.6597 1.093>, rotation=<rot 90 0 0>}
  {shapeType=Capsule, length=31.1859, radius=7.8991, center=<20.6634 13.6899 1.0934>, rotation=<rot 90 0 0>}
  {shapeType=Capsule, length=7.2189, radius=7.8991, center=<32.1948 24.6747 1.0934>, rotation=<rot 90 0 0>}
 ]
 calf_twist_01_l [{shapeType=Capsule, length=0.0002, radius=12.9335, rotation=<rot 90 0 0>}]
 calf_twist_02_l [{shapeType=Capsule, length=23.1819, radius=11.5875, rotation=<rot 90 0 0>}]
 dyn_crystal_01 [{shapeType=Capsule, length=0.4029, radius=1.2051, center=<1.646 -0.0001 0>, rotation=<rot 90 0 0>}]
 dyn_crystal_02 [{shapeType=Capsule, length=0, radius=0.5783, center=<0.393 0 0>, rotation=<rot 90 0 0>}]
 dyn_crystal_03 [{shapeType=Capsule, length=9.6156, radius=2.5852, center=<7.4498 0 0>, rotation=<rot 83.5399 0.0003 0.0003>}]
 dyn_crystal_master [{shapeType=Capsule, length=0.505, radius=0.505, rotation=<rot 90 0 0>}]
 dyn_l_arm_ribbon_01 [{shapeType=Capsule, length=1.3, radius=0.78, center=<3.6617 -0.0001 0>, rotation=<rot 0 -90 -90>}]
 dyn_l_arm_ribbon_02 [{shapeType=Box, extentX=6.6455, extentY=0.9987, extentZ=4.316, center=<3.7181 0.3373 0>, rotation=<rot 0 0 90>}]
 dyn_l_arm_ribbon_03 [{shapeType=Box, extentX=4.1649, extentY=0.926, extentZ=3.7271, center=<2.3289 0 0>, rotation=<rot 0 0 90>}]
 dyn_l_arm_ribbon_master [{shapeType=Box, extentX=1.01, extentY=1.01, extentZ=1.01, rotation=<rot 0 -90 0>}]
 dyn_leg_ribbon_01 [{shapeType=Box, extentX=2, extentY=2, extentZ=0.9495, center=<0.8592 0 0>}]
 dyn_leg_ribbon_02 [{shapeType=Box, extentX=2, extentY=2.6206, extentZ=1.2167, center=<1.0159 0 0>}]
 dyn_leg_ribbon_03 [{shapeType=Box, extentX=2, extentY=2.9932, extentZ=1.0092, center=<0.995 0 0>}]
 dyn_leg_ribbon_ball [{shapeType=Capsule, length=0.505, radius=0.505}]
 dyn_leg_ribbon_top_01 [{shapeType=Box, extentX=2, extentY=2, extentZ=0.9664, center=<1.2545 0 0>}]
 dyn_leg_ribbon_top_02 [{shapeType=Box, extentX=2, extentY=2.6694, extentZ=1.2131, center=<1.3051 0 0>}]
 dyn_leg_ribbon_top_03 [{shapeType=Box, extentX=2, extentY=2.5949, extentZ=0.8341, center=<1.0151 0 0.0001>}]
 dyn_mask_01 [{shapeType=Box, extentX=3.9147, extentY=17.1778, extentZ=27.6711, center=<0.1169 0 -0.6883>, rotation=<rot 9.641 0 0>}]
 dyn_pant_leg_back_l_01 [{shapeType=Box, extentX=5, extentY=7.6701, extentZ=2.1579, center=<2.7365 0 0>}]
 dyn_pant_leg_back_l_02 [{shapeType=Box, extentX=5.872, extentY=7.6701, extentZ=2.1579, center=<3.032 0.0001 0>}]
 dyn_pant_leg_back_l_03 [{shapeType=Box, extentX=5, extentY=7.6701, extentZ=2.1579, center=<2.5164 0 -0.0001>}]
 dyn_pant_leg_back_r_01 [{shapeType=Box, extentX=5, extentY=9.1633, extentZ=2.652, center=<2.5443 0 0>}]
 dyn_pant_leg_back_r_02 [{shapeType=Box, extentX=5.7461, extentY=9.1633, extentZ=2.652, center=<3.1819 0 0>}]
 dyn_pant_leg_back_r_03 [{shapeType=Box, extentX=5, extentY=9.1633, extentZ=2.652, center=<2.4197 0 0.0003>}]
 dyn_pant_leg_front_01 [{shapeType=Box, extentX=5, extentY=8.7027, extentZ=1.8027, center=<2.7237 0 -0.0001>}]
 dyn_pant_leg_front_02 [{shapeType=Box, extentX=5.3081, extentY=8.7027, extentZ=1.8027, center=<2.8152 0 -0.0001>}]
 dyn_pant_leg_front_03 [{shapeType=Box, extentX=5, extentY=8.7027, extentZ=1.8027, center=<2.3836 0 0>}]
 dyn_pant_leg_left_01 [{shapeType=Box, extentX=5, extentY=8.5846, extentZ=1.9809, center=<2.6623 0 -0.0001>}]
 dyn_pant_leg_left_02 [{shapeType=Box, extentX=5.6824, extentY=8.5846, extentZ=1.9809, center=<3.0791 0 0>}]
 dyn_pant_leg_left_03 [{shapeType=Box, extentX=5, extentY=8.5846, extentZ=1.9809, center=<2.5295 0 -0.0001>}]
 dyn_pant_leg_right_01 [{shapeType=Box, extentX=5, extentY=9.8783, extentZ=1.9545, center=<2.7422 0 0>}]
 dyn_pant_leg_right_02 [{shapeType=Box, extentX=5.6581, extentY=9.8783, extentZ=1.9545, center=<2.8589 -0.0002 0>}]
 dyn_pant_leg_right_03 [{shapeType=Box, extentX=5, extentY=9.8783, extentZ=1.9545, center=<2.4603 0 0>}]
 dyn_sash_back_01 [{shapeType=Box, extentX=4.7154, extentY=4.535, extentZ=1.5106, center=<2.345 -0.0001 0>}]
 dyn_sash_back_01_05 [{shapeType=Capsule, length=8.433, radius=2.59, center=<6.9893 0.0003 0.2262>, rotation=<rot 90 0 0>}]
 dyn_sash_back_02 [{shapeType=Capsule, length=7.4249, radius=3.4539, center=<7.0438 0.2497 0.3827>, rotation=<rot 0 95.556 -89.7544>}]
 dyn_sash_back_03 [{shapeType=Capsule, length=8.5962, radius=2.4846, center=<7.498 0.4648 0.8378>, rotation=<rot 0 89.8912 -89.3228>}]
 dyn_sash_back_04 [{shapeType=Capsule, length=3.5693, radius=2.265, center=<3.6127 0.4917 0.8456>, rotation=<rot 0 108.9013 -86.5434>}]
 dyn_sash_back_05 [{shapeType=Capsule, length=2.7566, radius=1.2906, center=<0.9146 0.4357 0.1024>, rotation=<rot 0 114.5511 -86.3906>}]
 dyn_sash_front_01 [{shapeType=Box, extentX=3.5, extentY=4.2597, extentZ=1.4616, center=<2.6661 0 0>}]
 dyn_sash_front_02 [{shapeType=Capsule, length=1.7849, radius=3.8723, center=<3.9793 -0.0001 1.4536>, rotation=<rot 90 0 0>}]
 dyn_sash_front_03 [{shapeType=Capsule, length=4.6692, radius=3.0358, center=<4.7385 -0.0292 0.1982>, rotation=<rot 9.1304 97.1218 -86.5902>}]
 dyn_sash_front_04 [{shapeType=Capsule, length=7.0194, radius=2.759, center=<6.5787 -0.3455 0.3332>, rotation=<rot 0 89.7173 -89.6393>}]
 dyn_sash_front_05 [{shapeType=Capsule, length=5.692, radius=2.7543, center=<5.7073 -0.206 0.463>, rotation=<rot 0 93.8958 -88.1457>}]
 dyn_sash_front_06 [{shapeType=Capsule, length=5.1019, radius=2.374, center=<4.9174 0.0167 0.4631>, rotation=<rot 0 -91.6411 -87.8661>}]
 dyn_sash_front_07 [{shapeType=Capsule, length=5.3293, radius=2.3517, center=<5.789 -0.0382 0.0669>, rotation=<rot 0 86.3821 -89.1753>}]
 dyn_sash_front_08 [{shapeType=Capsule, length=3.469, radius=2.4816, center=<2.6115 0.2153 -0.1534>, rotation=<rot 0 31.6907 -64.7808>}]
 dyn_scroll_01 [{shapeType=Capsule, length=9.5253, radius=4.9914, center=<0.2673 0 -0.2316>, rotation=<rot 0 0 90>}]
 dyn_scroll_02 [{shapeType=Capsule, length=14.6489, radius=3.1994, rotation=<rot 0 0 90>}]
 dyn_scroll_02_02 [{shapeType=Capsule, length=14.6489, radius=3.1994, rotation=<rot 0 0 90>}]
 foot_r [{shapeType=Capsule, length=10.188, radius=9.5974, center=<1.6187 -7.2871 -0.0001>, rotation=<rot 0 12.5239 -89.9999>}]
 hand_l [{shapeType=Capsule, length=5.0105, radius=4.6395, center=<4.2225 0.0001 0>, rotation=<rot 90 0 0>}]
 lowerarm_l [{shapeType=Capsule, length=13.025, radius=3.87, center=<9.5567 0.5413 0>, rotation=<rot 90 -90 -89.0154>}]
 pelvis [{shapeType=Capsule, length=7.4052, radius=10.4026, rotation=<rot 0 -90 0>}]
 root [{shapeType=Box, extentX=200, extentY=200, extentZ=50, center=<0 0 -25>}]
 spine_01 [{shapeType=Capsule, length=13.2875, radius=8.2727}]
 spine_02 [{shapeType=Capsule, length=7.5184, radius=9.6245}]
 spine_03 [{shapeType=Capsule, length=6.7695, radius=11.0251}]
 spine_04 [{shapeType=Capsule, length=7.2886, radius=11.8222}]
 thigh_l [
  {shapeType=Capsule, length=21.075, radius=9.081, center=<-19.3589 -0.0001 0>, rotation=<rot 90 0 0>}
  {shapeType=Capsule, length=21.075, radius=9.081, center=<-19.6197 -0.0005 6.3258>, rotation=<rot 90 0 0>}
  {shapeType=Capsule, length=20.65, radius=7.8288, center=<-19.3156 -0.0002 0>, rotation=<rot 90 0 0>}
 ]
 thigh_r [{shapeType=Capsule, length=23.4851, radius=10.5923, center=<21.846 1.1661 -0.1547>, rotation=<rot 87.6764 -180 -180>}]
}
constraints [
 {
  bone1Name=calf_l
  bone2Name=thigh_l
  swing1LimitDegrees=45
  swing1Motion=Free
  swing2LimitDegrees=45
  swing2Motion=Free
  twistLimitDegrees=45
  twistMotion=Free
 }
 {bone1Name=dyn_pant_leg_back_r_01, bone2Name=calf_l, swing1LimitDegrees=35, swing2LimitDegrees=35, twistLimitDegrees=5}
 {bone1Name=dyn_pant_leg_back_r_02, bone2Name=dyn_pant_leg_back_r_01, swing1LimitDegrees=35, swing2LimitDegrees=35, twistLimitDegrees=5}
 {bone1Name=dyn_pant_leg_back_r_03, bone2Name=dyn_pant_leg_back_r_02, swing1LimitDegrees=35, swing2LimitDegrees=35, twistLimitDegrees=5}
 {bone1Name=dyn_pant_leg_back_l_01, bone2Name=calf_l, swing1LimitDegrees=35, swing2LimitDegrees=35, twistLimitDegrees=5}
 {bone1Name=dyn_pant_leg_back_l_02, bone2Name=dyn_pant_leg_back_l_01, swing1LimitDegrees=35, swing2LimitDegrees=35, twistLimitDegrees=5}
 {bone1Name=dyn_pant_leg_back_l_03, bone2Name=dyn_pant_leg_back_l_02, swing1LimitDegrees=35, swing2LimitDegrees=35, twistLimitDegrees=5}
 {bone1Name=dyn_pant_leg_left_01, bone2Name=calf_l, swing1LimitDegrees=35, swing2LimitDegrees=35, twistLimitDegrees=5}
 {bone1Name=dyn_pant_leg_left_02, bone2Name=dyn_pant_leg_left_01, swing1LimitDegrees=35, swing2LimitDegrees=35, twistLimitDegrees=5}
 {bone1Name=dyn_pant_leg_left_03, bone2Name=dyn_pant_leg_left_02, swing1LimitDegrees=35, swing2LimitDegrees=35, twistLimitDegrees=5}
 {bone1Name=dyn_pant_leg_right_01, bone2Name=calf_l, swing1LimitDegrees=35, swing2LimitDegrees=35, twistLimitDegrees=5}
 {bone1Name=dyn_pant_leg_right_02, bone2Name=dyn_pant_leg_right_01, swing1LimitDegrees=35, swing2LimitDegrees=35, twistLimitDegrees=5}
 {bone1Name=dyn_pant_leg_right_03, bone2Name=dyn_pant_leg_right_02, swing1LimitDegrees=35, swing2LimitDegrees=35, twistLimitDegrees=5}
 {bone1Name=dyn_pant_leg_front_01, bone2Name=calf_l, swing1LimitDegrees=35, swing2LimitDegrees=35, twistLimitDegrees=5}
 {bone1Name=dyn_pant_leg_front_02, bone2Name=dyn_pant_leg_front_01, swing1LimitDegrees=35, swing2LimitDegrees=35, twistLimitDegrees=5}
 {bone1Name=dyn_pant_leg_front_03, bone2Name=dyn_pant_leg_front_02, swing1LimitDegrees=35, swing2LimitDegrees=35, twistLimitDegrees=5}
 {
  bone1Name=thigh_r
  bone2Name=pelvis
  swing1LimitDegrees=45
  swing1Motion=Free
  swing2LimitDegrees=45
  swing2Motion=Free
  twistLimitDegrees=45
  twistMotion=Free
 }
 {
  bone1Name=calf_r
  bone2Name=thigh_r
  swing1LimitDegrees=45
  swing1Motion=Free
  swing2LimitDegrees=45
  swing2Motion=Free
  twistLimitDegrees=45
  twistMotion=Free
 }
 {
  bone1Name=dyn_leg_ribbon_ball
  bone2Name=calf_r
  swing1LimitDegrees=45
  swing1Motion=Free
  swing2LimitDegrees=45
  swing2Motion=Free
  twistLimitDegrees=45
  twistMotion=Free
 }
 {bone1Name=dyn_leg_ribbon_top_01, bone2Name=dyn_leg_ribbon_ball, swing1LimitDegrees=15, swing2LimitDegrees=15, twistLimitDegrees=5}
 {bone1Name=dyn_leg_ribbon_top_02, bone2Name=dyn_leg_ribbon_top_01, swing1LimitDegrees=15, swing2LimitDegrees=15, twistLimitDegrees=5}
 {bone1Name=dyn_leg_ribbon_top_03, bone2Name=dyn_leg_ribbon_top_02, swing1LimitDegrees=15, swing2LimitDegrees=15, twistLimitDegrees=5}
 {bone1Name=dyn_leg_ribbon_01, bone2Name=dyn_leg_ribbon_ball, swing1LimitDegrees=15, swing2LimitDegrees=15, twistLimitDegrees=5}
 {bone1Name=dyn_leg_ribbon_02, bone2Name=dyn_leg_ribbon_01, swing1LimitDegrees=15, swing2LimitDegrees=15, twistLimitDegrees=5}
 {bone1Name=dyn_leg_ribbon_03, bone2Name=dyn_leg_ribbon_02, swing1LimitDegrees=15, swing2LimitDegrees=15, twistLimitDegrees=5}
 {bone1Name=dyn_l_arm_ribbon_03, bone2Name=dyn_l_arm_ribbon_02, swing1LimitDegrees=45, swing2LimitDegrees=45, twistLimitDegrees=20}
 {bone1Name=dyn_l_arm_ribbon_01, bone2Name=dyn_l_arm_ribbon_master, swing1LimitDegrees=60, swing2LimitDegrees=60, twistLimitDegrees=10}
 {bone1Name=dyn_l_arm_ribbon_02, bone2Name=dyn_l_arm_ribbon_01, swing1LimitDegrees=45, swing2LimitDegrees=45, twistLimitDegrees=20}
 {
  bone1Name=dyn_l_arm_ribbon_master
  bone2Name=lowerarm_l
  swing1LimitDegrees=45
  swing1Motion=Free
  swing2LimitDegrees=45
  swing2Motion=Free
  twistLimitDegrees=45
  twistMotion=Free
 }
 {
  bone1Name=lowerarm_l
  bone2Name=spine_04
  swing1LimitDegrees=45
  swing1Motion=Free
  swing2LimitDegrees=45
  swing2Motion=Free
  twistLimitDegrees=45
  twistMotion=Free
 }
 {
  bone1Name=hand_l
  bone2Name=lowerarm_l
  swing1LimitDegrees=45
  swing1Motion=Free
  swing2LimitDegrees=45
  swing2Motion=Free
  twistLimitDegrees=45
  twistMotion=Free
 }
 {bone1Name=dyn_sash_front_08, bone2Name=dyn_sash_front_07, swing1LimitDegrees=20, swing2LimitDegrees=20, twistLimitDegrees=20}
 {
  bone1Name=dyn_sash_front_02
  bone2Name=dyn_sash_front_01
  swing1LimitDegrees=10
  swing1Motion=Locked
  swing2LimitDegrees=10
  swing2Motion=Locked
  twistLimitDegrees=10
  twistMotion=Locked
 }
 {bone1Name=dyn_sash_front_03, bone2Name=dyn_sash_front_02, swing1LimitDegrees=20, swing2LimitDegrees=20, twistLimitDegrees=15}
 {bone1Name=dyn_sash_front_04, bone2Name=dyn_sash_front_03, swing1LimitDegrees=20, swing2LimitDegrees=20, twistLimitDegrees=15}
 {bone1Name=dyn_sash_front_05, bone2Name=dyn_sash_front_04, swing1LimitDegrees=20, swing2LimitDegrees=20, twistLimitDegrees=20}
 {bone1Name=dyn_sash_front_06, bone2Name=dyn_sash_front_05, swing1LimitDegrees=20, swing2LimitDegrees=20, twistLimitDegrees=20}
 {bone1Name=dyn_sash_front_07, bone2Name=dyn_sash_front_06, swing1LimitDegrees=20, swing2LimitDegrees=20, twistLimitDegrees=20}
 {bone1Name=dyn_sash_back_02, bone2Name=dyn_sash_back_01_05, swing1LimitDegrees=20, swing2LimitDegrees=20, twistLimitDegrees=5}
 {bone1Name=dyn_sash_back_03, bone2Name=dyn_sash_back_02, swing1LimitDegrees=45, swing2LimitDegrees=45, twistLimitDegrees=20}
 {bone1Name=dyn_sash_back_04, bone2Name=dyn_sash_back_03, swing1LimitDegrees=45, swing2LimitDegrees=45, twistLimitDegrees=20}
 {bone1Name=dyn_sash_back_05, bone2Name=dyn_sash_back_04, swing1LimitDegrees=45, swing2LimitDegrees=45, twistLimitDegrees=20}
 {
  bone1Name=thigh_l
  bone2Name=pelvis
  swing1LimitDegrees=45
  swing1Motion=Free
  swing2LimitDegrees=45
  swing2Motion=Free
  twistLimitDegrees=45
  twistMotion=Free
 }
 {bone1Name=dyn_scroll_02, bone2Name=pelvis, swing1LimitDegrees=5, swing2LimitDegrees=5, twistLimitDegrees=7.5}
 {bone1Name=dyn_scroll_01, bone2Name=pelvis, swing1LimitDegrees=5, swing2LimitDegrees=5, twistLimitDegrees=7.5}
 {
  bone1Name=dyn_scroll_02_02
  bone2Name=dyn_scroll_02
  swing1LimitDegrees=45
  swing1Motion=Free
  swing2LimitDegrees=45
  swing2Motion=Free
  twistLimitDegrees=45
  twistMotion=Free
 }
 {
  bone1Name=spine_02
  bone2Name=spine_01
  swing1LimitDegrees=45
  swing1Motion=Free
  swing2LimitDegrees=45
  swing2Motion=Free
  twistLimitDegrees=45
  twistMotion=Free
 }
 {
  bone1Name=dyn_crystal_master
  bone2Name=pelvis
  swing1LimitDegrees=45
  swing1Motion=Free
  swing2LimitDegrees=45
  swing2Motion=Free
  twistLimitDegrees=45
  twistMotion=Free
 }
 {bone1Name=dyn_crystal_01, bone2Name=dyn_crystal_master, swing1LimitDegrees=3, swing2LimitDegrees=3, twistLimitDegrees=3}
 {bone1Name=dyn_crystal_02, bone2Name=dyn_crystal_01, swing1LimitDegrees=20, swing2LimitDegrees=40, twistLimitDegrees=10}
 {bone1Name=dyn_crystal_03, bone2Name=dyn_crystal_02, swing1LimitDegrees=45, swing2LimitDegrees=45, twistLimitDegrees=20}
 {
  bone1Name=spine_01
  bone2Name=pelvis
  swing1LimitDegrees=45
  swing1Motion=Free
  swing2LimitDegrees=45
  swing2Motion=Free
  twistLimitDegrees=45
  twistMotion=Free
 }
 {
  bone1Name=pelvis
  bone2Name=root
  swing1LimitDegrees=45
  swing1Motion=Free
  swing2LimitDegrees=45
  swing2Motion=Free
  twistLimitDegrees=45
  twistMotion=Free
 }
 {
  bone1Name=foot_r
  bone2Name=calf_r
  swing1LimitDegrees=45
  swing1Motion=Free
  swing2LimitDegrees=45
  swing2Motion=Free
  twistLimitDegrees=45
  twistMotion=Free
 }
 {
  bone1Name=spine_03
  bone2Name=spine_02
  swing1LimitDegrees=45
  swing1Motion=Free
  swing2LimitDegrees=45
  swing2Motion=Free
  twistLimitDegrees=45
  twistMotion=Free
 }
 {
  bone1Name=spine_04
  bone2Name=spine_03
  swing1LimitDegrees=45
  swing1Motion=Free
  swing2LimitDegrees=45
  swing2Motion=Free
  twistLimitDegrees=45
  twistMotion=Free
 }
 {
  bone1Name=calf_twist_02_l
  bone2Name=calf_l
  swing1LimitDegrees=45
  swing1Motion=Free
  swing2LimitDegrees=45
  swing2Motion=Free
  twistLimitDegrees=45
  twistMotion=Free
 }
 {bone1Name=dyn_sash_back_01_05, bone2Name=dyn_sash_back_01, swing1LimitDegrees=20, swing2LimitDegrees=20, twistLimitDegrees=5}
 {
  bone1Name=calf_twist_01_l
  bone2Name=calf_l
  swing1LimitDegrees=45
  swing1Motion=Free
  swing2LimitDegrees=45
  swing2Motion=Free
  twistLimitDegrees=45
  twistMotion=Free
 }
]
