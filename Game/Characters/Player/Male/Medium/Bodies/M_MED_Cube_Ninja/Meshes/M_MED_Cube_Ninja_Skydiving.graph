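# Physics asset for the Cube Ninja medium male body's skydiving variant, covering dangling costume pieces.
# Capsules approximate pelvis, spine, legs and left forearm and hand, joined by free angular constraints, with a flat box below root.
# Angle-limited chains form left-calf pant flaps, right-calf and left-arm ribbons, sashes, and pelvis-hung scrolls and crystal.
# The mask body and both sash chains' first bodies have no constraint to the body.
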PhysicsAsset /Game/Characters/Player/Male/Medium/Bodies/M_MED_Cube_Ninja/Meshes/M_MED_Cube_Ninja_Skydiving
bodies {
 calf_l [{shapeType=Capsule, length=9.7147, radius=6.046, center=<-16.3497 -2.6658 -0.4358>, rotation=<rot 90 0 0>}]
 calf_r [{shapeType=Capsule, length=25.9235, radius=11.1867, center=<20.9218 1.6597 1.093>, rotation=<rot 90 0 0>}]
 dyn_crystal_01 [{shapeType=Capsule, length=0.4029, radius=1.2051, center=<1.646 -0.0001 0>, rotation=<rot 90 0 0>}]
 dyn_crystal_02 [{shapeType=Capsule, length=0, radius=0.5783, center=<0.393 0 0>, rotation=<rot 90 0 0>}]
 dyn_crystal_03 [{shapeType=Capsule, length=9.6156, radius=2.5852, center=<7.4498 0 0>, rotation=<rot 83.5399 0.0003 0.0003>}]
 dyn_crystal_master [{shapeType=Capsule, length=0.505, radius=0.505, rotation=<rot 90 0 0>}]
 dyn_l_arm_ribbon_01 [{shapeType=Capsule, length=2.2516, radius=1.5507, center=<2.9536 0.3785 0.1003>, rotation=<rot 0 -92.3575 -87.0882>}]
 dyn_l_arm_ribbon_02 [{shapeType=Box, extentX=6.6455, extentY=0.9987, extentZ=4.316, center=<3.7181 0.3373 0>, rotation=<rot 0 0 90>}]
 dyn_l_arm_ribbon_03 [{shapeType=Box, extentX=4.1649, extentY=0.926, extentZ=3.7271, center=<2.3289 0 0>, rotation=<rot 0 0 90>}]
 dyn_l_arm_ribbon_master [{shapeType=Box, extentX=1.01, extentY=1.01, extentZ=1.01, rotation=<rot 0 -90 0>}]
 dyn_leg_ribbon_01 [{shapeType=Box, extentX=2, extentY=2, extentZ=0.9495, center=<0.8592 0 0>}]
 dyn_leg_ribbon_02 [{shapeType=Box, extentX=2, extentY=2.6206, extentZ=1.2167, center=<1.0159 0 0>}]
 dyn_leg_ribbon_03 [{shapeType=Box, extentX=2, extentY=2.9932, extentZ=1.0092, center=<0.995 0 0>}]
 dyn_leg_ribbon_ball [{shapeType=Capsule, length=0.505, radius=0.505}]
 dyn_leg_ribbon_top_01 [{shapeType=Box, extentX=2, extentY=2, extentZ=0.9664, center=<1.2545 0 0>}]
 dyn_leg_ribbon_top_02 [{shapeType=Box, extentX=2, extentY=2.6694, extentZ=1.2131, center=<1.3051 0 0>}]
 dyn_leg_ribbon_top_03 [{shapeType=Box, extentX=2, extentY=2.5949, extentZ=0.8341, center=<1.0151 0 0.0001>}]
 dyn_mask_01 [{shapeType=Box, extentX=4.1615, extentY=17.1778, extentZ=21.3, center=<0.1169 0 -0.6883>, rotation=<rot 9.641 0 0>}]
 dyn_pant_leg_back_l_01 [{shapeType=Box, extentX=5, extentY=7.6701, extentZ=2.1579, center=<2.7365 0 0>}]
 dyn_pant_leg_back_l_02 [{shapeType=Box, extentX=5.872, extentY=7.6701, extentZ=2.1579, center=<3.032 0.0001 0>}]
 dyn_pant_leg_back_l_03 [{shapeType=Box, extentX=5, extentY=7.6701, extentZ=2.1579, center=<2.5164 0 -0.0001>}]
 dyn_pant_leg_back_r_01 [{shapeType=Box, extentX=5, extentY=9.1633, extentZ=2.652, center=<2.5443 0 0>}]
 dyn_pant_leg_back_r_02 [{shapeType=Box, extentX=5.7461, extentY=9.1633, extentZ=2.652, center=<3.1819 0 0>}]
 dyn_pant_leg_back_r_03 [{shapeType=Box, extentX=5, extentY=9.1633, extentZ=2.652, center=<2.4197 0 0.0003>}]
 dyn_pant_leg_front_01 [{shapeType=Box, extentX=5, extentY=8.7027, extentZ=1.8027, center=<2.7237 0 -0.0001>}]
 dyn_pant_leg_front_02 [{shapeType=Box, extentX=5.3081, extentY=8.7027, extentZ=1.8027, center=<2.8152 0 -0.0001>}]
 dyn_pant_leg_front_03 [{shapeType=Box, extentX=5, extentY=8.7027, extentZ=1.8027, center=<2.3836 0 0>}]
 dyn_pant_leg_left_01 [{shapeType=Box, extentX=5, extentY=8.5846, extentZ=1.9809, center=<2.6623 0 -0.0001>}]
 dyn_pant_leg_left_02 [{shapeType=Box, extentX=5.6824, extentY=8.5846, extentZ=1.9809, center=<3.0791 0 0>}]
 dyn_pant_leg_left_03 [{shapeType=Box, extentX=5, extentY=8.5846, extentZ=1.9809, center=<2.5295 0 -0.0001>}]
 dyn_pant_leg_right_01 [{shapeType=Box, extentX=5, extentY=9.8783, extentZ=1.9545, center=<2.7422 0 0>}]
 dyn_pant_leg_right_02 [{shapeType=Box, extentX=5.6581, extentY=9.8783, extentZ=1.9545, center=<2.8589 -0.0002 0>}]
 dyn_pant_leg_right_03 [{shapeType=Box, extentX=5, extentY=9.8783, extentZ=1.9545, center=<2.4603 0 0>}]
 dyn_sash_back_01 [{shapeType=Box, extentX=10.6591, extentY=4.535, extentZ=1.5106, center=<5.9181 -0.0002 0>}]
 dyn_sash_back_02 [{shapeType=Box, extentX=14.0493, extentY=4.8321, extentZ=2.0576, center=<7.003 0.1638 0>}]
 dyn_sash_back_03 [{shapeType=Box, extentX=12.7604, extentY=5.0416, extentZ=3.1184, center=<6.5067 0.2383 0>}]
 dyn_sash_back_04 [{shapeType=Box, extentX=8.3456, extentY=3.5, extentZ=3.0803, center=<4.6067 0.2273 0>}]
 dyn_sash_back_05 [{shapeType=Box, extentX=3.5, extentY=3.5, extentZ=2.7191, center=<2.0747 0 0>}]
 dyn_sash_front_01 [{shapeType=Box, extentX=3.5, extentY=4.2597, extentZ=1.4616, center=<2.6661 0 0>}]
 dyn_sash_front_02 [{shapeType=Box, extentX=7.209, extentY=5.5772, extentZ=1.4112, center=<3.9276 0.0002 0.281>}]
 dyn_sash_front_03 [{shapeType=Box, extentX=10.0857, extentY=5.6979, extentZ=1.9803, center=<4.9563 0.0001 0.1083>}]
 dyn_sash_front_04 [{shapeType=Box, extentX=13.4756, extentY=5.9422, extentZ=2.3145, center=<6.537 -0.4678 0>}]
 dyn_sash_front_05 [{shapeType=Box, extentX=11.3889, extentY=4.9582, extentZ=1.9892, center=<6.0105 -0.477 0>}]
 dyn_sash_front_06 [{shapeType=Box, extentX=10.8852, extentY=4.2597, extentZ=2.0482, center=<6.0675 0 0>}]
 dyn_sash_front_07 [{shapeType=Box, extentX=11.14, extentY=4.2597, extentZ=2.7099, center=<5.8194 0 0>}]
 dyn_sash_front_08 [{shapeType=Box, extentX=8.2024, extentY=6.3657, extentZ=3.5, center=<4.2059 -0.0001 0>}]
 dyn_scroll_01 [{shapeType=Capsule, length=12.647, radius=4.6806, center=<0.2673 0 -0.2316>, rotation=<rot 0 0 90>}]
 dyn_scroll_02 [{shapeType=Capsule, length=14.6489, radius=3.1994, rotation=<rot 0 0 90>}]
 dyn_scroll_02_02 [{shapeType=Capsule, length=14.0093, radius=3.839, rotation=<rot 0 0 90>}]
 hand_l [{shapeType=Capsule, length=5.0105, radius=4.6395, center=<4.2225 0.0001 0>, rotation=<rot 90 0 0>}]
 lowerarm_l [{shapeType=Capsule, length=12.7762, radius=4.122, center=<9.5567 0.5413 0>, rotation=<rot 90 -90 -89.0154>}]
 pelvis [{shapeType=Capsule, length=6.6712, radius=11.1366, rotation=<rot 0 -90 0>}]
 root [{shapeType=Box, extentX=200, extentY=200, extentZ=50, center=<0 0 -25>}]
 spine_01 [{shapeType=Capsule, length=13.2875, radius=8.2727}]
 spine_02 [{shapeType=Capsule, length=7.5184, radius=9.6245}]
 thigh_l [
  {shapeType=Capsule, length=21.075, radius=9.081, center=<-19.3589 -0.0001 0>, rotation=<rot 90 0 0>}
  {shapeType=Capsule, length=21.075, radius=9.081, center=<-19.6197 -0.0005 6.3258>, rotation=<rot 90 0 0>}
 ]
 thigh_r [
  {shapeType=Capsule, length=24.0207, radius=10.5923, center=<25.6787 1.1662 0.0004>, rotation=<rot 87.6858 -180 -180>}
  {shapeType=Capsule, length=8.3459, radius=11.895, center=<38.8781 1.166 0.1298>, rotation=<rot 87.678 -180 -180>}
 ]
}
constraints [
 {
  bone1Name=calf_l
  bone2Name=thigh_l
  swing1LimitDegrees=45
  swing1Motion=Free
  swing2LimitDegrees=45
  swing2Motion=Free
  twistLimitDegrees=45
  twistMotion=Free
 }
 {bone1Name=dyn_pant_leg_back_r_01, bone2Name=calf_l, swing1LimitDegrees=35, swing2LimitDegrees=35, twistLimitDegrees=5}
 {bone1Name=dyn_pant_leg_back_r_02, bone2Name=dyn_pant_leg_back_r_01, swing1LimitDegrees=35, swing2LimitDegrees=35, twistLimitDegrees=5}
 {bone1Name=dyn_pant_leg_back_r_03, bone2Name=dyn_pant_leg_back_r_02, swing1LimitDegrees=35, swing2LimitDegrees=35, twistLimitDegrees=5}
 {bone1Name=dyn_pant_leg_back_l_01, bone2Name=calf_l, swing1LimitDegrees=35, swing2LimitDegrees=35, twistLimitDegrees=5}
 {bone1Name=dyn_pant_leg_back_l_02, bone2Name=dyn_pant_leg_back_l_01, swing1LimitDegrees=35, swing2LimitDegrees=35, twistLimitDegrees=5}
 {bone1Name=dyn_pant_leg_back_l_03, bone2Name=dyn_pant_leg_back_l_02, swing1LimitDegrees=35, swing2LimitDegrees=35, twistLimitDegrees=5}
 {bone1Name=dyn_pant_leg_left_01, bone2Name=calf_l, swing1LimitDegrees=35, swing2LimitDegrees=35, twistLimitDegrees=5}
 {bone1Name=dyn_pant_leg_left_02, bone2Name=dyn_pant_leg_left_01, swing1LimitDegrees=35, swing2LimitDegrees=35, twistLimitDegrees=5}
 {bone1Name=dyn_pant_leg_left_03, bone2Name=dyn_pant_leg_left_02, swing1LimitDegrees=35, swing2LimitDegrees=35, twistLimitDegrees=5}
 {bone1Name=dyn_pant_leg_right_01, bone2Name=calf_l, swing1LimitDegrees=35, swing2LimitDegrees=35, twistLimitDegrees=5}
 {bone1Name=dyn_pant_leg_right_02, bone2Name=dyn_pant_leg_right_01, swing1LimitDegrees=35, swing2LimitDegrees=35, twistLimitDegrees=5}
 {bone1Name=dyn_pant_leg_right_03, bone2Name=dyn_pant_leg_right_02, swing1LimitDegrees=35, swing2LimitDegrees=35, twistLimitDegrees=5}
 {bone1Name=dyn_pant_leg_front_01, bone2Name=calf_l, swing1LimitDegrees=35, swing2LimitDegrees=35, twistLimitDegrees=5}
 {bone1Name=dyn_pant_leg_front_02, bone2Name=dyn_pant_leg_front_01, swing1LimitDegrees=35, swing2LimitDegrees=35, twistLimitDegrees=5}
 {bone1Name=dyn_pant_leg_front_03, bone2Name=dyn_pant_leg_front_02, swing1LimitDegrees=35, swing2LimitDegrees=35, twistLimitDegrees=5}
 {
  bone1Name=thigh_r
  bone2Name=pelvis
  swing1LimitDegrees=45
  swing1Motion=Free
  swing2LimitDegrees=45
  swing2Motion=Free
  twistLimitDegrees=45
  twistMotion=Free
 }
 {
  bone1Name=calf_r
  bone2Name=thigh_r
  swing1LimitDegrees=45
  swing1Motion=Free
  swing2LimitDegrees=45
  swing2Motion=Free
  twistLimitDegrees=45
  twistMotion=Free
 }
 {
  bone1Name=dyn_leg_ribbon_ball
  bone2Name=calf_r
  swing1LimitDegrees=45
  swing1Motion=Free
  swing2LimitDegrees=45
  swing2Motion=Free
  twistLimitDegrees=45
  twistMotion=Free
 }
 {bone1Name=dyn_leg_ribbon_top_01, bone2Name=dyn_leg_ribbon_ball, swing1LimitDegrees=15, swing2LimitDegrees=15, twistLimitDegrees=5}
 {bone1Name=dyn_leg_ribbon_top_02, bone2Name=dyn_leg_ribbon_top_01, swing1LimitDegrees=15, swing2LimitDegrees=15, twistLimitDegrees=5}
 {bone1Name=dyn_leg_ribbon_top_03, bone2Name=dyn_leg_ribbon_top_02, swing1LimitDegrees=15, swing2LimitDegrees=15, twistLimitDegrees=5}
 {bone1Name=dyn_leg_ribbon_01, bone2Name=dyn_leg_ribbon_ball, swing1LimitDegrees=15, swing2LimitDegrees=15, twistLimitDegrees=5}
 {bone1Name=dyn_leg_ribbon_02, bone2Name=dyn_leg_ribbon_01, swing1LimitDegrees=15, swing2LimitDegrees=15, twistLimitDegrees=5}
 {bone1Name=dyn_leg_ribbon_03, bone2Name=dyn_leg_ribbon_02, swing1LimitDegrees=15, swing2LimitDegrees=15, twistLimitDegrees=5}
 {bone1Name=dyn_l_arm_ribbon_03, bone2Name=dyn_l_arm_ribbon_02, swing1LimitDegrees=45, swing2LimitDegrees=45, twistLimitDegrees=20}
 {bone1Name=dyn_l_arm_ribbon_01, bone2Name=dyn_l_arm_ribbon_master, swing1LimitDegrees=60, swing2LimitDegrees=60, twistLimitDegrees=10}
 {bone1Name=dyn_l_arm_ribbon_02, bone2Name=dyn_l_arm_ribbon_01, swing1LimitDegrees=45, swing2LimitDegrees=45, twistLimitDegrees=20}
 {
  bone1Name=dyn_l_arm_ribbon_master
  bone2Name=lowerarm_l
  swing1LimitDegrees=45
  swing1Motion=Free
  swing2LimitDegrees=45
  swing2Motion=Free
  twistLimitDegrees=45
  twistMotion=Free
 }
 {
  bone1Name=lowerarm_l
  bone2Name=spine_02
  swing1LimitDegrees=45
  swing1Motion=Free
  swing2LimitDegrees=45
  swing2Motion=Free
  twistLimitDegrees=45
  twistMotion=Free
 }
 {
  bone1Name=hand_l
  bone2Name=lowerarm_l
  swing1LimitDegrees=45
  swing1Motion=Free
  swing2LimitDegrees=45
  swing2Motion=Free
  twistLimitDegrees=45
  twistMotion=Free
 }
 {bone1Name=dyn_sash_front_08, bone2Name=dyn_sash_front_07, swing1LimitDegrees=20, swing2LimitDegrees=20, twistLimitDegrees=20}
 {bone1Name=dyn_sash_front_02, bone2Name=dyn_sash_front_01, swing1LimitDegrees=2, swing2LimitDegrees=2, twistLimitDegrees=2}
 {bone1Name=dyn_sash_front_03, bone2Name=dyn_sash_front_02, swing1LimitDegrees=10, swing2LimitDegrees=10, twistLimitDegrees=10}
 {bone1Name=dyn_sash_front_04, bone2Name=dyn_sash_front_03, swing1LimitDegrees=20, swing2LimitDegrees=20, twistLimitDegrees=20}
 {bone1Name=dyn_sash_front_05, bone2Name=dyn_sash_front_04, swing1LimitDegrees=20, swing2LimitDegrees=20, twistLimitDegrees=20}
 {bone1Name=dyn_sash_front_06, bone2Name=dyn_sash_front_05, swing1LimitDegrees=20, swing2LimitDegrees=20, twistLimitDegrees=20}
 {bone1Name=dyn_sash_front_07, bone2Name=dyn_sash_front_06, swing1LimitDegrees=20, swing2LimitDegrees=20, twistLimitDegrees=20}
 {bone1Name=dyn_sash_back_02, bone2Name=dyn_sash_back_01, swing1LimitDegrees=20, swing2LimitDegrees=20, twistLimitDegrees=20}
 {bone1Name=dyn_sash_back_03, bone2Name=dyn_sash_back_02, swing1LimitDegrees=20, swing2LimitDegrees=20, twistLimitDegrees=20}
 {bone1Name=dyn_sash_back_04, bone2Name=dyn_sash_back_03, swing1LimitDegrees=20, swing2LimitDegrees=20, twistLimitDegrees=20}
 {bone1Name=dyn_sash_back_05, bone2Name=dyn_sash_back_04, swing1LimitDegrees=20, swing2LimitDegrees=20, twistLimitDegrees=20}
 {
  bone1Name=thigh_l
  bone2Name=pelvis
  swing1LimitDegrees=45
  swing1Motion=Free
  swing2LimitDegrees=45
  swing2Motion=Free
  twistLimitDegrees=45
  twistMotion=Free
 }
 {bone1Name=dyn_scroll_02, bone2Name=pelvis, swing1LimitDegrees=10, swing2LimitDegrees=5, twistLimitDegrees=10}
 {bone1Name=dyn_scroll_01, bone2Name=pelvis, swing1LimitDegrees=5, swing2LimitDegrees=5, twistLimitDegrees=10}
 {
  bone1Name=dyn_scroll_02_02
  bone2Name=dyn_scroll_02
  swing1LimitDegrees=45
  swing1Motion=Free
  swing2LimitDegrees=45
  swing2Motion=Free
  twistLimitDegrees=45
  twistMotion=Free
 }
 {
  bone1Name=spine_02
  bone2Name=spine_01
  swing1LimitDegrees=45
  swing1Motion=Free
  swing2LimitDegrees=45
  swing2Motion=Free
  twistLimitDegrees=45
  twistMotion=Free
 }
 {
  bone1Name=dyn_crystal_master
  bone2Name=pelvis
  swing1LimitDegrees=45
  swing1Motion=Free
  swing2LimitDegrees=45
  swing2Motion=Free
  twistLimitDegrees=45
  twistMotion=Free
 }
 {bone1Name=dyn_crystal_01, bone2Name=dyn_crystal_master, swing1LimitDegrees=3, swing2LimitDegrees=3, twistLimitDegrees=3}
 {bone1Name=dyn_crystal_02, bone2Name=dyn_crystal_01, swing1LimitDegrees=20, swing2LimitDegrees=40, twistLimitDegrees=10}
 {bone1Name=dyn_crystal_03, bone2Name=dyn_crystal_02, swing1LimitDegrees=45, swing2LimitDegrees=45, twistLimitDegrees=20}
 {
  bone1Name=spine_01
  bone2Name=pelvis
  swing1LimitDegrees=45
  swing1Motion=Free
  swing2LimitDegrees=45
  swing2Motion=Free
  twistLimitDegrees=45
  twistMotion=Free
 }
 {
  bone1Name=pelvis
  bone2Name=root
  swing1LimitDegrees=45
  swing1Motion=Free
  swing2LimitDegrees=45
  swing2Motion=Free
  twistLimitDegrees=45
  twistMotion=Free
 }
]
