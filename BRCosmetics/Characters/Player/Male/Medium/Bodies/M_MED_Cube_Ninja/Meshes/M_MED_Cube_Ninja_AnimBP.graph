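# Animation Blueprint for the M_MED_Cube_Ninja cosmetic body, adding secondary motion to sashes, belts, mask and scrolls.
# Branches copy the attached parent mesh's pose, then simulate dangling bones with RigidBody or per-bone AnimDynamics nodes.
# RigidBody variants use separate physics assets for air, crouch, skydiving, emote and frontend, with movement booleans blending several branches.
# Constraints and additive bone offsets pin scrolls to the belt and correct poses.
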AnimBlueprint /BRCosmetics/Characters/Player/Male/Medium/Bodies/M_MED_Cube_Ninja/Meshes/M_MED_Cube_Ninja_AnimBP
target_skeleton=&/M_MED_Cube_Ninja_Skeleton
graphs [
 AnimGraph {nodes [
   n1 = "Constraint - dyn_sash_back_master (PTT)"!(
    ComponentPose=Local To Component(LocalPose=Copy Pose From Mesh() {bUseAttachedParent=true})
    Alpha=1
    ConstraintWeights_0=1
    ConstraintWeights_1:float=0
    ConstraintWeights_2:float=0
   ) -> (Pose) {
    alphaBoolBlend {blendOption=Linear}
    boneToModify {boneName=dyn_sash_back_master}
    constraintSetup [
     {targetBone {boneName=dyn_scrolls_master}}
     {transformType=Translation, targetBone {boneName=None}}
     {transformType=Translation, targetBone {boneName=None}}
    ]
    constraintWeights [1,1,1]
   }
   n2 = "Constraint - dyn_sash_front_master (PTT)"!(ComponentPose=@n1, Alpha=1, ConstraintWeights_0=1, ConstraintWeights_1:float=0, ConstraintWeights_2:float=0) -> (Pose) {
    alphaBoolBlend {blendOption=Linear}
    boneToModify {boneName=dyn_sash_front_master}
    constraintSetup [
     {targetBone {boneName=dyn_scrolls_master}}
     {transformType=Translation, targetBone {boneName=None}}
     {transformType=Translation, targetBone {boneName=None}}
    ]
    constraintWeights [1,1,1]
   }
   n3 = RigidBody(ComponentPose=@n2) {
    bClampLinearTranslationLimitToRefPose=true
    bOverrideWorldGravity=true
    overridePhysicsAsset=/Game/Characters/Player/Male/Medium/Bodies/M_MED_Cube_Ninja/Meshes/M_MED_Cube_Ninja_Air
    simulationSpace=BaseBoneSpace
    alphaBoolBlend {blendOption=Linear}
    baseBoneRef {boneName=root}
    componentAppliedLinearAccClamp=<1200 1200 1200>
    componentLinearAccScale=<0.35 0.35 0.45>
    overrideWorldGravity {z=-980}
   }
   n4 = "Constraint - dyn_scrolls_master (TTT)"!(ComponentPose=@n3, Alpha=1, ConstraintWeights_0=1, ConstraintWeights_1:float=0, ConstraintWeights_2:float=0) -> (Pose) {
    alphaBoolBlend {blendOption=Linear}
    boneToModify {boneName=dyn_scrolls_master}
    constraintSetup [
     {transformType=Translation, targetBone {boneName=dyn_r_belt_03}}
     {transformType=Translation, targetBone {boneName=None}}
     {transformType=Translation, targetBone {boneName=None}}
    ]
    constraintWeights [1,1,1]
   }
   n5 = "Transform (Modify) Bone - Bone: dyn_scrolls_master"(ComponentPose=@n4, Alpha=1, Translation="0,0,3", Rotation="0,0,0", Scale="1,1,1") {
    rotationMode=BMM_Additive
    scaleMode=BMM_Additive
    translationMode=BMM_Additive
    alphaBoolBlend {blendOption=Linear}
    boneToModify {boneName=dyn_scrolls_master}
    translation {z=3}
   }
   n6 = "Transform (Modify) Bone - Bone: dyn_knee_adjustment_r"(ComponentPose=@n5, Alpha=1, Translation="5,0,0", Rotation="0,0,0", Scale="1,1,1") {
    rotationMode=BMM_Additive
    scaleMode=BMM_Additive
    translationMode=BMM_Additive
    alphaBoolBlend {blendOption=Linear}
    boneToModify {boneName=dyn_knee_adjustment_r}
    translation {x=5}
   }
   n7 = Component To Local(ComponentPose=@n6)
   n8 = OR Boolean(A=Get bIsJumping(), B=Get bIsFalling())
   n9 = RigidBody(ComponentPose=Local To Component(LocalPose=Copy Pose From Mesh() {bUseAttachedParent=true})) {
    bClampLinearTranslationLimitToRefPose=true
    bOverrideWorldGravity=true
    overridePhysicsAsset=/Game/Characters/Player/Male/Medium/Bodies/M_MED_Cube_Ninja/Meshes/M_MED_Cube_Ninja_Physics_Crouch
    simulationSpace=BaseBoneSpace
    alphaBoolBlend {blendOption=Linear}
    baseBoneRef {boneName=root}
    componentAppliedLinearAccClamp=<1200 1200 1200>
    componentLinearAccScale=<0.4 0.4 0.5>
    overrideWorldGravity {z=-980}
   }
   n10 = "Constraint - dyn_scrolls_master (TTT)"(ComponentPose=@n9, Alpha=1, ConstraintWeights_0=1, ConstraintWeights_1=0, ConstraintWeights_2=0) {
    alphaBoolBlend {blendOption=Linear}
    boneToModify {boneName=dyn_scrolls_master}
    constraintSetup [
     {transformType=Translation, targetBone {boneName=dyn_r_belt_03}}
     {transformType=Translation, targetBone {boneName=None}}
     {transformType=Translation, targetBone {boneName=None}}
    ]
    constraintWeights [1,1,1]
   }
   n11 = "Transform (Modify) Bone - Bone: dyn_mask_master"(ComponentPose=@n10, Alpha=1, Translation="0,0,0", Rotation="0,-10,25", Scale="1,1,1") {
    rotationMode=BMM_Additive
    scaleMode=BMM_Additive
    translationMode=BMM_Additive
    alphaBoolBlend {blendOption=Linear}
    boneToModify {boneName=dyn_mask_master}
    rotation {roll=25, yaw=-10}
   }
   n12 = "Transform (Modify) Bone - Bone: dyn_scroll_02"(ComponentPose=@n11, Alpha=Get Thigh_Rotate() -> (Thigh_Rotate:double), Translation="0,0,0", Rotation="0,-60,0", Scale="1,1,1") {
    rotationMode=BMM_Additive
    scaleMode=BMM_Additive
    translationMode=BMM_Additive
    alphaBoolBlend {blendOption=Linear}
    boneToModify {boneName=dyn_scroll_02}
    rotation {yaw=-60}
   }
   n13 = RigidBody!(
    ComponentPose=Local To Component(LocalPose=Copy Pose From Mesh() {bUseAttachedParent=true})
    Alpha=0.8
    OverrideWorldGravity:Vector="0,0,-1100"
    ExternalForce
   ) -> (Pose) {
    bClampLinearTranslationLimitToRefPose=true
    bOverrideWorldGravity=true
    overridePhysicsAsset=/Game/Characters/Player/Male/Medium/Bodies/M_MED_Cube_Ninja/Meshes/M_MED_Cube_Ninja_Skydiving
    simulationSpace=BaseBoneSpace
    alphaBoolBlend {blendOption=Linear}
    baseBoneRef {boneName=root}
    componentAppliedLinearAccClamp=<5000 5000 5000>
    componentLinearAccScale=<0.4 0.4 0.7>
    componentLinearVelScale=<0.4 0.4 0.7>
    overrideWorldGravity {z=-1100}
   }
   n14 = "Constraint - dyn_scrolls_master (TTT)"(ComponentPose=@n13, Alpha=1, ConstraintWeights_0=1, ConstraintWeights_1=0, ConstraintWeights_2=0) {
    alphaBoolBlend {blendOption=Linear}
    boneToModify {boneName=dyn_scrolls_master}
    constraintSetup [
     {transformType=Translation, targetBone {boneName=dyn_r_belt_03}}
     {transformType=Translation, targetBone {boneName=None}}
     {transformType=Translation, targetBone {boneName=None}}
    ]
    constraintWeights [1,1,1]
   }
   n15 = "Transform (Modify) Bone - Bone: dyn_scrolls_master"(ComponentPose=@n14, Alpha=1, Translation="0,0,3", Rotation="0,0,0", Scale="1,1,1") {
    rotationMode=BMM_Additive
    scaleMode=BMM_Additive
    translationMode=BMM_Additive
    alphaBoolBlend {blendOption=Linear}
    boneToModify {boneName=dyn_scrolls_master}
    translation {z=3}
   }
   Blend Poses by bool(
    BlendPose_0=Component To Local(ComponentPose=@n15)
    BlendPose_1=Blend Poses by bool(
     BlendPose_0=Component To Local(ComponentPose=@n12)
     BlendPose_1=Blend Poses by bool(BlendPose_0=@n7, BlendTime_0=0.4, BlendTime_1=0.3, bActiveValue=@n8) {blendType=Linear}
     BlendTime_0=0.2
     BlendTime_1=1.3
     bActiveValue=Get bIsCrouching()
    ) {blendType=Linear}
    bActiveValue=Get bIsSkyDiving()
   ) {blendType=Linear}
   n16 = "Constraint - dyn_sash_front_master (PTT)"(
    ComponentPose=Local To Component(LocalPose=Copy Pose From Mesh() {bUseAttachedParent=true})
    Alpha=1
    ConstraintWeights_0=1
    ConstraintWeights_1=0
    ConstraintWeights_2=0
   ) {
    alphaBoolBlend {blendOption=Linear}
    boneToModify {boneName=dyn_sash_front_master}
    constraintSetup [
     {targetBone {boneName=dyn_scrolls_master}}
     {transformType=Translation, targetBone {boneName=None}}
     {transformType=Translation, targetBone {boneName=None}}
    ]
    constraintWeights [1,1,1]
   }
   n17 = "Constraint - dyn_sash_back_master (PTT)"(ComponentPose=@n16, Alpha=1, ConstraintWeights_0=1, ConstraintWeights_1=0, ConstraintWeights_2=0) {
    alphaBoolBlend {blendOption=Linear}
    boneToModify {boneName=dyn_sash_back_master}
    constraintSetup [
     {targetBone {boneName=dyn_scrolls_master}}
     {transformType=Translation, targetBone {boneName=None}}
     {transformType=Translation, targetBone {boneName=None}}
    ]
    constraintWeights [1,1,1]
   }
   n18 = "Transform (Modify) Bone - Bone: dyn_sash_back_01_05"(ComponentPose=@n17, Alpha=1, Translation="0,3,-5", Rotation="0,0,0", Scale="1,1,1") {
    rotationMode=BMM_Additive
    rotationSpace=BCS_WorldSpace
    translationMode=BMM_Additive
    alphaBoolBlend {blendOption=Linear}
    boneToModify {boneName=dyn_sash_back_01_05}
    translation {y=3, z=-5}
   }
   n19 = "Transform (Modify) Bone - Bone: dyn_sash_front_01"(ComponentPose=@n18, Alpha=0.5, Translation="0,0,0", Rotation=Get TesselRotator_01() -> (TesselRotator_01:Rotator), Scale="1,1,1") {
    rotationMode=BMM_Additive
    rotationSpace=BCS_BoneSpace
    translationSpace=BCS_BoneSpace
    alphaBoolBlend {blendOption=Linear}
    alphaScaleBiasClamp {interpSpeedDecreasing=2, interpSpeedIncreasing=2}
    boneToModify {boneName=dyn_sash_front_01}
   }
   n20 = "Transform (Modify) Bone - Bone: dyn_sash_back_01"(ComponentPose=@n19, Alpha=1, Translation="0,0,0", Rotation=Get TesselRotator_02() -> (TesselRotator_02:Rotator), Scale="1,1,1") {
    rotationMode=BMM_Additive
    rotationSpace=BCS_BoneSpace
    translationSpace=BCS_BoneSpace
    alphaBoolBlend {blendOption=Linear}
    alphaScaleBiasClamp {bInterpResult=true, interpSpeedDecreasing=2, interpSpeedIncreasing=2}
    boneToModify {boneName=dyn_sash_back_01}
   }
   n21 = "Transform (Modify) Bone - Bone: dyn_l_arm_ribbon_master"(ComponentPose=@n20, Alpha=1, Translation="0,2,2", Rotation="0,0,0", Scale="1,1,1") {
    rotationMode=BMM_Additive
    rotationSpace=BCS_WorldSpace
    translationMode=BMM_Additive
    translationSpace=BCS_ParentBoneSpace
    alphaBoolBlend {blendOption=Linear}
    boneToModify {boneName=dyn_l_arm_ribbon_master}
    translation {y=2, z=2}
   }
   n22 = RigidBody(ComponentPose=@n21, OverrideWorldGravity=Get GravityOverride_tessel() -> (GravityOverride_tessel:Vector)) {
    bClampLinearTranslationLimitToRefPose=true
    bOverrideWorldGravity=true
    overridePhysicsAsset=/Game/Characters/Player/Male/Medium/Bodies/M_MED_Cube_Ninja/Meshes/M_MED_Cube_Ninja_Emote_Physics
    simulationSpace=BaseBoneSpace
    alphaBoolBlend {blendOption=Linear}
    baseBoneRef {boneName=root}
    componentAppliedLinearAccClamp=<5000 5000 5000>
    componentLinearAccScale=<0.8 0.8 0.9>
    componentLinearVelScale=<0.8 0.8 0.9>
    overrideWorldGravity {z=-1100}
   }
   n23 = "Transform (Modify) Bone - Bone: dyn_sash_front_01"(ComponentPose=@n22, Alpha=1, Translation="2.5,0,0", Rotation="0,0,0", Scale="1,1,1") {
    rotationMode=BMM_Additive
    rotationSpace=BCS_WorldSpace
    translationMode=BMM_Additive
    translationSpace=BCS_ParentBoneSpace
    alphaBoolBlend {blendOption=Linear}
    boneToModify {boneName=dyn_sash_front_01}
    translation {x=2.5}
   }
   n24 = "Transform (Modify) Bone - Bone: dyn_mask_01"(ComponentPose=@n23, Alpha=1, Translation="1.25,0,0", Rotation="0,0,0", Scale="1,1,1") {
    rotationMode=BMM_Additive
    rotationSpace=BCS_WorldSpace
    translationMode=BMM_Additive
    translationSpace=BCS_ParentBoneSpace
    alphaBoolBlend {blendOption=Linear}
    boneToModify {boneName=dyn_mask_01}
    translation {x=1.25}
   }
   "Constraint - dyn_scrolls_master (PTT)"!(ComponentPose=@n24, Alpha=1, ConstraintWeights_0=1, ConstraintWeights_1:float=0, ConstraintWeights_2:float=0) -> (Pose) {
    alphaBoolBlend {blendOption=Linear}
    boneToModify {boneName=dyn_scrolls_master}
    constraintSetup [
     {targetBone {boneName=dyn_r_belt_03}}
     {transformType=Translation, targetBone {boneName=None}}
     {transformType=Translation, targetBone {boneName=None}}
    ]
    constraintWeights [1,1,1]
   }
   n25 = Anim Dynamics - Bone: dyn_l_belt_01(Alpha=1, ExternalForce="0,0,0") {
    angularDampingOverride=0.7
    angularSpringConstant=1000
    bGravityOverrideInSimSpace=true
    bUseGravityOverride=true
    alphaBoolBlend {blendOption=Linear}
    boundBone {boneName=dyn_l_belt_01}
    componentAppliedLinearAccClamp=<1500 1500 1500>
    componentLinearAccScale=<0.4 0.4 0.4>
    componentLinearVelScale=<0.4 0.4 0.4>
    gravityOverride {z=-980}
    physicsBodyDefinitions [{
      sphereCollisionRadius=0
      boundBone {boneName=dyn_l_belt_01}
      boxExtents=<20 20 20>
      constraintSetup {angularLimitsMax=<3 -5 3>, angularLimitsMin=<-3 -12 -3>, angularTarget=<1 0 0>}
      localJointOffset=<-3 0 0>
     }]
    relativeSpaceBone {boneName=root}
   }
   n26 = Anim Dynamics - Bone: dyn_l_belt_02(ComponentPose=@n25, Alpha=1, ExternalForce="0,0,0") {
    angularDampingOverride=0.7
    angularSpringConstant=1000
    bGravityOverrideInSimSpace=true
    bUseGravityOverride=true
    alphaBoolBlend {blendOption=Linear}
    boundBone {boneName=dyn_l_belt_02}
    componentAppliedLinearAccClamp=<1500 1500 1500>
    componentLinearAccScale=<0.4 0.4 0.4>
    componentLinearVelScale=<0.4 0.4 0.4>
    gravityOverride {z=-980}
    physicsBodyDefinitions [{
      sphereCollisionRadius=0
      boundBone {boneName=dyn_l_belt_02}
      boxExtents=<20 20 20>
      constraintSetup {angularLimitsMax=<3 -5 3>, angularLimitsMin=<-3 -12 -3>, angularTarget=<1 0 0>}
      localJointOffset=<-3 0 0>
     }]
    relativeSpaceBone {boneName=root}
   }
   n27 = Anim Dynamics - Bone: dyn_l_belt_03(ComponentPose=@n26, Alpha=1, ExternalForce="0,0,0") {
    angularDampingOverride=0.7
    angularSpringConstant=1000
    bGravityOverrideInSimSpace=true
    bUseGravityOverride=true
    alphaBoolBlend {blendOption=Linear}
    boundBone {boneName=dyn_l_belt_03}
    componentAppliedLinearAccClamp=<1500 1500 1500>
    componentLinearAccScale=<0.4 0.4 0.4>
    componentLinearVelScale=<0.4 0.4 0.4>
    gravityOverride {z=-980}
    physicsBodyDefinitions [{
      sphereCollisionRadius=0
      boundBone {boneName=dyn_l_belt_03}
      boxExtents=<20 20 20>
      constraintSetup {angularLimitsMax=<3 -5 3>, angularLimitsMin=<-3 -12 -3>, angularTarget=<1 0 0>}
      localJointOffset=<-3 0 0>
     }]
    relativeSpaceBone {boneName=root}
   }
   n28 = Anim Dynamics - Bone: dyn_mask_01(ComponentPose=@n27, Alpha=1, ExternalForce="0,0,0") {
    angularDampingOverride=0.7
    angularSpringConstant=1000
    bGravityOverrideInSimSpace=true
    bUseGravityOverride=true
    alphaBoolBlend {blendOption=Linear}
    boundBone {boneName=dyn_mask_01}
    componentAppliedLinearAccClamp=<1500 1500 1500>
    componentLinearAccScale=<0.4 0.4 0.4>
    componentLinearVelScale=<0.4 0.4 0.4>
    gravityOverride {z=-980}
    physicsBodyDefinitions [{
      sphereCollisionRadius=0
      boundBone {boneName=dyn_mask_01}
      boxExtents=<20 20 20>
      constraintSetup {angularLimitsMax=<5 5 5>, angularLimitsMin=<-5 -5 -5>, angularTarget=<1 0 0>}
      localJointOffset=<5 0 0>
     }]
    relativeSpaceBone {boneName=root}
   }
   n29 = Anim Dynamics - Bone: dyn_r_belt_01(ComponentPose=@n28, Alpha=1, ExternalForce="0,0,0") {
    angularDampingOverride=0.7
    angularSpringConstant=1000
    bGravityOverrideInSimSpace=true
    bUseGravityOverride=true
    alphaBoolBlend {blendOption=Linear}
    boundBone {boneName=dyn_r_belt_01}
    componentAppliedLinearAccClamp=<1500 1500 1500>
    componentLinearAccScale=<0.4 0.4 0.4>
    componentLinearVelScale=<0.4 0.4 0.4>
    gravityOverride {z=-980}
    physicsBodyDefinitions [{
      sphereCollisionRadius=0
      boundBone {boneName=dyn_r_belt_01}
      boxExtents=<20 20 20>
      constraintSetup {angularLimitsMax=<3 0 3>, angularLimitsMin=<-3 -12 -3>, angularTarget=<1 0 0>}
      localJointOffset=<-3 0 0>
     }]
    relativeSpaceBone {boneName=root}
   }
   n30 = Anim Dynamics - Bone: dyn_r_belt_02(ComponentPose=@n29, Alpha=1, ExternalForce="0,0,0") {
    angularDampingOverride=0.7
    angularSpringConstant=1000
    bGravityOverrideInSimSpace=true
    bUseGravityOverride=true
    alphaBoolBlend {blendOption=Linear}
    boundBone {boneName=dyn_r_belt_02}
    componentAppliedLinearAccClamp=<1500 1500 1500>
    componentLinearAccScale=<0.4 0.4 0.4>
    componentLinearVelScale=<0.4 0.4 0.4>
    gravityOverride {z=-980}
    physicsBodyDefinitions [{
      sphereCollisionRadius=0
      boundBone {boneName=dyn_r_belt_02}
      boxExtents=<20 20 20>
      constraintSetup {angularLimitsMax=<3 0 3>, angularLimitsMin=<-3 -12 -3>, angularTarget=<1 0 0>}
      localJointOffset=<-3 0 0>
     }]
    relativeSpaceBone {boneName=root}
   }
   n31 = Anim Dynamics - Bone: dyn_r_belt_03(ComponentPose=@n30, Alpha=1, ExternalForce="0,0,0") {
    angularDampingOverride=0.7
    angularSpringConstant=1000
    bGravityOverrideInSimSpace=true
    bUseGravityOverride=true
    alphaBoolBlend {blendOption=Linear}
    boundBone {boneName=dyn_r_belt_03}
    componentAppliedLinearAccClamp=<1500 1500 1500>
    componentLinearAccScale=<0.4 0.4 0.4>
    componentLinearVelScale=<0.4 0.4 0.4>
    gravityOverride {z=-980}
    physicsBodyDefinitions [{
      sphereCollisionRadius=0
      boundBone {boneName=dyn_r_belt_03}
      boxExtents=<20 20 20>
      constraintSetup {angularLimitsMax=<3 0 3>, angularLimitsMin=<-3 -12 -3>, angularTarget=<1 0 0>}
      localJointOffset=<-3 0 0>
     }]
    relativeSpaceBone {boneName=root}
   }
   "Constraint - dyn_scrolls_master (TTT)"(ComponentPose=@n31, Alpha=1, ConstraintWeights_0=1, ConstraintWeights_1=0, ConstraintWeights_2=0) {
    alphaBoolBlend {blendOption=Linear}
    boneToModify {boneName=dyn_scrolls_master}
    constraintSetup [
     {transformType=Translation, targetBone {boneName=dyn_r_belt_03}}
     {transformType=Translation, targetBone {boneName=None}}
     {transformType=Translation, targetBone {boneName=None}}
    ]
    constraintWeights [1,1,1]
   }
   "Constraint - dyn_mask_master (PPP)"!(ComponentPose, Alpha=1, ConstraintWeights_0=0.3479, ConstraintWeights_1:float=0, ConstraintWeights_2:float=0.6521) -> (Pose) {
    alphaBoolBlend {blendOption=Linear}
    boneToModify {boneName=dyn_mask_master}
    constraintSetup [{targetBone {boneName=dyn_belt_front_l_04}}, {targetBone {boneName=dyn_belt_back_r_05}}, {targetBone {boneName=dyn_belt_back_l_04}}]
    constraintWeights [1,1,1]
   }
   "Constraint - dyn_mask_master (PPP)"(Alpha=1, ConstraintWeights_0=0.3479, ConstraintWeights_1=0, ConstraintWeights_2=0.6521) {
    alphaBoolBlend {blendOption=Linear}
    boneToModify {boneName=dyn_mask_master}
    constraintSetup [{targetBone {boneName=dyn_belt_front_l_04}}, {targetBone {boneName=dyn_belt_back_r_05}}, {targetBone {boneName=dyn_belt_back_l_04}}]
    constraintWeights [1,1,1]
   }
   "Constraint - dyn_mask_master (PPP)"(Alpha=1, ConstraintWeights_0=0.3479, ConstraintWeights_1=0, ConstraintWeights_2=0.6521) {
    alphaBoolBlend {blendOption=Linear}
    boneToModify {boneName=dyn_mask_master}
    constraintSetup [{targetBone {boneName=dyn_belt_front_l_04}}, {targetBone {boneName=dyn_belt_back_r_05}}, {targetBone {boneName=dyn_belt_back_l_04}}]
    constraintWeights [1,1,1]
   }
   "Constraint - dyn_scrolls_master (PTT)"(Alpha=1, ConstraintWeights_0=1, ConstraintWeights_1=0, ConstraintWeights_2=0) {
    alphaBoolBlend {blendOption=Linear}
    boneToModify {boneName=dyn_scrolls_master}
    constraintSetup [
     {targetBone {boneName=dyn_r_belt_03}}
     {transformType=Translation, targetBone {boneName=None}}
     {transformType=Translation, targetBone {boneName=None}}
    ]
    constraintWeights [1,1,1]
   }
   n32 = "Constraint - dyn_belt_front_l_04 (P)"(Alpha=0.1, ConstraintWeights_0=1) {
    alphaBoolBlend {blendOption=Linear}
    boneToModify {boneName=dyn_belt_front_l_04}
    constraintSetup [{targetBone {boneName=thigh_l}}]
    constraintWeights [1]
   }
   n33 = "Constraint - dyn_belt_back_l_04 (P)"(ComponentPose=@n32, Alpha=0.1, ConstraintWeights_0=1) {
    alphaBoolBlend {blendOption=Linear}
    boneToModify {boneName=dyn_belt_back_l_04}
    constraintSetup [{targetBone {boneName=thigh_l}}]
    constraintWeights [1]
   }
   n34 = "Constraint - dyn_belt_front_l_03 (P)"(ComponentPose=@n33, Alpha=0.1, ConstraintWeights_0=1) {
    alphaBoolBlend {blendOption=Linear}
    boneToModify {boneName=dyn_belt_front_l_03}
    constraintSetup [{targetBone {boneName=thigh_l}}]
    constraintWeights [1]
   }
   n35 = "Constraint - dyn_belt_back_l_03 (P)"(ComponentPose=@n34, Alpha=0.1, ConstraintWeights_0=1) {
    alphaBoolBlend {blendOption=Linear}
    boneToModify {boneName=dyn_belt_back_l_03}
    constraintSetup [{targetBone {boneName=thigh_l}}]
    constraintWeights [1]
   }
   n36 = "Constraint - dyn_belt_front_r_04 (P)"(ComponentPose=@n35, Alpha=0.1, ConstraintWeights_0=1) {
    alphaBoolBlend {blendOption=Linear}
    boneToModify {boneName=dyn_belt_front_r_04}
    constraintSetup [{targetBone {boneName=thigh_r}}]
    constraintWeights [1]
   }
   n37 = "Constraint - dyn_belt_back_r_04 (P)"(ComponentPose=@n36, Alpha=0.1, ConstraintWeights_0=1) {
    alphaBoolBlend {blendOption=Linear}
    boneToModify {boneName=dyn_belt_back_r_04}
    constraintSetup [{targetBone {boneName=thigh_r}}]
    constraintWeights [1]
   }
   n38 = "Constraint - dyn_belt_front_r_03 (P)"(ComponentPose=@n37, Alpha=0.1, ConstraintWeights_0=1) {
    alphaBoolBlend {blendOption=Linear}
    boneToModify {boneName=dyn_belt_front_r_03}
    constraintSetup [{targetBone {boneName=thigh_r}}]
    constraintWeights [1]
   }
   "Constraint - dyn_belt_back_r_03 (P)"(ComponentPose=@n38, Alpha=0.1, ConstraintWeights_0=1) {
    alphaBoolBlend {blendOption=Linear}
    boneToModify {boneName=dyn_belt_back_r_03}
    constraintSetup [{targetBone {boneName=thigh_r}}]
    constraintWeights [1]
   }
   "Constraint - dyn_mask_master (PPP)"(Alpha=1, ConstraintWeights_0=0.3479, ConstraintWeights_1=0, ConstraintWeights_2=0.6521) {
    alphaBoolBlend {blendOption=Linear}
    boneToModify {boneName=dyn_mask_master}
    constraintSetup [{targetBone {boneName=dyn_belt_front_l_04}}, {targetBone {boneName=dyn_belt_back_r_05}}, {targetBone {boneName=dyn_belt_back_l_04}}]
    constraintWeights [1,1,1]
   }
   "Constraint - dyn_mask_master (PPP)"(Alpha=1, ConstraintWeights_0=0.3479, ConstraintWeights_1=0, ConstraintWeights_2=0.6521) {
    alphaBoolBlend {blendOption=Linear}
    boneToModify {boneName=dyn_mask_master}
    constraintSetup [{targetBone {boneName=dyn_belt_front_l_04}}, {targetBone {boneName=dyn_belt_back_r_05}}, {targetBone {boneName=dyn_belt_back_l_04}}]
    constraintWeights [1,1,1]
   }
   "Transform (Modify) Bone - Bone: dyn_sash_back_01_05"(Alpha=1, Translation="-5,3,0", Rotation="0,0,0", Scale="1,1,1") {
    rotationMode=BMM_Additive
    rotationSpace=BCS_WorldSpace
    translationMode=BMM_Additive
    alphaBoolBlend {blendOption=Linear}
    boneToModify {boneName=dyn_sash_back_01_05}
    translation=<-5 3>
   }
   n39 = "Constraint - dyn_sash_front_master (PTT)"(
    ComponentPose=Local To Component(LocalPose=Copy Pose From Mesh() {bUseAttachedParent=true})
    Alpha=1
    ConstraintWeights_0=1
    ConstraintWeights_1=0
    ConstraintWeights_2=0
   ) {
    alphaBoolBlend {blendOption=Linear}
    boneToModify {boneName=dyn_sash_front_master}
    constraintSetup [
     {targetBone {boneName=dyn_scrolls_master}}
     {transformType=Translation, targetBone {boneName=None}}
     {transformType=Translation, targetBone {boneName=None}}
    ]
    constraintWeights [1,1,1]
   }
   n40 = "Constraint - dyn_sash_back_master (PTT)"(ComponentPose=@n39, Alpha=1, ConstraintWeights_0=1, ConstraintWeights_1=0, ConstraintWeights_2=0) {
    alphaBoolBlend {blendOption=Linear}
    boneToModify {boneName=dyn_sash_back_master}
    constraintSetup [
     {targetBone {boneName=dyn_scrolls_master}}
     {transformType=Translation, targetBone {boneName=None}}
     {transformType=Translation, targetBone {boneName=None}}
    ]
    constraintWeights [1,1,1]
   }
   n41 = RigidBody(ComponentPose=@n40, OverrideWorldGravity="0,0,-1100") {
    bClampLinearTranslationLimitToRefPose=true
    bOverrideWorldGravity=true
    overridePhysicsAsset=&/M_MED_Cube_Ninja_Physics_Frontend
    simulationSpace=BaseBoneSpace
    alphaBoolBlend {blendOption=Linear}
    baseBoneRef {boneName=root}
    componentAppliedLinearAccClamp=<5000 5000 5000>
    componentLinearAccScale=<0.8 0.8 0.9>
    componentLinearVelScale=<0.8 0.8 0.9>
    overrideWorldGravity {z=-1100}
   }
   n42 = "Constraint - dyn_scrolls_master (TTT)"(ComponentPose=@n41, Alpha=1, ConstraintWeights_0=1, ConstraintWeights_1=0, ConstraintWeights_2=0) {
    alphaBoolBlend {blendOption=Linear}
    boneToModify {boneName=dyn_scrolls_master}
    constraintSetup [
     {transformType=Translation, targetBone {boneName=dyn_r_belt_03}}
     {transformType=Translation, targetBone {boneName=None}}
     {transformType=Translation, targetBone {boneName=None}}
    ]
    constraintWeights [1,1,1]
   }
   n43 = Component To Local(ComponentPose=@n42)
   n44 = "Constraint - dyn_sash_front_master (PTT)"(
    ComponentPose=Local To Component(LocalPose=Copy Pose From Mesh() {bUseAttachedParent=true})
    Alpha=1
    ConstraintWeights_0=1
    ConstraintWeights_1=0
    ConstraintWeights_2=0
   ) {
    alphaBoolBlend {blendOption=Linear}
    boneToModify {boneName=dyn_sash_front_master}
    constraintSetup [
     {targetBone {boneName=dyn_scrolls_master}}
     {transformType=Translation, targetBone {boneName=None}}
     {transformType=Translation, targetBone {boneName=None}}
    ]
    constraintWeights [1,1,1]
   }
   n45 = "Constraint - dyn_sash_back_master (PTT)"(ComponentPose=@n44, Alpha=1, ConstraintWeights_0=1, ConstraintWeights_1=0, ConstraintWeights_2=0) {
    alphaBoolBlend {blendOption=Linear}
    boneToModify {boneName=dyn_sash_back_master}
    constraintSetup [
     {targetBone {boneName=dyn_scrolls_master}}
     {transformType=Translation, targetBone {boneName=None}}
     {transformType=Translation, targetBone {boneName=None}}
    ]
    constraintWeights [1,1,1]
   }
   n46 = Anim Dynamics - Bone: dyn_l_belt_01(ComponentPose=@n45, Alpha=1, ExternalForce="0,0,0") {
    angularDampingOverride=0.7
    angularSpringConstant=1000
    bGravityOverrideInSimSpace=true
    bUseGravityOverride=true
    simulationSpace=BoneRelative
    alphaBoolBlend {blendOption=Linear}
    boundBone {boneName=dyn_l_belt_01}
    componentAppliedLinearAccClamp=<1500 1500 1500>
    componentLinearAccScale=<0.4 0.4 0.4>
    componentLinearVelScale=<0.4 0.4 0.4>
    gravityOverride {z=-980}
    physicsBodyDefinitions [{
      sphereCollisionRadius=0
      boundBone {boneName=dyn_l_belt_01}
      boxExtents=<20 20 20>
      constraintSetup {angularLimitsMax=<3 -5 3>, angularLimitsMin=<-3 -12 -3>, angularTarget=<1 0 0>}
      localJointOffset=<-3 0 0>
     }]
    relativeSpaceBone {boneName=root}
   }
   n47 = Anim Dynamics - Bone: dyn_l_belt_02(ComponentPose=@n46, Alpha=1, ExternalForce="0,0,0") {
    angularDampingOverride=0.7
    angularSpringConstant=1000
    bGravityOverrideInSimSpace=true
    bUseGravityOverride=true
    simulationSpace=BoneRelative
    alphaBoolBlend {blendOption=Linear}
    boundBone {boneName=dyn_l_belt_02}
    componentAppliedLinearAccClamp=<1500 1500 1500>
    componentLinearAccScale=<0.4 0.4 0.4>
    componentLinearVelScale=<0.4 0.4 0.4>
    gravityOverride {z=-980}
    physicsBodyDefinitions [{
      sphereCollisionRadius=0
      boundBone {boneName=dyn_l_belt_02}
      boxExtents=<20 20 20>
      constraintSetup {angularLimitsMax=<3 -5 3>, angularLimitsMin=<-3 -12 -3>, angularTarget=<1 0 0>}
      localJointOffset=<-3 0 0>
     }]
    relativeSpaceBone {boneName=root}
   }
   n48 = Anim Dynamics - Bone: dyn_l_belt_03(ComponentPose=@n47, Alpha=1, ExternalForce="0,0,0") {
    angularDampingOverride=0.7
    angularSpringConstant=1000
    bGravityOverrideInSimSpace=true
    bUseGravityOverride=true
    simulationSpace=BoneRelative
    alphaBoolBlend {blendOption=Linear}
    boundBone {boneName=dyn_l_belt_03}
    componentAppliedLinearAccClamp=<1500 1500 1500>
    componentLinearAccScale=<0.4 0.4 0.4>
    componentLinearVelScale=<0.4 0.4 0.4>
    gravityOverride {z=-980}
    physicsBodyDefinitions [{
      sphereCollisionRadius=0
      boundBone {boneName=dyn_l_belt_03}
      boxExtents=<20 20 20>
      constraintSetup {angularLimitsMax=<3 5 3>, angularLimitsMin=<-3 0 -3>, angularTarget=<1 0 0>}
      localJointOffset=<-3 0 0>
     }]
    relativeSpaceBone {boneName=root}
   }
   n49 = Anim Dynamics - Bone: dyn_mask_01(ComponentPose=@n48, Alpha=1, ExternalForce="0,0,0") {
    angularDampingOverride=0.7
    angularSpringConstant=1000
    bGravityOverrideInSimSpace=true
    bUseGravityOverride=true
    simulationSpace=BoneRelative
    alphaBoolBlend {blendOption=Linear}
    boundBone {boneName=dyn_mask_01}
    componentAppliedLinearAccClamp=<1500 1500 1500>
    componentLinearAccScale=<0.4 0.4 0.4>
    componentLinearVelScale=<0.4 0.4 0.4>
    gravityOverride {z=-980}
    physicsBodyDefinitions [{
      sphereCollisionRadius=0
      boundBone {boneName=dyn_mask_01}
      boxExtents=<20 20 20>
      constraintSetup {angularLimitsMax=<3 15 3>, angularLimitsMin=<-3 0 -3>, angularTarget=<1 0 0>}
      localJointOffset=<5 0 0>
     }]
    relativeSpaceBone {boneName=root}
   }
   n50 = Anim Dynamics - Bone: dyn_r_belt_01(ComponentPose=@n49, Alpha=1, ExternalForce="0,0,0") {
    angularDampingOverride=0.7
    angularSpringConstant=1000
    bGravityOverrideInSimSpace=true
    bUseGravityOverride=true
    simulationSpace=BoneRelative
    alphaBoolBlend {blendOption=Linear}
    boundBone {boneName=dyn_r_belt_01}
    componentAppliedLinearAccClamp=<1500 1500 1500>
    componentLinearAccScale=<0.4 0.4 0.4>
    componentLinearVelScale=<0.4 0.4 0.4>
    gravityOverride {z=-980}
    physicsBodyDefinitions [{
      sphereCollisionRadius=0
      boundBone {boneName=dyn_r_belt_01}
      boxExtents=<20 20 20>
      constraintSetup {angularLimitsMax=<3 -5 3>, angularLimitsMin=<-3 -12 -3>, angularTarget=<1 -5 0>}
      localJointOffset=<3 0 0>
     }]
    relativeSpaceBone {boneName=root}
   }
   n51 = Anim Dynamics - Bone: dyn_r_belt_02(ComponentPose=@n50, Alpha=1, ExternalForce="0,0,0") {
    angularDampingOverride=0.7
    angularSpringConstant=1000
    bGravityOverrideInSimSpace=true
    bUseGravityOverride=true
    simulationSpace=BoneRelative
    alphaBoolBlend {blendOption=Linear}
    boundBone {boneName=dyn_r_belt_02}
    componentAppliedLinearAccClamp=<1500 1500 1500>
    componentLinearAccScale=<0.4 0.4 0.4>
    componentLinearVelScale=<0.4 0.4 0.4>
    gravityOverride {z=-980}
    physicsBodyDefinitions [{
      sphereCollisionRadius=0
      boundBone {boneName=dyn_r_belt_02}
      boxExtents=<20 20 20>
      constraintSetup {angularLimitsMax=<3 -5 3>, angularLimitsMin=<-3 -12 -3>, angularTarget=<1 -5 0>}
      localJointOffset=<3 0 0>
     }]
    relativeSpaceBone {boneName=root}
   }
   n52 = Anim Dynamics - Bone: dyn_r_belt_03(ComponentPose=@n51, Alpha=1, ExternalForce="0,0,0") {
    angularDampingOverride=0.7
    angularSpringConstant=1000
    bGravityOverrideInSimSpace=true
    bUseGravityOverride=true
    simulationSpace=BoneRelative
    alphaBoolBlend {blendOption=Linear}
    boundBone {boneName=dyn_r_belt_03}
    componentAppliedLinearAccClamp=<1500 1500 1500>
    componentLinearAccScale=<0.4 0.4 0.4>
    componentLinearVelScale=<0.4 0.4 0.4>
    gravityOverride {z=-980}
    physicsBodyDefinitions [{
      sphereCollisionRadius=0
      boundBone {boneName=dyn_r_belt_03}
      boxExtents=<20 20 20>
      constraintSetup {angularLimitsMax=<3 -5 3>, angularLimitsMin=<-3 -12 -3>, angularTarget=<1 -5 0>}
      localJointOffset=<3 0 0>
     }]
    relativeSpaceBone {boneName=root}
   }
   n53 = "Transform (Modify) Bone - Bone: dyn_l_belt_01"(ComponentPose=@n52, Alpha=1, Translation="0,0,1.5", Rotation="5,0,0", Scale="1,1,1") {
    rotationMode=BMM_Additive
    translationMode=BMM_Additive
    alphaBoolBlend {blendOption=Linear}
    boneToModify {boneName=dyn_l_belt_01}
    rotation {pitch=5}
    translation {z=1.5}
   }
   n54 = "Transform (Modify) Bone - Bone: dyn_r_belt_01"(ComponentPose=@n53, Alpha=1, Translation="0,0,1.5", Rotation="-8,0,0", Scale="1,1,1") {
    rotationMode=BMM_Additive
    translationMode=BMM_Additive
    alphaBoolBlend {blendOption=Linear}
    boneToModify {boneName=dyn_r_belt_01}
    rotation {pitch=-8}
    translation {z=1.5}
   }
   n55 = "Constraint - dyn_scrolls_master (TTT)"(ComponentPose=@n54, Alpha=1, ConstraintWeights_0=1, ConstraintWeights_1=0, ConstraintWeights_2=0) {
    alphaBoolBlend {blendOption=Linear}
    boneToModify {boneName=dyn_scrolls_master}
    constraintSetup [
     {transformType=Translation, targetBone {boneName=dyn_r_belt_03}}
     {transformType=Translation, targetBone {boneName=None}}
     {transformType=Translation, targetBone {boneName=None}}
    ]
    constraintWeights [1,1,1]
   }
   n56 = Component To Local(ComponentPose=@n55)
   n57 = "Constraint - dyn_sash_front_master (PTT)"(
    ComponentPose=Local To Component(LocalPose=Copy Pose From Mesh() {bUseAttachedParent=true})
    Alpha=1
    ConstraintWeights_0=1
    ConstraintWeights_1=0
    ConstraintWeights_2=0
   ) {
    alphaBoolBlend {blendOption=Linear}
    boneToModify {boneName=dyn_sash_front_master}
    constraintSetup [
     {targetBone {boneName=dyn_scrolls_master}}
     {transformType=Translation, targetBone {boneName=None}}
     {transformType=Translation, targetBone {boneName=None}}
    ]
    constraintWeights [1,1,1]
   }
   n58 = "Constraint - dyn_sash_back_master (PTT)"(ComponentPose=@n57, Alpha=1, ConstraintWeights_0=1, ConstraintWeights_1=0, ConstraintWeights_2=0) {
    alphaBoolBlend {blendOption=Linear}
    boneToModify {boneName=dyn_sash_back_master}
    constraintSetup [
     {targetBone {boneName=dyn_scrolls_master}}
     {transformType=Translation, targetBone {boneName=None}}
     {transformType=Translation, targetBone {boneName=None}}
    ]
    constraintWeights [1,1,1]
   }
   n59 = Anim Dynamics - Bone: dyn_l_belt_01(ComponentPose=@n58, Alpha=1, ExternalForce="0,0,0") {
    angularDampingOverride=0.7
    angularSpringConstant=1000
    bGravityOverrideInSimSpace=true
    bUseGravityOverride=true
    simulationSpace=BoneRelative
    alphaBoolBlend {blendOption=Linear}
    boundBone {boneName=dyn_l_belt_01}
    componentAppliedLinearAccClamp=<1500 1500 1500>
    componentLinearAccScale=<0.4 0.4 0.4>
    componentLinearVelScale=<0.4 0.4 0.4>
    gravityOverride {z=-980}
    physicsBodyDefinitions [{
      sphereCollisionRadius=0
      boundBone {boneName=dyn_l_belt_01}
      boxExtents=<20 20 20>
      constraintSetup {angularLimitsMax=<3 -5 0>, angularLimitsMin=<-3 -12 0>, angularTarget=<1 0 0>}
      localJointOffset=<-3 0 0>
     }]
    relativeSpaceBone {boneName=root}
   }
   n60 = Anim Dynamics - Bone: dyn_l_belt_02(ComponentPose=@n59, Alpha=1, ExternalForce="0,0,0") {
    angularDampingOverride=0.7
    angularSpringConstant=1000
    bGravityOverrideInSimSpace=true
    bUseGravityOverride=true
    simulationSpace=BoneRelative
    alphaBoolBlend {blendOption=Linear}
    boundBone {boneName=dyn_l_belt_02}
    componentAppliedLinearAccClamp=<1500 1500 1500>
    componentLinearAccScale=<0.4 0.4 0.4>
    componentLinearVelScale=<0.4 0.4 0.4>
    gravityOverride {z=-980}
    physicsBodyDefinitions [{
      sphereCollisionRadius=0
      boundBone {boneName=dyn_l_belt_02}
      boxExtents=<20 20 20>
      constraintSetup {angularLimitsMax=<3 -5 0>, angularLimitsMin=<-3 -12 0>, angularTarget=<1 0 0>}
      localJointOffset=<-3 0 0>
     }]
    relativeSpaceBone {boneName=root}
   }
   n61 = Anim Dynamics - Bone: dyn_l_belt_03(ComponentPose=@n60, Alpha=1, ExternalForce="0,0,0") {
    angularDampingOverride=0.7
    angularSpringConstant=1000
    bGravityOverrideInSimSpace=true
    bUseGravityOverride=true
    simulationSpace=BoneRelative
    alphaBoolBlend {blendOption=Linear}
    boundBone {boneName=dyn_l_belt_03}
    componentAppliedLinearAccClamp=<1500 1500 1500>
    componentLinearAccScale=<0.4 0.4 0.4>
    componentLinearVelScale=<0.4 0.4 0.4>
    gravityOverride {z=-980}
    physicsBodyDefinitions [{
      sphereCollisionRadius=0
      boundBone {boneName=dyn_l_belt_03}
      boxExtents=<20 20 20>
      constraintSetup {angularLimitsMax=<3 5 3>, angularLimitsMin=<-3 0 -3>, angularTarget=<1 0 0>}
      localJointOffset=<-3 0 0>
     }]
    relativeSpaceBone {boneName=root}
   }
   n62 = Anim Dynamics - Bone: dyn_mask_01(ComponentPose=@n61, Alpha=1, ExternalForce="0,0,0") {
    angularDampingOverride=0.7
    angularSpringConstant=1000
    bGravityOverrideInSimSpace=true
    bUseGravityOverride=true
    simulationSpace=BoneRelative
    alphaBoolBlend {blendOption=Linear}
    boundBone {boneName=dyn_mask_01}
    componentAppliedLinearAccClamp=<1500 1500 1500>
    componentLinearAccScale=<0.4 0.4 0.4>
    componentLinearVelScale=<0.4 0.4 0.4>
    gravityOverride {z=-980}
    physicsBodyDefinitions [{
      sphereCollisionRadius=0
      boundBone {boneName=dyn_mask_01}
      boxExtents=<20 20 20>
      constraintSetup {angularLimitsMax=<3 15 0>, angularLimitsMin=<-3 0 0>, angularTarget=<1 0 0>}
      localJointOffset=<5 0 0>
     }]
    relativeSpaceBone {boneName=root}
   }
   n63 = Anim Dynamics - Bone: dyn_r_belt_01(ComponentPose=@n62, Alpha=1, ExternalForce="0,0,0") {
    angularDampingOverride=0.7
    angularSpringConstant=1000
    bGravityOverrideInSimSpace=true
    bUseGravityOverride=true
    simulationSpace=BoneRelative
    alphaBoolBlend {blendOption=Linear}
    boundBone {boneName=dyn_r_belt_01}
    componentAppliedLinearAccClamp=<1500 1500 1500>
    componentLinearAccScale=<0.4 0.4 0.4>
    componentLinearVelScale=<0.4 0.4 0.4>
    gravityOverride {z=-980}
    physicsBodyDefinitions [{
      sphereCollisionRadius=0
      boundBone {boneName=dyn_r_belt_01}
      boxExtents=<20 20 20>
      constraintSetup {angularLimitsMax=<3 -5 0>, angularLimitsMin=<-3 -12 0>, angularTarget=<1 -5 0>}
      localJointOffset=<3 0 0>
     }]
    relativeSpaceBone {boneName=root}
   }
   n64 = Anim Dynamics - Bone: dyn_r_belt_02(ComponentPose=@n63, Alpha=1, ExternalForce="0,0,0") {
    angularDampingOverride=0.7
    angularSpringConstant=1000
    bGravityOverrideInSimSpace=true
    bUseGravityOverride=true
    simulationSpace=BoneRelative
    alphaBoolBlend {blendOption=Linear}
    boundBone {boneName=dyn_r_belt_02}
    componentAppliedLinearAccClamp=<1500 1500 1500>
    componentLinearAccScale=<0.4 0.4 0.4>
    componentLinearVelScale=<0.4 0.4 0.4>
    gravityOverride {z=-980}
    physicsBodyDefinitions [{
      sphereCollisionRadius=0
      boundBone {boneName=dyn_r_belt_02}
      boxExtents=<20 20 20>
      constraintSetup {angularLimitsMax=<3 -5 0>, angularLimitsMin=<-3 -12 0>, angularTarget=<1 -5 0>}
      localJointOffset=<3 0 0>
     }]
    relativeSpaceBone {boneName=root}
   }
   n65 = Anim Dynamics - Bone: dyn_r_belt_03(ComponentPose=@n64, Alpha=1, ExternalForce="0,0,0") {
    angularDampingOverride=0.7
    angularSpringConstant=1000
    bGravityOverrideInSimSpace=true
    bUseGravityOverride=true
    simulationSpace=BoneRelative
    alphaBoolBlend {blendOption=Linear}
    boundBone {boneName=dyn_r_belt_03}
    componentAppliedLinearAccClamp=<1500 1500 1500>
    componentLinearAccScale=<0.4 0.4 0.4>
    componentLinearVelScale=<0.4 0.4 0.4>
    gravityOverride {z=-980}
    physicsBodyDefinitions [{
      sphereCollisionRadius=0
      boundBone {boneName=dyn_r_belt_03}
      boxExtents=<20 20 20>
      constraintSetup {angularLimitsMax=<3 -5 3>, angularLimitsMin=<-3 -12 -3>, angularTarget=<1 -5 0>}
      localJointOffset=<3 0 0>
     }]
    relativeSpaceBone {boneName=root}
   }
   n66 = "Constraint - dyn_scrolls_master (TTT)"(ComponentPose=@n65, Alpha=1, ConstraintWeights_0=1, ConstraintWeights_1=0, ConstraintWeights_2=0) {
    alphaBoolBlend {blendOption=Linear}
    boneToModify {boneName=dyn_scrolls_master}
    constraintSetup [
     {transformType=Translation, targetBone {boneName=dyn_r_belt_03}}
     {transformType=Translation, targetBone {boneName=None}}
     {transformType=Translation, targetBone {boneName=None}}
    ]
    constraintWeights [1,1,1]
   }
   n67 = Component To Local(ComponentPose=@n66)
   n68 = OR Boolean(A=Get bIsJumping(), B=Get bIsFalling())
   n69 = Anim Dynamics - Bone: dyn_leg_ribbon_top_01(
    ComponentPose=Local To Component(LocalPose=Blend Poses by bool(BlendPose_0=@n56, BlendPose_1=@n67, bActiveValue=@n68) {blendType=Linear})
    Alpha=1
    ExternalForce="0,0,0"
   ) {
    angularDampingOverride=0.7
    angularSpringConstant=1000
    bGravityOverrideInSimSpace=true
    bUseGravityOverride=true
    simulationSpace=BoneRelative
    alphaBoolBlend {blendOption=Linear}
    boundBone {boneName=dyn_leg_ribbon_top_01}
    componentAppliedLinearAccClamp=<1500 1500 1500>
    componentLinearAccScale=<0.4 0.4 0.4>
    componentLinearVelScale=<0.4 0.4 0.4>
    gravityOverride {z=-980}
    physicsBodyDefinitions [{
      sphereCollisionRadius=0
      boundBone {boneName=dyn_leg_ribbon_top_01}
      boxExtents=<15 15 15>
      constraintSetup {angularLimitsMax=<5 5 5>, angularLimitsMin=<-5 -5 -5>, angularTarget=<1 0 0>}
      localJointOffset=<3 0 0>
     }]
    relativeSpaceBone {boneName=root}
   }
   n70 = Anim Dynamics - Bone: dyn_leg_ribbon_top_02(ComponentPose=@n69, Alpha=1, ExternalForce="0,0,0") {
    angularDampingOverride=0.7
    angularSpringConstant=1000
    bGravityOverrideInSimSpace=true
    bUseGravityOverride=true
    simulationSpace=BoneRelative
    alphaBoolBlend {blendOption=Linear}
    boundBone {boneName=dyn_leg_ribbon_top_02}
    componentAppliedLinearAccClamp=<1500 1500 1500>
    componentLinearAccScale=<0.4 0.4 0.4>
    componentLinearVelScale=<0.4 0.4 0.4>
    gravityOverride {z=-980}
    physicsBodyDefinitions [{
      sphereCollisionRadius=0
      boundBone {boneName=dyn_leg_ribbon_top_02}
      boxExtents=<15 15 15>
      constraintSetup {angularLimitsMax=<5 5 5>, angularLimitsMin=<-5 -5 -5>, angularTarget=<1 0 0>}
      localJointOffset=<3 0 0>
     }]
    relativeSpaceBone {boneName=root}
   }
   n71 = Anim Dynamics - Bone: dyn_leg_ribbon_top_03(ComponentPose=@n70, Alpha=1, ExternalForce="0,0,0") {
    angularDampingOverride=0.7
    angularSpringConstant=1000
    bGravityOverrideInSimSpace=true
    bUseGravityOverride=true
    simulationSpace=BoneRelative
    alphaBoolBlend {blendOption=Linear}
    boundBone {boneName=dyn_leg_ribbon_top_03}
    componentAppliedLinearAccClamp=<1500 1500 1500>
    componentLinearAccScale=<0.4 0.4 0.4>
    componentLinearVelScale=<0.4 0.4 0.4>
    gravityOverride {z=-980}
    physicsBodyDefinitions [{
      sphereCollisionRadius=0
      boundBone {boneName=dyn_leg_ribbon_top_03}
      boxExtents=<15 15 15>
      constraintSetup {angularLimitsMax=<5 5 5>, angularLimitsMin=<-5 -5 -5>, angularTarget=<1 0 0>}
      localJointOffset=<3 0 0>
     }]
    relativeSpaceBone {boneName=root}
   }
   n72 = Anim Dynamics - Bone: dyn_leg_ribbon_01(ComponentPose=@n71, Alpha=1, ExternalForce="0,0,0") {
    angularDampingOverride=0.7
    angularSpringConstant=1000
    bGravityOverrideInSimSpace=true
    bUseGravityOverride=true
    simulationSpace=BoneRelative
    alphaBoolBlend {blendOption=Linear}
    boundBone {boneName=dyn_leg_ribbon_01}
    componentAppliedLinearAccClamp=<1500 1500 1500>
    componentLinearAccScale=<0.4 0.4 0.4>
    componentLinearVelScale=<0.4 0.4 0.4>
    gravityOverride {z=-980}
    physicsBodyDefinitions [{
      sphereCollisionRadius=0
      boundBone {boneName=dyn_leg_ribbon_01}
      boxExtents=<20 20 20>
      constraintSetup {angularLimitsMax=<5 5 5>, angularLimitsMin=<-5 -5 -5>, angularTarget=<1 0 0>}
      localJointOffset=<3 0 0>
     }]
    relativeSpaceBone {boneName=root}
   }
   n73 = Anim Dynamics - Bone: dyn_leg_ribbon_02(ComponentPose=@n72, Alpha=1, ExternalForce="0,0,0") {
    angularDampingOverride=0.7
    angularSpringConstant=1000
    bGravityOverrideInSimSpace=true
    bUseGravityOverride=true
    simulationSpace=BoneRelative
    alphaBoolBlend {blendOption=Linear}
    boundBone {boneName=dyn_leg_ribbon_02}
    componentAppliedLinearAccClamp=<1500 1500 1500>
    componentLinearAccScale=<0.4 0.4 0.4>
    componentLinearVelScale=<0.4 0.4 0.4>
    gravityOverride {z=-980}
    physicsBodyDefinitions [{
      sphereCollisionRadius=0
      boundBone {boneName=dyn_leg_ribbon_02}
      boxExtents=<20 20 20>
      constraintSetup {angularLimitsMax=<5 5 5>, angularLimitsMin=<-5 -5 -5>, angularTarget=<1 0 0>}
      localJointOffset=<3 0 0>
     }]
    relativeSpaceBone {boneName=root}
   }
   n74 = Anim Dynamics - Bone: dyn_leg_ribbon_03(ComponentPose=@n73, Alpha=1, ExternalForce="0,0,0") {
    angularDampingOverride=0.7
    angularSpringConstant=1000
    bGravityOverrideInSimSpace=true
    bUseGravityOverride=true
    simulationSpace=BoneRelative
    alphaBoolBlend {blendOption=Linear}
    boundBone {boneName=dyn_leg_ribbon_03}
    componentAppliedLinearAccClamp=<1500 1500 1500>
    componentLinearAccScale=<0.4 0.4 0.4>
    componentLinearVelScale=<0.4 0.4 0.4>
    gravityOverride {z=-980}
    physicsBodyDefinitions [{
      sphereCollisionRadius=0
      boundBone {boneName=dyn_leg_ribbon_03}
      boxExtents=<20 20 20>
      constraintSetup {angularLimitsMax=<5 5 5>, angularLimitsMin=<-5 -5 -5>, angularTarget=<1 0 0>}
      localJointOffset=<3 0 0>
     }]
    relativeSpaceBone {boneName=root}
   }
   n75 = Anim Dynamics - Bone: dyn_crystal_01(ComponentPose=@n74, Alpha=1, ExternalForce="0,0,0") {
    angularDampingOverride=0.9
    angularSpringConstant=1000
    bGravityOverrideInSimSpace=true
    bOverrideAngularDamping=true
    bUseGravityOverride=true
    simulationSpace=BoneRelative
    alphaBoolBlend {blendOption=Linear}
    boundBone {boneName=dyn_crystal_01}
    componentAppliedLinearAccClamp=<1500 1500 1500>
    componentLinearAccScale=<0.4 0.4 0.4>
    componentLinearVelScale=<0.4 0.4 0.4>
    gravityOverride {z=-980}
    physicsBodyDefinitions [{
      sphereCollisionRadius=0
      boundBone {boneName=dyn_crystal_01}
      boxExtents=<25 10 10>
      constraintSetup {angularLimitsMax=<5 5 5>, angularLimitsMin=<-5 -5 -5>, angularTarget=<1 0 0>}
      localJointOffset=<3 0 0>
     }]
    relativeSpaceBone {boneName=root}
   }
   n76 = Anim Dynamics - Bone: dyn_crystal_02(ComponentPose=@n75, Alpha=1, ExternalForce="0,0,0") {
    angularDampingOverride=0.9
    angularSpringConstant=1000
    bGravityOverrideInSimSpace=true
    bOverrideAngularDamping=true
    bUseGravityOverride=true
    simulationSpace=BoneRelative
    alphaBoolBlend {blendOption=Linear}
    boundBone {boneName=dyn_crystal_02}
    componentAppliedLinearAccClamp=<1500 1500 1500>
    componentLinearAccScale=<0.4 0.4 0.4>
    componentLinearVelScale=<0.4 0.4 0.4>
    gravityOverride {z=-980}
    physicsBodyDefinitions [{
      sphereCollisionRadius=0
      boundBone {boneName=dyn_crystal_02}
      boxExtents=<25 12 12>
      constraintSetup {angularLimitsMax=<2 10 10>, angularLimitsMin=<-2 -10 -10>, angularTarget=<1 0 0>}
      localJointOffset=<3 0 0>
     }]
    relativeSpaceBone {boneName=root}
   }
   n77 = Anim Dynamics - Bone: dyn_crystal_03(ComponentPose=@n76, Alpha=1, ExternalForce="0,0,0") {
    angularDampingOverride=0.9
    angularSpringConstant=1000
    bGravityOverrideInSimSpace=true
    bOverrideAngularDamping=true
    bUseGravityOverride=true
    simulationSpace=BoneRelative
    alphaBoolBlend {blendOption=Linear}
    boundBone {boneName=dyn_crystal_03}
    componentAppliedLinearAccClamp=<1500 1500 1500>
    componentLinearAccScale=<0.4 0.4 0.4>
    componentLinearVelScale=<0.4 0.4 0.4>
    gravityOverride {z=-980}
    physicsBodyDefinitions [{
      sphereCollisionRadius=0
      boundBone {boneName=dyn_crystal_03}
      boxExtents=<25 12 12>
      constraintSetup {angularLimitsMax=<40 5 40>, angularLimitsMin=<-40 -40 -40>, angularTarget=<1 0 0>}
      localJointOffset=<3 0 0>
     }]
    relativeSpaceBone {boneName=root}
   }
   n78 = Anim Dynamics - Bone: dyn_pant_leg_back_r_01(ComponentPose=@n77, Alpha=1, ExternalForce="0,0,0") {
    angularDampingOverride=0.9
    angularSpringConstant=1000
    bGravityOverrideInSimSpace=true
    bOverrideAngularDamping=true
    bUseGravityOverride=true
    simulationSpace=BoneRelative
    alphaBoolBlend {blendOption=Linear}
    boundBone {boneName=dyn_pant_leg_back_r_01}
    componentAppliedLinearAccClamp=<1500 1500 1500>
    componentLinearAccScale=<0.4 0.4 0.4>
    componentLinearVelScale=<0.4 0.4 0.4>
    gravityOverride {z=-980}
    physicsBodyDefinitions [{
      sphereCollisionRadius=0
      boundBone {boneName=dyn_pant_leg_back_r_01}
      boxExtents=<18 18 18>
      constraintSetup {angularLimitsMax=<10 5 10>, angularLimitsMin=<-10 -30 -10>, angularTarget=<1 0 0>}
      localJointOffset=<2 0 0>
     }]
    relativeSpaceBone {boneName=root}
   }
   n79 = Anim Dynamics - Bone: dyn_pant_leg_back_r_02(ComponentPose=@n78, Alpha=1, ExternalForce="0,0,0") {
    angularDampingOverride=0.9
    angularSpringConstant=1000
    bGravityOverrideInSimSpace=true
    bOverrideAngularDamping=true
    bUseGravityOverride=true
    simulationSpace=BoneRelative
    alphaBoolBlend {blendOption=Linear}
    boundBone {boneName=dyn_pant_leg_back_r_02}
    componentAppliedLinearAccClamp=<1500 1500 1500>
    componentLinearAccScale=<0.4 0.4 0.4>
    componentLinearVelScale=<0.4 0.4 0.4>
    gravityOverride {z=-980}
    physicsBodyDefinitions [{
      sphereCollisionRadius=0
      boundBone {boneName=dyn_pant_leg_back_r_02}
      boxExtents=<18 18 18>
      constraintSetup {angularLimitsMax=<10 5 10>, angularLimitsMin=<-10 -30 -10>, angularTarget=<1 0 0>}
      localJointOffset=<2 0 0>
     }]
    relativeSpaceBone {boneName=root}
   }
   n80 = Anim Dynamics - Bone: dyn_pant_leg_back_r_03(ComponentPose=@n79, Alpha=1, ExternalForce="0,0,0") {
    angularDampingOverride=0.9
    angularSpringConstant=1000
    bGravityOverrideInSimSpace=true
    bOverrideAngularDamping=true
    bUseGravityOverride=true
    simulationSpace=BoneRelative
    alphaBoolBlend {blendOption=Linear}
    boundBone {boneName=dyn_pant_leg_back_r_03}
    componentAppliedLinearAccClamp=<1500 1500 1500>
    componentLinearAccScale=<0.4 0.4 0.4>
    componentLinearVelScale=<0.4 0.4 0.4>
    gravityOverride {z=-980}
    physicsBodyDefinitions [{
      sphereCollisionRadius=0
      boundBone {boneName=dyn_pant_leg_back_r_03}
      boxExtents=<18 18 18>
      constraintSetup {angularLimitsMax=<10 5 10>, angularLimitsMin=<-10 -30 -10>, angularTarget=<1 0 0>}
      localJointOffset=<2 0 0>
     }]
    relativeSpaceBone {boneName=root}
   }
   n81 = Anim Dynamics - Bone: dyn_pant_leg_back_l_01(ComponentPose=@n80, Alpha=1, ExternalForce="0,0,0") {
    angularDampingOverride=0.9
    angularSpringConstant=1000
    bGravityOverrideInSimSpace=true
    bOverrideAngularDamping=true
    bUseGravityOverride=true
    simulationSpace=BoneRelative
    alphaBoolBlend {blendOption=Linear}
    boundBone {boneName=dyn_pant_leg_back_l_01}
    componentAppliedLinearAccClamp=<1500 1500 1500>
    componentLinearAccScale=<0.4 0.4 0.4>
    componentLinearVelScale=<0.4 0.4 0.4>
    gravityOverride {z=-980}
    physicsBodyDefinitions [{
      sphereCollisionRadius=0
      boundBone {boneName=dyn_pant_leg_back_l_01}
      boxExtents=<18 18 18>
      constraintSetup {angularLimitsMax=<10 5 10>, angularLimitsMin=<-10 -30 -10>, angularTarget=<1 0 0>}
      localJointOffset=<2 0 0>
     }]
    relativeSpaceBone {boneName=root}
   }
   n82 = Anim Dynamics - Bone: dyn_pant_leg_back_l_02(ComponentPose=@n81, Alpha=1, ExternalForce="0,0,0") {
    angularDampingOverride=0.9
    angularSpringConstant=1000
    bGravityOverrideInSimSpace=true
    bOverrideAngularDamping=true
    bUseGravityOverride=true
    simulationSpace=BoneRelative
    alphaBoolBlend {blendOption=Linear}
    boundBone {boneName=dyn_pant_leg_back_l_02}
    componentAppliedLinearAccClamp=<1500 1500 1500>
    componentLinearAccScale=<0.4 0.4 0.4>
    componentLinearVelScale=<0.4 0.4 0.4>
    gravityOverride {z=-980}
    physicsBodyDefinitions [{
      sphereCollisionRadius=0
      boundBone {boneName=dyn_pant_leg_back_l_02}
      boxExtents=<18 18 18>
      constraintSetup {angularLimitsMax=<10 5 10>, angularLimitsMin=<-10 -30 -10>, angularTarget=<1 0 0>}
      localJointOffset=<2 0 0>
     }]
    relativeSpaceBone {boneName=root}
   }
   n83 = Anim Dynamics - Bone: dyn_pant_leg_back_l_03(ComponentPose=@n82, Alpha=1, ExternalForce="0,0,0") {
    angularDampingOverride=0.9
    angularSpringConstant=1000
    bGravityOverrideInSimSpace=true
    bOverrideAngularDamping=true
    bUseGravityOverride=true
    simulationSpace=BoneRelative
    alphaBoolBlend {blendOption=Linear}
    boundBone {boneName=dyn_pant_leg_back_l_03}
    componentAppliedLinearAccClamp=<1500 1500 1500>
    componentLinearAccScale=<0.4 0.4 0.4>
    componentLinearVelScale=<0.4 0.4 0.4>
    gravityOverride {z=-980}
    physicsBodyDefinitions [{
      sphereCollisionRadius=0
      boundBone {boneName=dyn_pant_leg_back_l_03}
      boxExtents=<18 18 18>
      constraintSetup {angularLimitsMax=<10 5 10>, angularLimitsMin=<-10 -30 -10>, angularTarget=<1 0 0>}
      localJointOffset=<2 0 0>
     }]
    relativeSpaceBone {boneName=root}
   }
   n84 = Anim Dynamics - Bone: dyn_pant_leg_left_01(ComponentPose=@n83, Alpha=1, ExternalForce="0,0,0") {
    angularDampingOverride=0.9
    angularSpringConstant=1000
    bGravityOverrideInSimSpace=true
    bOverrideAngularDamping=true
    bUseGravityOverride=true
    simulationSpace=BoneRelative
    alphaBoolBlend {blendOption=Linear}
    boundBone {boneName=dyn_pant_leg_left_01}
    componentAppliedLinearAccClamp=<1500 1500 1500>
    componentLinearAccScale=<0.4 0.4 0.4>
    componentLinearVelScale=<0.4 0.4 0.4>
    gravityOverride {z=-980}
    physicsBodyDefinitions [{
      sphereCollisionRadius=0
      boundBone {boneName=dyn_pant_leg_left_01}
      boxExtents=<18 18 18>
      constraintSetup {angularLimitsMax=<10 5 10>, angularLimitsMin=<-10 -30 -10>, angularTarget=<1 0 0>}
      localJointOffset=<2 0 0>
     }]
    relativeSpaceBone {boneName=root}
   }
   n85 = Anim Dynamics - Bone: dyn_pant_leg_left_02(ComponentPose=@n84, Alpha=1, ExternalForce="0,0,0") {
    angularDampingOverride=0.9
    angularSpringConstant=1000
    bGravityOverrideInSimSpace=true
    bOverrideAngularDamping=true
    bUseGravityOverride=true
    simulationSpace=BoneRelative
    alphaBoolBlend {blendOption=Linear}
    boundBone {boneName=dyn_pant_leg_left_02}
    componentAppliedLinearAccClamp=<1500 1500 1500>
    componentLinearAccScale=<0.4 0.4 0.4>
    componentLinearVelScale=<0.4 0.4 0.4>
    gravityOverride {z=-980}
    physicsBodyDefinitions [{
      sphereCollisionRadius=0
      boundBone {boneName=dyn_pant_leg_left_02}
      boxExtents=<18 18 18>
      constraintSetup {angularLimitsMax=<10 5 10>, angularLimitsMin=<-10 -30 -10>, angularTarget=<1 0 0>}
      localJointOffset=<2 0 0>
     }]
    relativeSpaceBone {boneName=root}
   }
   n86 = Anim Dynamics - Bone: dyn_pant_leg_left_03(ComponentPose=@n85, Alpha=1, ExternalForce="0,0,0") {
    angularDampingOverride=0.9
    angularSpringConstant=1000
    bGravityOverrideInSimSpace=true
    bOverrideAngularDamping=true
    bUseGravityOverride=true
    simulationSpace=BoneRelative
    alphaBoolBlend {blendOption=Linear}
    boundBone {boneName=dyn_pant_leg_left_03}
    componentAppliedLinearAccClamp=<1500 1500 1500>
    componentLinearAccScale=<0.4 0.4 0.4>
    componentLinearVelScale=<0.4 0.4 0.4>
    gravityOverride {z=-980}
    physicsBodyDefinitions [{
      sphereCollisionRadius=0
      boundBone {boneName=dyn_pant_leg_left_03}
      boxExtents=<18 18 18>
      constraintSetup {angularLimitsMax=<10 5 10>, angularLimitsMin=<-10 -30 -10>, angularTarget=<1 0 0>}
      localJointOffset=<2 0 0>
     }]
    relativeSpaceBone {boneName=root}
   }
   n87 = Anim Dynamics - Bone: dyn_pant_leg_right_01(ComponentPose=@n86, Alpha=1, ExternalForce="0,0,0") {
    angularDampingOverride=0.9
    angularSpringConstant=1000
    bGravityOverrideInSimSpace=true
    bOverrideAngularDamping=true
    bUseGravityOverride=true
    simulationSpace=BoneRelative
    alphaBoolBlend {blendOption=Linear}
    boundBone {boneName=dyn_pant_leg_right_01}
    componentAppliedLinearAccClamp=<1500 1500 1500>
    componentLinearAccScale=<0.4 0.4 0.4>
    componentLinearVelScale=<0.4 0.4 0.4>
    gravityOverride {z=-980}
    physicsBodyDefinitions [{
      sphereCollisionRadius=0
      boundBone {boneName=dyn_pant_leg_right_01}
      boxExtents=<18 18 18>
      constraintSetup {angularLimitsMax=<10 5 10>, angularLimitsMin=<-10 -30 -10>, angularTarget=<1 0 0>}
      localJointOffset=<2 0 0>
     }]
    relativeSpaceBone {boneName=root}
   }
   n88 = Anim Dynamics - Bone: dyn_pant_leg_right_02(ComponentPose=@n87, Alpha=1, ExternalForce="0,0,0") {
    angularDampingOverride=0.9
    angularSpringConstant=1000
    bGravityOverrideInSimSpace=true
    bOverrideAngularDamping=true
    bUseGravityOverride=true
    simulationSpace=BoneRelative
    alphaBoolBlend {blendOption=Linear}
    boundBone {boneName=dyn_pant_leg_right_02}
    componentAppliedLinearAccClamp=<1500 1500 1500>
    componentLinearAccScale=<0.4 0.4 0.4>
    componentLinearVelScale=<0.4 0.4 0.4>
    gravityOverride {z=-980}
    physicsBodyDefinitions [{
      sphereCollisionRadius=0
      boundBone {boneName=dyn_pant_leg_right_02}
      boxExtents=<18 18 18>
      constraintSetup {angularLimitsMax=<10 5 10>, angularLimitsMin=<-10 -30 -10>, angularTarget=<1 0 0>}
      localJointOffset=<2 0 0>
     }]
    relativeSpaceBone {boneName=root}
   }
   n89 = Anim Dynamics - Bone: dyn_pant_leg_right_03(ComponentPose=@n88, Alpha=1, ExternalForce="0,0,0") {
    angularDampingOverride=0.9
    angularSpringConstant=1000
    bGravityOverrideInSimSpace=true
    bOverrideAngularDamping=true
    bUseGravityOverride=true
    simulationSpace=BoneRelative
    alphaBoolBlend {blendOption=Linear}
    boundBone {boneName=dyn_pant_leg_right_03}
    componentAppliedLinearAccClamp=<1500 1500 1500>
    componentLinearAccScale=<0.4 0.4 0.4>
    componentLinearVelScale=<0.4 0.4 0.4>
    gravityOverride {z=-980}
    physicsBodyDefinitions [{
      sphereCollisionRadius=0
      boundBone {boneName=dyn_pant_leg_right_03}
      boxExtents=<18 18 18>
      constraintSetup {angularLimitsMax=<10 5 10>, angularLimitsMin=<-10 -30 -10>, angularTarget=<1 0 0>}
      localJointOffset=<2 0 0>
     }]
    relativeSpaceBone {boneName=root}
   }
   n90 = Anim Dynamics - Bone: dyn_pant_leg_front_01(ComponentPose=@n89, Alpha=1, ExternalForce="0,0,0") {
    angularDampingOverride=0.9
    angularSpringConstant=1000
    bGravityOverrideInSimSpace=true
    bOverrideAngularDamping=true
    bUseGravityOverride=true
    simulationSpace=BoneRelative
    alphaBoolBlend {blendOption=Linear}
    boundBone {boneName=dyn_pant_leg_front_01}
    componentAppliedLinearAccClamp=<1500 1500 1500>
    componentLinearAccScale=<0.4 0.4 0.4>
    componentLinearVelScale=<0.4 0.4 0.4>
    gravityOverride {z=-980}
    physicsBodyDefinitions [{
      sphereCollisionRadius=0
      boundBone {boneName=dyn_pant_leg_front_01}
      boxExtents=<18 18 18>
      constraintSetup {angularLimitsMax=<10 5 10>, angularLimitsMin=<-10 -30 -10>, angularTarget=<1 0 0>}
      localJointOffset=<2 0 0>
     }]
    relativeSpaceBone {boneName=root}
   }
   n91 = Anim Dynamics - Bone: dyn_pant_leg_front_02(ComponentPose=@n90, Alpha=1, ExternalForce="0,0,0") {
    angularDampingOverride=0.9
    angularSpringConstant=1000
    bGravityOverrideInSimSpace=true
    bOverrideAngularDamping=true
    bUseGravityOverride=true
    simulationSpace=BoneRelative
    alphaBoolBlend {blendOption=Linear}
    boundBone {boneName=dyn_pant_leg_front_02}
    componentAppliedLinearAccClamp=<1500 1500 1500>
    componentLinearAccScale=<0.4 0.4 0.4>
    componentLinearVelScale=<0.4 0.4 0.4>
    gravityOverride {z=-980}
    physicsBodyDefinitions [{
      sphereCollisionRadius=0
      boundBone {boneName=dyn_pant_leg_front_02}
      boxExtents=<18 18 18>
      constraintSetup {angularLimitsMax=<10 5 10>, angularLimitsMin=<-10 -30 -10>, angularTarget=<1 0 0>}
      localJointOffset=<2 0 0>
     }]
    relativeSpaceBone {boneName=root}
   }
   n92 = Anim Dynamics - Bone: dyn_pant_leg_front_03(ComponentPose=@n91, Alpha=1, ExternalForce="0,0,0") {
    angularDampingOverride=0.9
    angularSpringConstant=1000
    bGravityOverrideInSimSpace=true
    bOverrideAngularDamping=true
    bUseGravityOverride=true
    simulationSpace=BoneRelative
    alphaBoolBlend {blendOption=Linear}
    boundBone {boneName=dyn_pant_leg_front_03}
    componentAppliedLinearAccClamp=<1500 1500 1500>
    componentLinearAccScale=<0.4 0.4 0.4>
    componentLinearVelScale=<0.4 0.4 0.4>
    gravityOverride {z=-980}
    physicsBodyDefinitions [{
      sphereCollisionRadius=0
      boundBone {boneName=dyn_pant_leg_front_03}
      boxExtents=<18 18 18>
      constraintSetup {angularLimitsMax=<10 5 10>, angularLimitsMin=<-10 -30 -10>, angularTarget=<1 0 0>}
      localJointOffset=<2 0 0>
     }]
    relativeSpaceBone {boneName=root}
   }
   n93 = RigidBody(ComponentPose=@n92, OverrideWorldGravity="0,0,-1100") {
    bClampLinearTranslationLimitToRefPose=true
    bOverrideWorldGravity=true
    overridePhysicsAsset=&/M_MED_Cube_Ninja_Physics
    simulationSpace=BaseBoneSpace
    alphaBoolBlend {blendOption=Linear}
    baseBoneRef {boneName=root}
    componentAppliedLinearAccClamp=<5000 5000 5000>
    componentLinearAccScale=<0.8 0.8 0.9>
    componentLinearVelScale=<0.8 0.8 0.9>
    overrideWorldGravity {z=-1100}
   }
   n94 = Component To Local(ComponentPose=@n93)
   n95 = Get bIsInFrontEnd()
   n96 = "Transform (Modify) Bone - Bone: dyn_r_belt_01"(
    ComponentPose=Local To Component(LocalPose=Copy Pose From Mesh() {bUseAttachedParent=true})
    Alpha=1
    Translation="0,0,0"
    Rotation="20,0,0"
    Scale="1,1,1"
   ) {
    rotationMode=BMM_Additive
    rotationSpace=BCS_ParentBoneSpace
    translationMode=BMM_Additive
    alphaBoolBlend {blendOption=Linear}
    boneToModify {boneName=dyn_r_belt_01}
    rotation {pitch=20}
   }
   n97 = "Transform (Modify) Bone - Bone: dyn_r_belt_02"(ComponentPose=@n96, Alpha=1, Translation="0,0,0", Rotation="15,0,0", Scale="1,1,1") {
    rotationMode=BMM_Additive
    rotationSpace=BCS_ParentBoneSpace
    translationMode=BMM_Additive
    alphaBoolBlend {blendOption=Linear}
    boneToModify {boneName=dyn_r_belt_02}
    rotation {pitch=15}
   }
   n98 = "Transform (Modify) Bone - Bone: dyn_l_belt_01"(ComponentPose=@n97, Alpha=1, Translation="0,0,0", Rotation="10,0,0", Scale="1,1,1") {
    rotationMode=BMM_Additive
    rotationSpace=BCS_ParentBoneSpace
    translationMode=BMM_Additive
    alphaBoolBlend {blendOption=Linear}
    boneToModify {boneName=dyn_l_belt_01}
    rotation {pitch=10}
   }
   n99 = "Transform (Modify) Bone - Bone: dyn_l_belt_02"(ComponentPose=@n98, Alpha=1, Translation="0,0,0", Rotation="10,0,0", Scale="1,1,1") {
    rotationMode=BMM_Additive
    rotationSpace=BCS_ParentBoneSpace
    translationMode=BMM_Additive
    alphaBoolBlend {blendOption=Linear}
    boneToModify {boneName=dyn_l_belt_02}
    rotation {pitch=10}
   }
   n100 = "Constraint - dyn_sash_front_master (PTT)"(ComponentPose=@n99, Alpha=1, ConstraintWeights_0=1, ConstraintWeights_1=0, ConstraintWeights_2=0) {
    alphaBoolBlend {blendOption=Linear}
    boneToModify {boneName=dyn_sash_front_master}
    constraintSetup [
     {targetBone {boneName=dyn_scrolls_master}}
     {transformType=Translation, targetBone {boneName=None}}
     {transformType=Translation, targetBone {boneName=None}}
    ]
    constraintWeights [1,1,1]
   }
   n101 = "Constraint - dyn_sash_back_master (PTT)"(ComponentPose=@n100, Alpha=1, ConstraintWeights_0=1, ConstraintWeights_1=0, ConstraintWeights_2=0) {
    alphaBoolBlend {blendOption=Linear}
    boneToModify {boneName=dyn_sash_back_master}
    constraintSetup [
     {targetBone {boneName=dyn_scrolls_master}}
     {transformType=Translation, targetBone {boneName=None}}
     {transformType=Translation, targetBone {boneName=None}}
    ]
    constraintWeights [1,1,1]
   }
   n102 = "Transform (Modify) Bone - Bone: dyn_sash_front_01"(ComponentPose=@n101, Alpha=0.5, Translation="0,0,0", Rotation=Get TesselRotator_01(), Scale="1,1,1") {
    rotationMode=BMM_Additive
    rotationSpace=BCS_BoneSpace
    translationSpace=BCS_BoneSpace
    alphaBoolBlend {blendOption=Linear}
    alphaScaleBiasClamp {interpSpeedDecreasing=2, interpSpeedIncreasing=2}
    boneToModify {boneName=dyn_sash_front_01}
   }
   n103 = "Transform (Modify) Bone - Bone: dyn_sash_back_01"(ComponentPose=@n102, Alpha=1, Translation="0,0,0", Rotation=Get TesselRotator_02(), Scale="1,1,1") {
    rotationMode=BMM_Additive
    rotationSpace=BCS_BoneSpace
    translationSpace=BCS_BoneSpace
    alphaBoolBlend {blendOption=Linear}
    alphaScaleBiasClamp {bInterpResult=true, interpSpeedDecreasing=2, interpSpeedIncreasing=2}
    boneToModify {boneName=dyn_sash_back_01}
   }
   n104 = Get GravityOverride_tessel()
   n105 = Anim Dynamics - Bone: dyn_l_belt_01!(ComponentPose=@n103, Alpha=1, GravityOverride:Vector=@n104, ExternalForce="0,0,0") -> (Pose) {
    angularDampingOverride=0.7
    angularSpringConstant=1000
    bGravityOverrideInSimSpace=true
    bUseGravityOverride=true
    simulationSpace=BoneRelative
    alphaBoolBlend {blendOption=Linear}
    boundBone {boneName=dyn_l_belt_01}
    componentAppliedLinearAccClamp=<1500 1500 1500>
    componentLinearAccScale=<0.4 0.4 0.4>
    componentLinearVelScale=<0.4 0.4 0.4>
    gravityOverride {z=-980}
    physicsBodyDefinitions [{
      sphereCollisionRadius=0
      boundBone {boneName=dyn_l_belt_01}
      boxExtents=<20 20 20>
      constraintSetup {angularLimitsMax=<3 -5 0>, angularLimitsMin=<-3 -12 0>, angularTarget=<1 0 0>}
      localJointOffset=<-3 0 0>
     }]
    relativeSpaceBone {boneName=dyn_r_belt_master}
   }
   n106 = Anim Dynamics - Bone: dyn_l_belt_02!(ComponentPose=@n105, Alpha=1, GravityOverride:Vector=@n104, ExternalForce="0,0,0") -> (Pose) {
    angularDampingOverride=0.7
    angularSpringConstant=1000
    bGravityOverrideInSimSpace=true
    bUseGravityOverride=true
    simulationSpace=BoneRelative
    alphaBoolBlend {blendOption=Linear}
    boundBone {boneName=dyn_l_belt_02}
    componentAppliedLinearAccClamp=<1500 1500 1500>
    componentLinearAccScale=<0.4 0.4 0.4>
    componentLinearVelScale=<0.4 0.4 0.4>
    gravityOverride {z=-980}
    physicsBodyDefinitions [{
      sphereCollisionRadius=0
      boundBone {boneName=dyn_l_belt_02}
      boxExtents=<20 20 20>
      constraintSetup {angularLimitsMax=<3 -5 0>, angularLimitsMin=<-3 -12 0>, angularTarget=<1 0 0>}
      localJointOffset=<-3 0 0>
     }]
    relativeSpaceBone {boneName=dyn_r_belt_master}
   }
   n107 = Anim Dynamics - Bone: dyn_l_belt_03!(ComponentPose=@n106, Alpha=1, GravityOverride:Vector=@n104, ExternalForce="0,0,0") -> (Pose) {
    angularDampingOverride=0.7
    angularSpringConstant=1000
    bGravityOverrideInSimSpace=true
    bUseGravityOverride=true
    simulationSpace=BoneRelative
    alphaBoolBlend {blendOption=Linear}
    boundBone {boneName=dyn_l_belt_03}
    componentAppliedLinearAccClamp=<1500 1500 1500>
    componentLinearAccScale=<0.4 0.4 0.4>
    componentLinearVelScale=<0.4 0.4 0.4>
    gravityOverride {z=-980}
    physicsBodyDefinitions [{
      sphereCollisionRadius=0
      boundBone {boneName=dyn_l_belt_03}
      boxExtents=<20 20 20>
      constraintSetup {angularLimitsMax=<3 -5 3>, angularLimitsMin=<-3 -12 -3>, angularTarget=<1 0 0>}
      localJointOffset=<-3 0 0>
     }]
    relativeSpaceBone {boneName=dyn_r_belt_master}
   }
   n108 = Anim Dynamics - Bone: dyn_mask_01!(ComponentPose=@n107, Alpha=1, GravityOverride:Vector=@n104, ExternalForce="0,0,0") -> (Pose) {
    angularDampingOverride=0.7
    angularSpringConstant=1000
    bGravityOverrideInSimSpace=true
    bUseGravityOverride=true
    simulationSpace=BoneRelative
    alphaBoolBlend {blendOption=Linear}
    boundBone {boneName=dyn_mask_01}
    componentAppliedLinearAccClamp=<1500 1500 1500>
    componentLinearAccScale=<0.4 0.4 0.4>
    componentLinearVelScale=<0.4 0.4 0.4>
    gravityOverride {z=-980}
    physicsBodyDefinitions [{
      sphereCollisionRadius=0
      boundBone {boneName=dyn_mask_01}
      boxExtents=<20 20 20>
      constraintSetup {angularLimitsMax=<3 5 3>, angularLimitsMin=<-3 -5 -3>, angularTarget=<1 0 0>}
      localJointOffset=<5 0 0>
     }]
    relativeSpaceBone {boneName=dyn_r_belt_master}
   }
   n109 = Anim Dynamics - Bone: dyn_r_belt_01!(ComponentPose=@n108, Alpha=1, GravityOverride:Vector=@n104, ExternalForce="0,0,0") -> (Pose) {
    angularDampingOverride=0.7
    angularSpringConstant=1000
    bGravityOverrideInSimSpace=true
    bUseGravityOverride=true
    simulationSpace=BoneRelative
    alphaBoolBlend {blendOption=Linear}
    boundBone {boneName=dyn_r_belt_01}
    componentAppliedLinearAccClamp=<1500 1500 1500>
    componentLinearAccScale=<0.4 0.4 0.4>
    componentLinearVelScale=<0.4 0.4 0.4>
    gravityOverride {z=-980}
    physicsBodyDefinitions [{
      sphereCollisionRadius=0
      boundBone {boneName=dyn_r_belt_01}
      boxExtents=<20 20 20>
      constraintSetup {angularLimitsMax=<3 -5 0>, angularLimitsMin=<-3 -12 0>, angularTarget=<1 -5 0>}
      localJointOffset=<3 0 0>
     }]
    relativeSpaceBone {boneName=dyn_r_belt_master}
   }
   n110 = Anim Dynamics - Bone: dyn_r_belt_02!(ComponentPose=@n109, Alpha=1, GravityOverride:Vector=@n104, ExternalForce="0,0,0") -> (Pose) {
    angularDampingOverride=0.7
    angularSpringConstant=1000
    bGravityOverrideInSimSpace=true
    bUseGravityOverride=true
    simulationSpace=BoneRelative
    alphaBoolBlend {blendOption=Linear}
    boundBone {boneName=dyn_r_belt_02}
    componentAppliedLinearAccClamp=<1500 1500 1500>
    componentLinearAccScale=<0.4 0.4 0.4>
    componentLinearVelScale=<0.4 0.4 0.4>
    gravityOverride {z=-980}
    physicsBodyDefinitions [{
      sphereCollisionRadius=0
      boundBone {boneName=dyn_r_belt_02}
      boxExtents=<20 20 20>
      constraintSetup {angularLimitsMax=<3 -5 0>, angularLimitsMin=<-3 -12 0>, angularTarget=<1 -5 0>}
      localJointOffset=<3 0 0>
     }]
    relativeSpaceBone {boneName=dyn_r_belt_master}
   }
   n111 = Anim Dynamics - Bone: dyn_r_belt_03!(ComponentPose=@n110, Alpha=1, GravityOverride:Vector=@n104, ExternalForce="0,0,0") -> (Pose) {
    angularDampingOverride=0.7
    angularSpringConstant=1000
    bGravityOverrideInSimSpace=true
    bUseGravityOverride=true
    simulationSpace=BoneRelative
    alphaBoolBlend {blendOption=Linear}
    boundBone {boneName=dyn_r_belt_03}
    componentAppliedLinearAccClamp=<1500 1500 1500>
    componentLinearAccScale=<0.4 0.4 0.4>
    componentLinearVelScale=<0.4 0.4 0.4>
    gravityOverride {z=-980}
    physicsBodyDefinitions [{
      sphereCollisionRadius=0
      boundBone {boneName=dyn_r_belt_03}
      boxExtents=<20 20 20>
      constraintSetup {angularLimitsMax=<3 -5 3>, angularLimitsMin=<-3 -12 -3>, angularTarget=<1 -5 0>}
      localJointOffset=<3 0 0>
     }]
    relativeSpaceBone {boneName=dyn_r_belt_master}
   }
   n112 = Anim Dynamics - Bone: dyn_mask_01(ComponentPose=@n111, Alpha=1, GravityOverride=@n104, ExternalForce="0,0,0") {
    angularDampingOverride=0.7
    angularSpringConstant=1000
    bGravityOverrideInSimSpace=true
    bUseGravityOverride=true
    simulationSpace=BoneRelative
    alphaBoolBlend {blendOption=Linear}
    boundBone {boneName=dyn_mask_01}
    componentAppliedLinearAccClamp=<1500 1500 1500>
    componentLinearAccScale=<0.4 0.4 0.4>
    componentLinearVelScale=<0.4 0.4 0.4>
    gravityOverride {z=-980}
    physicsBodyDefinitions [{
      sphereCollisionRadius=0
      boundBone {boneName=dyn_mask_01}
      boxExtents=<5 5 5>
      constraintSetup {angularLimitsMax=<5 -5 5>, angularLimitsMin=<-5 -12 -5>, angularTarget=<1 -5 0>}
      localJointOffset=<3 0 0>
     }]
    relativeSpaceBone {boneName=dyn_r_belt_master}
   }
   n113 = "Constraint - dyn_scrolls_master (TTT)"(ComponentPose=@n112, Alpha=1, ConstraintWeights_0=1, ConstraintWeights_1=0, ConstraintWeights_2=0) {
    alphaBoolBlend {blendOption=Linear}
    boneToModify {boneName=dyn_scrolls_master}
    constraintSetup [
     {transformType=Translation, targetBone {boneName=dyn_r_belt_03}}
     {transformType=Translation, targetBone {boneName=None}}
     {transformType=Translation, targetBone {boneName=None}}
    ]
    constraintWeights [1,1,1]
   }
   n114 = Get GravityOverride_tessel()
   n115 = Anim Dynamics - Bone: dyn_leg_ribbon_top_01!(ComponentPose=@n113, Alpha=1, GravityOverride:Vector=@n114, ExternalForce="0,0,0") -> (Pose) {
    angularDampingOverride=0.7
    angularSpringConstant=1000
    bGravityOverrideInSimSpace=true
    bUseGravityOverride=true
    simulationSpace=BoneRelative
    alphaBoolBlend {blendOption=Linear}
    boundBone {boneName=dyn_leg_ribbon_top_01}
    componentAppliedLinearAccClamp=<1500 1500 1500>
    componentLinearAccScale=<0.4 0.4 0.4>
    componentLinearVelScale=<0.4 0.4 0.4>
    gravityOverride {z=-980}
    physicsBodyDefinitions [{
      sphereCollisionRadius=0
      boundBone {boneName=dyn_leg_ribbon_top_01}
      boxExtents=<15 15 15>
      constraintSetup {angularLimitsMax=<5 5 5>, angularLimitsMin=<-5 -5 -5>, angularTarget=<1 0 0>}
      localJointOffset=<3 0 0>
     }]
    relativeSpaceBone {boneName=root}
   }
   n116 = Anim Dynamics - Bone: dyn_leg_ribbon_top_02!(ComponentPose=@n115, Alpha=1, GravityOverride:Vector=@n114, ExternalForce="0,0,0") -> (Pose) {
    angularDampingOverride=0.7
    angularSpringConstant=1000
    bGravityOverrideInSimSpace=true
    bUseGravityOverride=true
    simulationSpace=BoneRelative
    alphaBoolBlend {blendOption=Linear}
    boundBone {boneName=dyn_leg_ribbon_top_02}
    componentAppliedLinearAccClamp=<1500 1500 1500>
    componentLinearAccScale=<0.4 0.4 0.4>
    componentLinearVelScale=<0.4 0.4 0.4>
    gravityOverride {z=-980}
    physicsBodyDefinitions [{
      sphereCollisionRadius=0
      boundBone {boneName=dyn_leg_ribbon_top_02}
      boxExtents=<15 15 15>
      constraintSetup {angularLimitsMax=<5 5 5>, angularLimitsMin=<-5 -5 -5>, angularTarget=<1 0 0>}
      localJointOffset=<3 0 0>
     }]
    relativeSpaceBone {boneName=root}
   }
   n117 = Anim Dynamics - Bone: dyn_leg_ribbon_top_03!(ComponentPose=@n116, Alpha=1, GravityOverride:Vector=@n114, ExternalForce="0,0,0") -> (Pose) {
    angularDampingOverride=0.7
    angularSpringConstant=1000
    bGravityOverrideInSimSpace=true
    bUseGravityOverride=true
    simulationSpace=BoneRelative
    alphaBoolBlend {blendOption=Linear}
    boundBone {boneName=dyn_leg_ribbon_top_03}
    componentAppliedLinearAccClamp=<1500 1500 1500>
    componentLinearAccScale=<0.4 0.4 0.4>
    componentLinearVelScale=<0.4 0.4 0.4>
    gravityOverride {z=-980}
    physicsBodyDefinitions [{
      sphereCollisionRadius=0
      boundBone {boneName=dyn_leg_ribbon_top_03}
      boxExtents=<15 15 15>
      constraintSetup {angularLimitsMax=<5 5 5>, angularLimitsMin=<-5 -5 -5>, angularTarget=<1 0 0>}
      localJointOffset=<3 0 0>
     }]
    relativeSpaceBone {boneName=root}
   }
   n118 = Anim Dynamics - Bone: dyn_leg_ribbon_01!(ComponentPose=@n117, Alpha=1, GravityOverride:Vector=@n114, ExternalForce="0,0,0") -> (Pose) {
    angularDampingOverride=0.7
    angularSpringConstant=1000
    bGravityOverrideInSimSpace=true
    bUseGravityOverride=true
    simulationSpace=BoneRelative
    alphaBoolBlend {blendOption=Linear}
    boundBone {boneName=dyn_leg_ribbon_01}
    componentAppliedLinearAccClamp=<1500 1500 1500>
    componentLinearAccScale=<0.4 0.4 0.4>
    componentLinearVelScale=<0.4 0.4 0.4>
    gravityOverride {z=-980}
    physicsBodyDefinitions [{
      sphereCollisionRadius=0
      boundBone {boneName=dyn_leg_ribbon_01}
      boxExtents=<20 20 20>
      constraintSetup {angularLimitsMax=<5 5 5>, angularLimitsMin=<-5 -5 -5>, angularTarget=<1 0 0>}
      localJointOffset=<3 0 0>
     }]
    relativeSpaceBone {boneName=root}
   }
   n119 = Anim Dynamics - Bone: dyn_leg_ribbon_02!(ComponentPose=@n118, Alpha=1, GravityOverride:Vector=@n114, ExternalForce="0,0,0") -> (Pose) {
    angularDampingOverride=0.7
    angularSpringConstant=1000
    bGravityOverrideInSimSpace=true
    bUseGravityOverride=true
    simulationSpace=BoneRelative
    alphaBoolBlend {blendOption=Linear}
    boundBone {boneName=dyn_leg_ribbon_02}
    componentAppliedLinearAccClamp=<1500 1500 1500>
    componentLinearAccScale=<0.4 0.4 0.4>
    componentLinearVelScale=<0.4 0.4 0.4>
    gravityOverride {z=-980}
    physicsBodyDefinitions [{
      sphereCollisionRadius=0
      boundBone {boneName=dyn_leg_ribbon_02}
      boxExtents=<20 20 20>
      constraintSetup {angularLimitsMax=<5 5 5>, angularLimitsMin=<-5 -5 -5>, angularTarget=<1 0 0>}
      localJointOffset=<3 0 0>
     }]
    relativeSpaceBone {boneName=root}
   }
   n120 = Anim Dynamics - Bone: dyn_leg_ribbon_03!(ComponentPose=@n119, Alpha=1, GravityOverride:Vector=@n114, ExternalForce="0,0,0") -> (Pose) {
    angularDampingOverride=0.7
    angularSpringConstant=1000
    bGravityOverrideInSimSpace=true
    bUseGravityOverride=true
    simulationSpace=BoneRelative
    alphaBoolBlend {blendOption=Linear}
    boundBone {boneName=dyn_leg_ribbon_03}
    componentAppliedLinearAccClamp=<1500 1500 1500>
    componentLinearAccScale=<0.4 0.4 0.4>
    componentLinearVelScale=<0.4 0.4 0.4>
    gravityOverride {z=-980}
    physicsBodyDefinitions [{
      sphereCollisionRadius=0
      boundBone {boneName=dyn_leg_ribbon_03}
      boxExtents=<20 20 20>
      constraintSetup {angularLimitsMax=<5 5 5>, angularLimitsMin=<-5 -5 -5>, angularTarget=<1 0 0>}
      localJointOffset=<3 0 0>
     }]
    relativeSpaceBone {boneName=root}
   }
   n121 = Anim Dynamics - Bone: dyn_crystal_01!(ComponentPose=@n120, Alpha=1, GravityOverride:Vector="0,0,-980", ExternalForce="0,0,0") -> (Pose) {
    angularDampingOverride=0.9
    angularSpringConstant=1000
    bGravityOverrideInSimSpace=true
    bOverrideAngularDamping=true
    bUseGravityOverride=true
    simulationSpace=BoneRelative
    alphaBoolBlend {blendOption=Linear}
    boundBone {boneName=dyn_crystal_01}
    componentAppliedLinearAccClamp=<1500 1500 1500>
    componentLinearAccScale=<0.4 0.4 0.4>
    componentLinearVelScale=<0.4 0.4 0.4>
    gravityOverride {z=-980}
    physicsBodyDefinitions [{
      sphereCollisionRadius=0
      boundBone {boneName=dyn_crystal_01}
      boxExtents=<10 10 10>
      constraintSetup {angularLimitsMax=<5 5 5>, angularLimitsMin=<-5 -5 -5>, angularTarget=<1 0 0>}
      localJointOffset=<3 0 0>
     }]
    relativeSpaceBone {boneName=root}
   }
   n122 = Anim Dynamics - Bone: dyn_crystal_02!(ComponentPose=@n121, Alpha=1, GravityOverride:Vector="0,0,-980", ExternalForce="0,0,0") -> (Pose) {
    angularDampingOverride=0.9
    angularSpringConstant=1000
    bGravityOverrideInSimSpace=true
    bOverrideAngularDamping=true
    bUseGravityOverride=true
    simulationSpace=BoneRelative
    alphaBoolBlend {blendOption=Linear}
    boundBone {boneName=dyn_crystal_02}
    componentAppliedLinearAccClamp=<1500 1500 1500>
    componentLinearAccScale=<0.4 0.4 0.4>
    componentLinearVelScale=<0.4 0.4 0.4>
    gravityOverride {z=-980}
    physicsBodyDefinitions [{
      sphereCollisionRadius=0
      boundBone {boneName=dyn_crystal_02}
      boxExtents=<12 12 12>
      constraintSetup {angularLimitsMax=<2 10 10>, angularLimitsMin=<-2 -10 -10>, angularTarget=<1 0 0>}
      localJointOffset=<3 0 0>
     }]
    relativeSpaceBone {boneName=root}
   }
   n123 = Anim Dynamics - Bone: dyn_crystal_03!(ComponentPose=@n122, Alpha=1, GravityOverride:Vector="0,-1100,-980", ExternalForce="0,0,0") -> (Pose) {
    angularDampingOverride=0.9
    angularSpringConstant=1000
    bGravityOverrideInSimSpace=true
    bOverrideAngularDamping=true
    bUseGravityOverride=true
    simulationSpace=BoneRelative
    alphaBoolBlend {blendOption=Linear}
    boundBone {boneName=dyn_crystal_03}
    componentAppliedLinearAccClamp=<1500 1500 1500>
    componentLinearAccScale=<0.4 0.4 0.4>
    componentLinearVelScale=<0.4 0.4 0.4>
    gravityOverride {z=-980}
    physicsBodyDefinitions [{
      sphereCollisionRadius=0
      boundBone {boneName=dyn_crystal_03}
      boxExtents=<12 12 12>
      constraintSetup {angularLimitsMax=<40 40 40>, angularLimitsMin=<-40 -40 -40>, angularTarget=<1 0 0>}
      localJointOffset=<3 0 0>
     }]
    relativeSpaceBone {boneName=root}
   }
   n124 = Get GravityOverride_tessel()
   n125 = Anim Dynamics - Bone: dyn_pant_leg_back_r_01!(ComponentPose=@n123, Alpha=1, GravityOverride:Vector=@n124, ExternalForce="0,0,0") -> (Pose) {
    angularDampingOverride=0.9
    angularSpringConstant=1000
    bGravityOverrideInSimSpace=true
    bOverrideAngularDamping=true
    bUseGravityOverride=true
    simulationSpace=BoneRelative
    alphaBoolBlend {blendOption=Linear}
    boundBone {boneName=dyn_pant_leg_back_r_01}
    componentAppliedLinearAccClamp=<1500 1500 1500>
    componentLinearAccScale=<0.4 0.4 0.4>
    componentLinearVelScale=<0.4 0.4 0.4>
    gravityOverride {z=-980}
    physicsBodyDefinitions [{
      sphereCollisionRadius=0
      boundBone {boneName=dyn_pant_leg_back_r_01}
      boxExtents=<18 18 18>
      constraintSetup {angularLimitsMax=<10 5 10>, angularLimitsMin=<-10 -30 -10>, angularTarget=<1 0 0>}
      localJointOffset=<2 0 0>
     }]
    relativeSpaceBone {boneName=root}
   }
   n126 = Anim Dynamics - Bone: dyn_pant_leg_back_r_02!(ComponentPose=@n125, Alpha=1, GravityOverride:Vector=@n124, ExternalForce="0,0,0") -> (Pose) {
    angularDampingOverride=0.9
    angularSpringConstant=1000
    bGravityOverrideInSimSpace=true
    bOverrideAngularDamping=true
    bUseGravityOverride=true
    simulationSpace=BoneRelative
    alphaBoolBlend {blendOption=Linear}
    boundBone {boneName=dyn_pant_leg_back_r_02}
    componentAppliedLinearAccClamp=<1500 1500 1500>
    componentLinearAccScale=<0.4 0.4 0.4>
    componentLinearVelScale=<0.4 0.4 0.4>
    gravityOverride {z=-980}
    physicsBodyDefinitions [{
      sphereCollisionRadius=0
      boundBone {boneName=dyn_pant_leg_back_r_02}
      boxExtents=<18 18 18>
      constraintSetup {angularLimitsMax=<10 5 10>, angularLimitsMin=<-10 -30 -10>, angularTarget=<1 0 0>}
      localJointOffset=<2 0 0>
     }]
    relativeSpaceBone {boneName=root}
   }
   n127 = Anim Dynamics - Bone: dyn_pant_leg_back_r_03!(ComponentPose=@n126, Alpha=1, GravityOverride:Vector=@n124, ExternalForce="0,0,0") -> (Pose) {
    angularDampingOverride=0.9
    angularSpringConstant=1000
    bGravityOverrideInSimSpace=true
    bOverrideAngularDamping=true
    bUseGravityOverride=true
    simulationSpace=BoneRelative
    alphaBoolBlend {blendOption=Linear}
    boundBone {boneName=dyn_pant_leg_back_r_03}
    componentAppliedLinearAccClamp=<1500 1500 1500>
    componentLinearAccScale=<0.4 0.4 0.4>
    componentLinearVelScale=<0.4 0.4 0.4>
    gravityOverride {z=-980}
    physicsBodyDefinitions [{
      sphereCollisionRadius=0
      boundBone {boneName=dyn_pant_leg_back_r_03}
      boxExtents=<18 18 18>
      constraintSetup {angularLimitsMax=<10 5 10>, angularLimitsMin=<-10 -30 -10>, angularTarget=<1 0 0>}
      localJointOffset=<2 0 0>
     }]
    relativeSpaceBone {boneName=root}
   }
   n128 = Anim Dynamics - Bone: dyn_pant_leg_back_l_01!(ComponentPose=@n127, Alpha=1, GravityOverride:Vector=@n124, ExternalForce="0,0,0") -> (Pose) {
    angularDampingOverride=0.9
    angularSpringConstant=1000
    bGravityOverrideInSimSpace=true
    bOverrideAngularDamping=true
    bUseGravityOverride=true
    simulationSpace=BoneRelative
    alphaBoolBlend {blendOption=Linear}
    boundBone {boneName=dyn_pant_leg_back_l_01}
    componentAppliedLinearAccClamp=<1500 1500 1500>
    componentLinearAccScale=<0.4 0.4 0.4>
    componentLinearVelScale=<0.4 0.4 0.4>
    gravityOverride {z=-980}
    physicsBodyDefinitions [{
      sphereCollisionRadius=0
      boundBone {boneName=dyn_pant_leg_back_l_01}
      boxExtents=<18 18 18>
      constraintSetup {angularLimitsMax=<10 5 10>, angularLimitsMin=<-10 -30 -10>, angularTarget=<1 0 0>}
      localJointOffset=<2 0 0>
     }]
    relativeSpaceBone {boneName=root}
   }
   n129 = Anim Dynamics - Bone: dyn_pant_leg_back_l_02!(ComponentPose=@n128, Alpha=1, GravityOverride:Vector=@n124, ExternalForce="0,0,0") -> (Pose) {
    angularDampingOverride=0.9
    angularSpringConstant=1000
    bGravityOverrideInSimSpace=true
    bOverrideAngularDamping=true
    bUseGravityOverride=true
    simulationSpace=BoneRelative
    alphaBoolBlend {blendOption=Linear}
    boundBone {boneName=dyn_pant_leg_back_l_02}
    componentAppliedLinearAccClamp=<1500 1500 1500>
    componentLinearAccScale=<0.4 0.4 0.4>
    componentLinearVelScale=<0.4 0.4 0.4>
    gravityOverride {z=-980}
    physicsBodyDefinitions [{
      sphereCollisionRadius=0
      boundBone {boneName=dyn_pant_leg_back_l_02}
      boxExtents=<18 18 18>
      constraintSetup {angularLimitsMax=<10 5 10>, angularLimitsMin=<-10 -30 -10>, angularTarget=<1 0 0>}
      localJointOffset=<2 0 0>
     }]
    relativeSpaceBone {boneName=root}
   }
   n130 = Anim Dynamics - Bone: dyn_pant_leg_back_l_03!(ComponentPose=@n129, Alpha=1, GravityOverride:Vector=@n124, ExternalForce="0,0,0") -> (Pose) {
    angularDampingOverride=0.9
    angularSpringConstant=1000
    bGravityOverrideInSimSpace=true
    bOverrideAngularDamping=true
    bUseGravityOverride=true
    simulationSpace=BoneRelative
    alphaBoolBlend {blendOption=Linear}
    boundBone {boneName=dyn_pant_leg_back_l_03}
    componentAppliedLinearAccClamp=<1500 1500 1500>
    componentLinearAccScale=<0.4 0.4 0.4>
    componentLinearVelScale=<0.4 0.4 0.4>
    gravityOverride {z=-980}
    physicsBodyDefinitions [{
      sphereCollisionRadius=0
      boundBone {boneName=dyn_pant_leg_back_l_03}
      boxExtents=<18 18 18>
      constraintSetup {angularLimitsMax=<10 5 10>, angularLimitsMin=<-10 -30 -10>, angularTarget=<1 0 0>}
      localJointOffset=<2 0 0>
     }]
    relativeSpaceBone {boneName=root}
   }
   n131 = Anim Dynamics - Bone: dyn_pant_leg_left_01!(ComponentPose=@n130, Alpha=1, GravityOverride:Vector=@n124, ExternalForce="0,0,0") -> (Pose) {
    angularDampingOverride=0.9
    angularSpringConstant=1000
    bGravityOverrideInSimSpace=true
    bOverrideAngularDamping=true
    bUseGravityOverride=true
    simulationSpace=BoneRelative
    alphaBoolBlend {blendOption=Linear}
    boundBone {boneName=dyn_pant_leg_left_01}
    componentAppliedLinearAccClamp=<1500 1500 1500>
    componentLinearAccScale=<0.4 0.4 0.4>
    componentLinearVelScale=<0.4 0.4 0.4>
    gravityOverride {z=-980}
    physicsBodyDefinitions [{
      sphereCollisionRadius=0
      boundBone {boneName=dyn_pant_leg_left_01}
      boxExtents=<18 18 18>
      constraintSetup {angularLimitsMax=<10 5 10>, angularLimitsMin=<-10 -30 -10>, angularTarget=<1 0 0>}
      localJointOffset=<2 0 0>
     }]
    relativeSpaceBone {boneName=root}
   }
   n132 = Anim Dynamics - Bone: dyn_pant_leg_left_02!(ComponentPose=@n131, Alpha=1, GravityOverride:Vector=@n124, ExternalForce="0,0,0") -> (Pose) {
    angularDampingOverride=0.9
    angularSpringConstant=1000
    bGravityOverrideInSimSpace=true
    bOverrideAngularDamping=true
    bUseGravityOverride=true
    simulationSpace=BoneRelative
    alphaBoolBlend {blendOption=Linear}
    boundBone {boneName=dyn_pant_leg_left_02}
    componentAppliedLinearAccClamp=<1500 1500 1500>
    componentLinearAccScale=<0.4 0.4 0.4>
    componentLinearVelScale=<0.4 0.4 0.4>
    gravityOverride {z=-980}
    physicsBodyDefinitions [{
      sphereCollisionRadius=0
      boundBone {boneName=dyn_pant_leg_left_02}
      boxExtents=<18 18 18>
      constraintSetup {angularLimitsMax=<10 5 10>, angularLimitsMin=<-10 -30 -10>, angularTarget=<1 0 0>}
      localJointOffset=<2 0 0>
     }]
    relativeSpaceBone {boneName=root}
   }
   n133 = Anim Dynamics - Bone: dyn_pant_leg_left_03!(ComponentPose=@n132, Alpha=1, GravityOverride:Vector=@n124, ExternalForce="0,0,0") -> (Pose) {
    angularDampingOverride=0.9
    angularSpringConstant=1000
    bGravityOverrideInSimSpace=true
    bOverrideAngularDamping=true
    bUseGravityOverride=true
    simulationSpace=BoneRelative
    alphaBoolBlend {blendOption=Linear}
    boundBone {boneName=dyn_pant_leg_left_03}
    componentAppliedLinearAccClamp=<1500 1500 1500>
    componentLinearAccScale=<0.4 0.4 0.4>
    componentLinearVelScale=<0.4 0.4 0.4>
    gravityOverride {z=-980}
    physicsBodyDefinitions [{
      sphereCollisionRadius=0
      boundBone {boneName=dyn_pant_leg_left_03}
      boxExtents=<18 18 18>
      constraintSetup {angularLimitsMax=<10 5 10>, angularLimitsMin=<-10 -30 -10>, angularTarget=<1 0 0>}
      localJointOffset=<2 0 0>
     }]
    relativeSpaceBone {boneName=root}
   }
   n134 = Anim Dynamics - Bone: dyn_pant_leg_right_01!(ComponentPose=@n133, Alpha=1, GravityOverride:Vector=@n124, ExternalForce="0,0,0") -> (Pose) {
    angularDampingOverride=0.9
    angularSpringConstant=1000
    bGravityOverrideInSimSpace=true
    bOverrideAngularDamping=true
    bUseGravityOverride=true
    simulationSpace=BoneRelative
    alphaBoolBlend {blendOption=Linear}
    boundBone {boneName=dyn_pant_leg_right_01}
    componentAppliedLinearAccClamp=<1500 1500 1500>
    componentLinearAccScale=<0.4 0.4 0.4>
    componentLinearVelScale=<0.4 0.4 0.4>
    gravityOverride {z=-980}
    physicsBodyDefinitions [{
      sphereCollisionRadius=0
      boundBone {boneName=dyn_pant_leg_right_01}
      boxExtents=<18 18 18>
      constraintSetup {angularLimitsMax=<10 5 10>, angularLimitsMin=<-10 -30 -10>, angularTarget=<1 0 0>}
      localJointOffset=<2 0 0>
     }]
    relativeSpaceBone {boneName=root}
   }
   n135 = Anim Dynamics - Bone: dyn_pant_leg_right_02!(ComponentPose=@n134, Alpha=1, GravityOverride:Vector=@n124, ExternalForce="0,0,0") -> (Pose) {
    angularDampingOverride=0.9
    angularSpringConstant=1000
    bGravityOverrideInSimSpace=true
    bOverrideAngularDamping=true
    bUseGravityOverride=true
    simulationSpace=BoneRelative
    alphaBoolBlend {blendOption=Linear}
    boundBone {boneName=dyn_pant_leg_right_02}
    componentAppliedLinearAccClamp=<1500 1500 1500>
    componentLinearAccScale=<0.4 0.4 0.4>
    componentLinearVelScale=<0.4 0.4 0.4>
    gravityOverride {z=-980}
    physicsBodyDefinitions [{
      sphereCollisionRadius=0
      boundBone {boneName=dyn_pant_leg_right_02}
      boxExtents=<18 18 18>
      constraintSetup {angularLimitsMax=<10 5 10>, angularLimitsMin=<-10 -30 -10>, angularTarget=<1 0 0>}
      localJointOffset=<2 0 0>
     }]
    relativeSpaceBone {boneName=root}
   }
   n136 = Anim Dynamics - Bone: dyn_pant_leg_right_03!(ComponentPose=@n135, Alpha=1, GravityOverride:Vector=@n124, ExternalForce="0,0,0") -> (Pose) {
    angularDampingOverride=0.9
    angularSpringConstant=1000
    bGravityOverrideInSimSpace=true
    bOverrideAngularDamping=true
    bUseGravityOverride=true
    simulationSpace=BoneRelative
    alphaBoolBlend {blendOption=Linear}
    boundBone {boneName=dyn_pant_leg_right_03}
    componentAppliedLinearAccClamp=<1500 1500 1500>
    componentLinearAccScale=<0.4 0.4 0.4>
    componentLinearVelScale=<0.4 0.4 0.4>
    gravityOverride {z=-980}
    physicsBodyDefinitions [{
      sphereCollisionRadius=0
      boundBone {boneName=dyn_pant_leg_right_03}
      boxExtents=<18 18 18>
      constraintSetup {angularLimitsMax=<10 5 10>, angularLimitsMin=<-10 -30 -10>, angularTarget=<1 0 0>}
      localJointOffset=<2 0 0>
     }]
    relativeSpaceBone {boneName=root}
   }
   n137 = Anim Dynamics - Bone: dyn_pant_leg_front_01!(ComponentPose=@n136, Alpha=1, GravityOverride:Vector=@n124, ExternalForce="0,0,0") -> (Pose) {
    angularDampingOverride=0.9
    angularSpringConstant=1000
    bGravityOverrideInSimSpace=true
    bOverrideAngularDamping=true
    bUseGravityOverride=true
    simulationSpace=BoneRelative
    alphaBoolBlend {blendOption=Linear}
    boundBone {boneName=dyn_pant_leg_front_01}
    componentAppliedLinearAccClamp=<1500 1500 1500>
    componentLinearAccScale=<0.4 0.4 0.4>
    componentLinearVelScale=<0.4 0.4 0.4>
    gravityOverride {z=-980}
    physicsBodyDefinitions [{
      sphereCollisionRadius=0
      boundBone {boneName=dyn_pant_leg_front_01}
      boxExtents=<18 18 18>
      constraintSetup {angularLimitsMax=<10 5 10>, angularLimitsMin=<-10 -30 -10>, angularTarget=<1 0 0>}
      localJointOffset=<2 0 0>
     }]
    relativeSpaceBone {boneName=root}
   }
   n138 = Anim Dynamics - Bone: dyn_pant_leg_front_02!(ComponentPose=@n137, Alpha=1, GravityOverride:Vector=@n124, ExternalForce="0,0,0") -> (Pose) {
    angularDampingOverride=0.9
    angularSpringConstant=1000
    bGravityOverrideInSimSpace=true
    bOverrideAngularDamping=true
    bUseGravityOverride=true
    simulationSpace=BoneRelative
    alphaBoolBlend {blendOption=Linear}
    boundBone {boneName=dyn_pant_leg_front_02}
    componentAppliedLinearAccClamp=<1500 1500 1500>
    componentLinearAccScale=<0.4 0.4 0.4>
    componentLinearVelScale=<0.4 0.4 0.4>
    gravityOverride {z=-980}
    physicsBodyDefinitions [{
      sphereCollisionRadius=0
      boundBone {boneName=dyn_pant_leg_front_02}
      boxExtents=<18 18 18>
      constraintSetup {angularLimitsMax=<10 5 10>, angularLimitsMin=<-10 -30 -10>, angularTarget=<1 0 0>}
      localJointOffset=<2 0 0>
     }]
    relativeSpaceBone {boneName=root}
   }
   n139 = Anim Dynamics - Bone: dyn_pant_leg_front_03!(ComponentPose=@n138, Alpha=1, GravityOverride:Vector=@n124, ExternalForce="0,0,0") -> (Pose) {
    angularDampingOverride=0.9
    angularSpringConstant=1000
    bGravityOverrideInSimSpace=true
    bOverrideAngularDamping=true
    bUseGravityOverride=true
    simulationSpace=BoneRelative
    alphaBoolBlend {blendOption=Linear}
    boundBone {boneName=dyn_pant_leg_front_03}
    componentAppliedLinearAccClamp=<1500 1500 1500>
    componentLinearAccScale=<0.4 0.4 0.4>
    componentLinearVelScale=<0.4 0.4 0.4>
    gravityOverride {z=-980}
    physicsBodyDefinitions [{
      sphereCollisionRadius=0
      boundBone {boneName=dyn_pant_leg_front_03}
      boxExtents=<18 18 18>
      constraintSetup {angularLimitsMax=<10 5 10>, angularLimitsMin=<-10 -30 -10>, angularTarget=<1 0 0>}
      localJointOffset=<2 0 0>
     }]
    relativeSpaceBone {boneName=root}
   }
   n140 = Anim Dynamics - Bone: dyn_l_arm_ribbon_01(ComponentPose=@n139, Alpha=1, ExternalForce="0,0,0") {
    angularDampingOverride=0.7
    angularSpringConstant=1000
    bGravityOverrideInSimSpace=true
    bOverrideAngularDamping=true
    bUseGravityOverride=true
    simulationSpace=BoneRelative
    alphaBoolBlend {blendOption=Linear}
    boundBone {boneName=dyn_l_arm_ribbon_01}
    componentAppliedLinearAccClamp=<1500 1500 1500>
    componentLinearAccScale=<0.4 0.4 0.4>
    componentLinearVelScale=<0.4 0.4 0.4>
    gravityOverride {z=-980}
    physicsBodyDefinitions [{
      sphereCollisionRadius=0
      boundBone {boneName=dyn_l_arm_ribbon_01}
      boxExtents=<15 15 15>
      constraintSetup {angularLimitsMax=<5 5 5>, angularLimitsMin=<-5 -5 -5>, angularTarget=<1 0 0>}
      localJointOffset=<2.5 0 0>
     }]
    relativeSpaceBone {boneName=root}
   }
   n141 = Anim Dynamics - Bone: dyn_l_arm_ribbon_02(ComponentPose=@n140, Alpha=1, ExternalForce="0,0,0") {
    angularDampingOverride=0.7
    angularSpringConstant=1000
    bGravityOverrideInSimSpace=true
    bOverrideAngularDamping=true
    bUseGravityOverride=true
    simulationSpace=BoneRelative
    alphaBoolBlend {blendOption=Linear}
    boundBone {boneName=dyn_l_arm_ribbon_02}
    componentAppliedLinearAccClamp=<1500 1500 1500>
    componentLinearAccScale=<0.4 0.4 0.4>
    componentLinearVelScale=<0.4 0.4 0.4>
    gravityOverride {z=-980}
    physicsBodyDefinitions [{
      sphereCollisionRadius=0
      boundBone {boneName=dyn_l_arm_ribbon_02}
      boxExtents=<15 15 15>
      constraintSetup {angularLimitsMax=<5 5 5>, angularLimitsMin=<-5 -5 -5>, angularTarget=<1 0 0>}
      localJointOffset=<2.5 0 0>
     }]
    relativeSpaceBone {boneName=root}
   }
   n142 = Anim Dynamics - Bone: dyn_l_arm_ribbon_03(ComponentPose=@n141, Alpha=1, ExternalForce="0,0,0") {
    angularDampingOverride=0.7
    angularSpringConstant=1000
    bGravityOverrideInSimSpace=true
    bOverrideAngularDamping=true
    bUseGravityOverride=true
    simulationSpace=BoneRelative
    alphaBoolBlend {blendOption=Linear}
    boundBone {boneName=dyn_l_arm_ribbon_03}
    componentAppliedLinearAccClamp=<1500 1500 1500>
    componentLinearAccScale=<0.4 0.4 0.4>
    componentLinearVelScale=<0.4 0.4 0.4>
    gravityOverride {z=-980}
    physicsBodyDefinitions [{
      sphereCollisionRadius=0
      boundBone {boneName=dyn_l_arm_ribbon_03}
      boxExtents=<15 15 15>
      constraintSetup {angularLimitsMax=<5 5 5>, angularLimitsMin=<-5 -5 -5>, angularTarget=<1 0 0>}
      localJointOffset=<2.5 0 0>
     }]
    relativeSpaceBone {boneName=root}
   }
   n143 = RigidBody(ComponentPose=@n142, OverrideWorldGravity=Get GravityOverride_tessel()) {
    bClampLinearTranslationLimitToRefPose=true
    bOverrideWorldGravity=true
    overridePhysicsAsset=&/M_MED_Cube_Ninja_CloudFloat_Physics
    simulationSpace=BaseBoneSpace
    alphaBoolBlend {blendOption=Linear}
    baseBoneRef {boneName=root}
    componentAppliedLinearAccClamp=<5000 5000 5000>
    componentLinearAccScale=<0.8 0.8 0.9>
    componentLinearVelScale=<0.8 0.8 0.9>
    overrideWorldGravity {z=-1100}
   }
   n144 = Control Rig(Source=Copy Pose From Mesh() {bUseAttachedParent=true}) {
    alphaBoolBlend {blendOption=Linear}
    controlRigAssetReference {blueprintRigClass=&/M_MED_Cube_Ninja_CtrlRig.M_MED_Cube_Ninja_CtrlRig_C}
   }
   n145 = "Constraint - dyn_sash_front_master (PTT)"(ComponentPose=Local To Component(LocalPose=@n144), Alpha=1, ConstraintWeights_0=1, ConstraintWeights_1=0, ConstraintWeights_2=0) {
    alphaBoolBlend {blendOption=Linear}
    boneToModify {boneName=dyn_sash_front_master}
    constraintSetup [
     {targetBone {boneName=dyn_scrolls_master}}
     {transformType=Translation, targetBone {boneName=None}}
     {transformType=Translation, targetBone {boneName=None}}
    ]
    constraintWeights [1,1,1]
   }
   n146 = "Constraint - dyn_sash_back_master (PTT)"(ComponentPose=@n145, Alpha=1, ConstraintWeights_0=1, ConstraintWeights_1=0, ConstraintWeights_2=0) {
    alphaBoolBlend {blendOption=Linear}
    boneToModify {boneName=dyn_sash_back_master}
    constraintSetup [
     {targetBone {boneName=dyn_scrolls_master}}
     {transformType=Translation, targetBone {boneName=None}}
     {transformType=Translation, targetBone {boneName=None}}
    ]
    constraintWeights [1,1,1]
   }
   n147 = Anim Dynamics - Bone: dyn_l_belt_01!(ComponentPose=@n146, Alpha=1, ExternalForce="0,0,0") -> (Pose) {
    angularDampingOverride=0.7
    angularSpringConstant=1000
    bGravityOverrideInSimSpace=true
    bUseGravityOverride=true
    simulationSpace=BoneRelative
    alphaBoolBlend {blendOption=Linear}
    boundBone {boneName=dyn_l_belt_01}
    componentAppliedLinearAccClamp=<1500 1500 1500>
    componentLinearAccScale=<0.4 0.4 0.4>
    componentLinearVelScale=<0.4 0.4 0.4>
    gravityOverride {z=-980}
    physicsBodyDefinitions [{
      sphereCollisionRadius=0
      boundBone {boneName=dyn_l_belt_01}
      boxExtents=<20 20 20>
      constraintSetup {angularLimitsMax=<3 -5 0>, angularLimitsMin=<-3 -12 0>, angularTarget=<1 0 0>}
      localJointOffset=<-3 0 0>
     }]
    relativeSpaceBone {boneName=root}
   }
   n148 = Anim Dynamics - Bone: dyn_l_belt_02!(ComponentPose=@n147, Alpha=1, ExternalForce="0,0,0") -> (Pose) {
    angularDampingOverride=0.7
    angularSpringConstant=1000
    bGravityOverrideInSimSpace=true
    bUseGravityOverride=true
    simulationSpace=BoneRelative
    alphaBoolBlend {blendOption=Linear}
    boundBone {boneName=dyn_l_belt_02}
    componentAppliedLinearAccClamp=<1500 1500 1500>
    componentLinearAccScale=<0.4 0.4 0.4>
    componentLinearVelScale=<0.4 0.4 0.4>
    gravityOverride {z=-980}
    physicsBodyDefinitions [{
      sphereCollisionRadius=0
      boundBone {boneName=dyn_l_belt_02}
      boxExtents=<20 20 20>
      constraintSetup {angularLimitsMax=<3 -5 0>, angularLimitsMin=<-3 -12 0>, angularTarget=<1 0 0>}
      localJointOffset=<-3 0 0>
     }]
    relativeSpaceBone {boneName=root}
   }
   n149 = Anim Dynamics - Bone: dyn_l_belt_03!(ComponentPose=@n148, Alpha=1, ExternalForce="0,0,0") -> (Pose) {
    angularDampingOverride=0.7
    angularSpringConstant=1000
    bGravityOverrideInSimSpace=true
    bUseGravityOverride=true
    simulationSpace=BoneRelative
    alphaBoolBlend {blendOption=Linear}
    boundBone {boneName=dyn_l_belt_03}
    componentAppliedLinearAccClamp=<1500 1500 1500>
    componentLinearAccScale=<0.4 0.4 0.4>
    componentLinearVelScale=<0.4 0.4 0.4>
    gravityOverride {z=-980}
    physicsBodyDefinitions [{
      sphereCollisionRadius=0
      boundBone {boneName=dyn_l_belt_03}
      boxExtents=<20 20 20>
      constraintSetup {angularLimitsMax=<3 5 3>, angularLimitsMin=<-3 0 -3>, angularTarget=<1 0 0>}
      localJointOffset=<-3 0 0>
     }]
    relativeSpaceBone {boneName=root}
   }
   n150 = Anim Dynamics - Bone: dyn_mask_01!(ComponentPose=@n149, Alpha=1, ExternalForce="0,0,0") -> (Pose) {
    angularDampingOverride=0.7
    angularSpringConstant=1000
    bGravityOverrideInSimSpace=true
    bUseGravityOverride=true
    simulationSpace=BoneRelative
    alphaBoolBlend {blendOption=Linear}
    boundBone {boneName=dyn_mask_01}
    componentAppliedLinearAccClamp=<1500 1500 1500>
    componentLinearAccScale=<0.4 0.4 0.4>
    componentLinearVelScale=<0.4 0.4 0.4>
    gravityOverride {z=-980}
    physicsBodyDefinitions [{
      sphereCollisionRadius=0
      boundBone {boneName=dyn_mask_01}
      boxExtents=<20 20 20>
      constraintSetup {angularLimitsMax=<3 15 3>, angularLimitsMin=<-3 0 -3>, angularTarget=<1 0 0>}
      localJointOffset=<5 0 0>
     }]
    relativeSpaceBone {boneName=root}
   }
   n151 = Anim Dynamics - Bone: dyn_r_belt_01!(ComponentPose=@n150, Alpha=1, ExternalForce="0,0,0") -> (Pose) {
    angularDampingOverride=0.7
    angularSpringConstant=1000
    bGravityOverrideInSimSpace=true
    bUseGravityOverride=true
    simulationSpace=BoneRelative
    alphaBoolBlend {blendOption=Linear}
    boundBone {boneName=dyn_r_belt_01}
    componentAppliedLinearAccClamp=<1500 1500 1500>
    componentLinearAccScale=<0.4 0.4 0.4>
    componentLinearVelScale=<0.4 0.4 0.4>
    gravityOverride {z=-980}
    physicsBodyDefinitions [{
      sphereCollisionRadius=0
      boundBone {boneName=dyn_r_belt_01}
      boxExtents=<20 20 20>
      constraintSetup {angularLimitsMax=<3 -5 0>, angularLimitsMin=<-3 -12 0>, angularTarget=<1 -5 0>}
      localJointOffset=<3 0 0>
     }]
    relativeSpaceBone {boneName=root}
   }
   n152 = Anim Dynamics - Bone: dyn_r_belt_02!(ComponentPose=@n151, Alpha=1, ExternalForce="0,0,0") -> (Pose) {
    angularDampingOverride=0.7
    angularSpringConstant=1000
    bGravityOverrideInSimSpace=true
    bUseGravityOverride=true
    simulationSpace=BoneRelative
    alphaBoolBlend {blendOption=Linear}
    boundBone {boneName=dyn_r_belt_02}
    componentAppliedLinearAccClamp=<1500 1500 1500>
    componentLinearAccScale=<0.4 0.4 0.4>
    componentLinearVelScale=<0.4 0.4 0.4>
    gravityOverride {z=-980}
    physicsBodyDefinitions [{
      sphereCollisionRadius=0
      boundBone {boneName=dyn_r_belt_02}
      boxExtents=<20 20 20>
      constraintSetup {angularLimitsMax=<3 -5 0>, angularLimitsMin=<-3 -12 0>, angularTarget=<1 -5 0>}
      localJointOffset=<3 0 0>
     }]
    relativeSpaceBone {boneName=root}
   }
   n153 = Anim Dynamics - Bone: dyn_r_belt_03!(ComponentPose=@n152, Alpha=1, ExternalForce="0,0,0") -> (Pose) {
    angularDampingOverride=0.7
    angularSpringConstant=1000
    bGravityOverrideInSimSpace=true
    bUseGravityOverride=true
    simulationSpace=BoneRelative
    alphaBoolBlend {blendOption=Linear}
    boundBone {boneName=dyn_r_belt_03}
    componentAppliedLinearAccClamp=<1500 1500 1500>
    componentLinearAccScale=<0.4 0.4 0.4>
    componentLinearVelScale=<0.4 0.4 0.4>
    gravityOverride {z=-980}
    physicsBodyDefinitions [{
      sphereCollisionRadius=0
      boundBone {boneName=dyn_r_belt_03}
      boxExtents=<20 20 20>
      constraintSetup {angularLimitsMax=<3 -5 3>, angularLimitsMin=<-3 -12 -3>, angularTarget=<1 -5 0>}
      localJointOffset=<3 0 0>
     }]
    relativeSpaceBone {boneName=root}
   }
   n154 = "Constraint - dyn_scrolls_master (TTT)"(ComponentPose=@n153, Alpha=1, ConstraintWeights_0=1, ConstraintWeights_1=0, ConstraintWeights_2=0) {
    alphaBoolBlend {blendOption=Linear}
    boneToModify {boneName=dyn_scrolls_master}
    constraintSetup [
     {transformType=Translation, targetBone {boneName=dyn_r_belt_03}}
     {transformType=Translation, targetBone {boneName=None}}
     {transformType=Translation, targetBone {boneName=None}}
    ]
    constraintWeights [1,1,1]
   }
   n155 = Anim Dynamics - Bone: dyn_leg_ribbon_top_01!(ComponentPose=@n154, Alpha=1, ExternalForce="0,0,0") -> (Pose) {
    angularDampingOverride=0.7
    angularSpringConstant=1000
    bGravityOverrideInSimSpace=true
    bUseGravityOverride=true
    simulationSpace=BoneRelative
    alphaBoolBlend {blendOption=Linear}
    boundBone {boneName=dyn_leg_ribbon_top_01}
    componentAppliedLinearAccClamp=<1500 1500 1500>
    componentLinearAccScale=<0.4 0.4 0.4>
    componentLinearVelScale=<0.4 0.4 0.4>
    gravityOverride {z=-980}
    physicsBodyDefinitions [{
      sphereCollisionRadius=0
      boundBone {boneName=dyn_leg_ribbon_top_01}
      boxExtents=<15 15 15>
      constraintSetup {angularLimitsMax=<5 5 5>, angularLimitsMin=<-5 -5 -5>, angularTarget=<1 0 0>}
      localJointOffset=<3 0 0>
     }]
    relativeSpaceBone {boneName=root}
   }
   n156 = Anim Dynamics - Bone: dyn_leg_ribbon_top_02!(ComponentPose=@n155, Alpha=1, ExternalForce="0,0,0") -> (Pose) {
    angularDampingOverride=0.7
    angularSpringConstant=1000
    bGravityOverrideInSimSpace=true
    bUseGravityOverride=true
    simulationSpace=BoneRelative
    alphaBoolBlend {blendOption=Linear}
    boundBone {boneName=dyn_leg_ribbon_top_02}
    componentAppliedLinearAccClamp=<1500 1500 1500>
    componentLinearAccScale=<0.4 0.4 0.4>
    componentLinearVelScale=<0.4 0.4 0.4>
    gravityOverride {z=-980}
    physicsBodyDefinitions [{
      sphereCollisionRadius=0
      boundBone {boneName=dyn_leg_ribbon_top_02}
      boxExtents=<15 15 15>
      constraintSetup {angularLimitsMax=<5 5 5>, angularLimitsMin=<-5 -5 -5>, angularTarget=<1 0 0>}
      localJointOffset=<3 0 0>
     }]
    relativeSpaceBone {boneName=root}
   }
   n157 = Anim Dynamics - Bone: dyn_leg_ribbon_top_03!(ComponentPose=@n156, Alpha=1, ExternalForce="0,0,0") -> (Pose) {
    angularDampingOverride=0.7
    angularSpringConstant=1000
    bGravityOverrideInSimSpace=true
    bUseGravityOverride=true
    simulationSpace=BoneRelative
    alphaBoolBlend {blendOption=Linear}
    boundBone {boneName=dyn_leg_ribbon_top_03}
    componentAppliedLinearAccClamp=<1500 1500 1500>
    componentLinearAccScale=<0.4 0.4 0.4>
    componentLinearVelScale=<0.4 0.4 0.4>
    gravityOverride {z=-980}
    physicsBodyDefinitions [{
      sphereCollisionRadius=0
      boundBone {boneName=dyn_leg_ribbon_top_03}
      boxExtents=<15 15 15>
      constraintSetup {angularLimitsMax=<5 5 5>, angularLimitsMin=<-5 -5 -5>, angularTarget=<1 0 0>}
      localJointOffset=<3 0 0>
     }]
    relativeSpaceBone {boneName=root}
   }
   n158 = Anim Dynamics - Bone: dyn_leg_ribbon_01!(ComponentPose=@n157, Alpha=1, ExternalForce="0,0,0") -> (Pose) {
    angularDampingOverride=0.7
    angularSpringConstant=1000
    bGravityOverrideInSimSpace=true
    bUseGravityOverride=true
    simulationSpace=BoneRelative
    alphaBoolBlend {blendOption=Linear}
    boundBone {boneName=dyn_leg_ribbon_01}
    componentAppliedLinearAccClamp=<1500 1500 1500>
    componentLinearAccScale=<0.4 0.4 0.4>
    componentLinearVelScale=<0.4 0.4 0.4>
    gravityOverride {z=-980}
    physicsBodyDefinitions [{
      sphereCollisionRadius=0
      boundBone {boneName=dyn_leg_ribbon_01}
      boxExtents=<20 20 20>
      constraintSetup {angularLimitsMax=<5 5 5>, angularLimitsMin=<-5 -5 -5>, angularTarget=<1 0 0>}
      localJointOffset=<3 0 0>
     }]
    relativeSpaceBone {boneName=root}
   }
   n159 = Anim Dynamics - Bone: dyn_leg_ribbon_02!(ComponentPose=@n158, Alpha=1, ExternalForce="0,0,0") -> (Pose) {
    angularDampingOverride=0.7
    angularSpringConstant=1000
    bGravityOverrideInSimSpace=true
    bUseGravityOverride=true
    simulationSpace=BoneRelative
    alphaBoolBlend {blendOption=Linear}
    boundBone {boneName=dyn_leg_ribbon_02}
    componentAppliedLinearAccClamp=<1500 1500 1500>
    componentLinearAccScale=<0.4 0.4 0.4>
    componentLinearVelScale=<0.4 0.4 0.4>
    gravityOverride {z=-980}
    physicsBodyDefinitions [{
      sphereCollisionRadius=0
      boundBone {boneName=dyn_leg_ribbon_02}
      boxExtents=<20 20 20>
      constraintSetup {angularLimitsMax=<5 5 5>, angularLimitsMin=<-5 -5 -5>, angularTarget=<1 0 0>}
      localJointOffset=<3 0 0>
     }]
    relativeSpaceBone {boneName=root}
   }
   n160 = Anim Dynamics - Bone: dyn_leg_ribbon_03!(ComponentPose=@n159, Alpha=1, ExternalForce="0,0,0") -> (Pose) {
    angularDampingOverride=0.7
    angularSpringConstant=1000
    bGravityOverrideInSimSpace=true
    bUseGravityOverride=true
    simulationSpace=BoneRelative
    alphaBoolBlend {blendOption=Linear}
    boundBone {boneName=dyn_leg_ribbon_03}
    componentAppliedLinearAccClamp=<1500 1500 1500>
    componentLinearAccScale=<0.4 0.4 0.4>
    componentLinearVelScale=<0.4 0.4 0.4>
    gravityOverride {z=-980}
    physicsBodyDefinitions [{
      sphereCollisionRadius=0
      boundBone {boneName=dyn_leg_ribbon_03}
      boxExtents=<20 20 20>
      constraintSetup {angularLimitsMax=<5 5 5>, angularLimitsMin=<-5 -5 -5>, angularTarget=<1 0 0>}
      localJointOffset=<3 0 0>
     }]
    relativeSpaceBone {boneName=root}
   }
   n161 = Anim Dynamics - Bone: dyn_crystal_01!(ComponentPose=@n160, Alpha=1, ExternalForce="0,0,0") -> (Pose) {
    angularDampingOverride=0.9
    angularSpringConstant=1000
    bGravityOverrideInSimSpace=true
    bOverrideAngularDamping=true
    bUseGravityOverride=true
    alphaBoolBlend {blendOption=Linear}
    boundBone {boneName=dyn_crystal_01}
    componentAppliedLinearAccClamp=<1500 1500 1500>
    componentLinearAccScale=<0.4 0.4 0.4>
    componentLinearVelScale=<0.4 0.4 0.4>
    gravityOverride {z=-980}
    physicsBodyDefinitions [{
      sphereCollisionRadius=0
      boundBone {boneName=dyn_crystal_01}
      boxExtents=<10 10 10>
      constraintSetup {angularLimitsMax=<5 5 5>, angularLimitsMin=<-5 -5 -5>, angularTarget=<1 0 0>}
      localJointOffset=<3 0 0>
     }]
    relativeSpaceBone {boneName=root}
   }
   n162 = Anim Dynamics - Bone: dyn_crystal_02!(ComponentPose=@n161, Alpha=1, ExternalForce="0,0,0") -> (Pose) {
    angularDampingOverride=0.9
    angularSpringConstant=1000
    bGravityOverrideInSimSpace=true
    bOverrideAngularDamping=true
    bUseGravityOverride=true
    alphaBoolBlend {blendOption=Linear}
    boundBone {boneName=dyn_crystal_02}
    componentAppliedLinearAccClamp=<1500 1500 1500>
    componentLinearAccScale=<0.4 0.4 0.4>
    componentLinearVelScale=<0.4 0.4 0.4>
    gravityOverride {z=-980}
    physicsBodyDefinitions [{
      sphereCollisionRadius=0
      boundBone {boneName=dyn_crystal_02}
      boxExtents=<12 12 12>
      constraintSetup {angularLimitsMax=<2 10 10>, angularLimitsMin=<-2 -10 -10>, angularTarget=<1 0 0>}
      localJointOffset=<3 0 0>
     }]
    relativeSpaceBone {boneName=root}
   }
   n163 = Anim Dynamics - Bone: dyn_crystal_03!(ComponentPose=@n162, Alpha=1, ExternalForce="0,0,0") -> (Pose) {
    angularDampingOverride=0.9
    angularSpringConstant=1000
    bGravityOverrideInSimSpace=true
    bOverrideAngularDamping=true
    bUseGravityOverride=true
    alphaBoolBlend {blendOption=Linear}
    boundBone {boneName=dyn_crystal_03}
    componentAppliedLinearAccClamp=<1500 1500 1500>
    componentLinearAccScale=<0.4 0.4 0.4>
    componentLinearVelScale=<0.4 0.4 0.4>
    gravityOverride {z=-980}
    physicsBodyDefinitions [{
      sphereCollisionRadius=0
      boundBone {boneName=dyn_crystal_03}
      boxExtents=<12 12 12>
      constraintSetup {angularLimitsMax=<40 5 40>, angularLimitsMin=<-40 -40 -40>, angularTarget=<1 0 0>}
      localJointOffset=<3 0 0>
     }]
    relativeSpaceBone {boneName=root}
   }
   n164 = Anim Dynamics - Bone: dyn_pant_leg_back_r_01!(ComponentPose=@n163, Alpha=1, ExternalForce="0,0,0") -> (Pose) {
    angularDampingOverride=0.9
    angularSpringConstant=1000
    bGravityOverrideInSimSpace=true
    bOverrideAngularDamping=true
    bUseGravityOverride=true
    simulationSpace=BoneRelative
    alphaBoolBlend {blendOption=Linear}
    boundBone {boneName=dyn_pant_leg_back_r_01}
    componentAppliedLinearAccClamp=<1500 1500 1500>
    componentLinearAccScale=<0.4 0.4 0.4>
    componentLinearVelScale=<0.4 0.4 0.4>
    gravityOverride {z=-980}
    physicsBodyDefinitions [{
      sphereCollisionRadius=0
      boundBone {boneName=dyn_pant_leg_back_r_01}
      boxExtents=<18 18 18>
      constraintSetup {angularLimitsMax=<10 5 10>, angularLimitsMin=<-10 -30 -10>, angularTarget=<1 0 0>}
      localJointOffset=<2 0 0>
     }]
    relativeSpaceBone {boneName=root}
   }
   n165 = Anim Dynamics - Bone: dyn_pant_leg_back_r_02!(ComponentPose=@n164, Alpha=1, ExternalForce="0,0,0") -> (Pose) {
    angularDampingOverride=0.9
    angularSpringConstant=1000
    bGravityOverrideInSimSpace=true
    bOverrideAngularDamping=true
    bUseGravityOverride=true
    simulationSpace=BoneRelative
    alphaBoolBlend {blendOption=Linear}
    boundBone {boneName=dyn_pant_leg_back_r_02}
    componentAppliedLinearAccClamp=<1500 1500 1500>
    componentLinearAccScale=<0.4 0.4 0.4>
    componentLinearVelScale=<0.4 0.4 0.4>
    gravityOverride {z=-980}
    physicsBodyDefinitions [{
      sphereCollisionRadius=0
      boundBone {boneName=dyn_pant_leg_back_r_02}
      boxExtents=<18 18 18>
      constraintSetup {angularLimitsMax=<10 5 10>, angularLimitsMin=<-10 -30 -10>, angularTarget=<1 0 0>}
      localJointOffset=<2 0 0>
     }]
    relativeSpaceBone {boneName=root}
   }
   n166 = Anim Dynamics - Bone: dyn_pant_leg_back_r_03!(ComponentPose=@n165, Alpha=1, ExternalForce="0,0,0") -> (Pose) {
    angularDampingOverride=0.9
    angularSpringConstant=1000
    bGravityOverrideInSimSpace=true
    bOverrideAngularDamping=true
    bUseGravityOverride=true
    simulationSpace=BoneRelative
    alphaBoolBlend {blendOption=Linear}
    boundBone {boneName=dyn_pant_leg_back_r_03}
    componentAppliedLinearAccClamp=<1500 1500 1500>
    componentLinearAccScale=<0.4 0.4 0.4>
    componentLinearVelScale=<0.4 0.4 0.4>
    gravityOverride {z=-980}
    physicsBodyDefinitions [{
      sphereCollisionRadius=0
      boundBone {boneName=dyn_pant_leg_back_r_03}
      boxExtents=<18 18 18>
      constraintSetup {angularLimitsMax=<10 5 10>, angularLimitsMin=<-10 -30 -10>, angularTarget=<1 0 0>}
      localJointOffset=<2 0 0>
     }]
    relativeSpaceBone {boneName=root}
   }
   n167 = Anim Dynamics - Bone: dyn_pant_leg_back_l_01!(ComponentPose=@n166, Alpha=1, ExternalForce="0,0,0") -> (Pose) {
    angularDampingOverride=0.9
    angularSpringConstant=1000
    bGravityOverrideInSimSpace=true
    bOverrideAngularDamping=true
    bUseGravityOverride=true
    simulationSpace=BoneRelative
    alphaBoolBlend {blendOption=Linear}
    boundBone {boneName=dyn_pant_leg_back_l_01}
    componentAppliedLinearAccClamp=<1500 1500 1500>
    componentLinearAccScale=<0.4 0.4 0.4>
    componentLinearVelScale=<0.4 0.4 0.4>
    gravityOverride {z=-980}
    physicsBodyDefinitions [{
      sphereCollisionRadius=0
      boundBone {boneName=dyn_pant_leg_back_l_01}
      boxExtents=<18 18 18>
      constraintSetup {angularLimitsMax=<10 5 10>, angularLimitsMin=<-10 -30 -10>, angularTarget=<1 0 0>}
      localJointOffset=<2 0 0>
     }]
    relativeSpaceBone {boneName=root}
   }
   n168 = Anim Dynamics - Bone: dyn_pant_leg_back_l_02!(ComponentPose=@n167, Alpha=1, ExternalForce="0,0,0") -> (Pose) {
    angularDampingOverride=0.9
    angularSpringConstant=1000
    bGravityOverrideInSimSpace=true
    bOverrideAngularDamping=true
    bUseGravityOverride=true
    simulationSpace=BoneRelative
    alphaBoolBlend {blendOption=Linear}
    boundBone {boneName=dyn_pant_leg_back_l_02}
    componentAppliedLinearAccClamp=<1500 1500 1500>
    componentLinearAccScale=<0.4 0.4 0.4>
    componentLinearVelScale=<0.4 0.4 0.4>
    gravityOverride {z=-980}
    physicsBodyDefinitions [{
      sphereCollisionRadius=0
      boundBone {boneName=dyn_pant_leg_back_l_02}
      boxExtents=<18 18 18>
      constraintSetup {angularLimitsMax=<10 5 10>, angularLimitsMin=<-10 -30 -10>, angularTarget=<1 0 0>}
      localJointOffset=<2 0 0>
     }]
    relativeSpaceBone {boneName=root}
   }
   n169 = Anim Dynamics - Bone: dyn_pant_leg_back_l_03!(ComponentPose=@n168, Alpha=1, ExternalForce="0,0,0") -> (Pose) {
    angularDampingOverride=0.9
    angularSpringConstant=1000
    bGravityOverrideInSimSpace=true
    bOverrideAngularDamping=true
    bUseGravityOverride=true
    simulationSpace=BoneRelative
    alphaBoolBlend {blendOption=Linear}
    boundBone {boneName=dyn_pant_leg_back_l_03}
    componentAppliedLinearAccClamp=<1500 1500 1500>
    componentLinearAccScale=<0.4 0.4 0.4>
    componentLinearVelScale=<0.4 0.4 0.4>
    gravityOverride {z=-980}
    physicsBodyDefinitions [{
      sphereCollisionRadius=0
      boundBone {boneName=dyn_pant_leg_back_l_03}
      boxExtents=<18 18 18>
      constraintSetup {angularLimitsMax=<10 5 10>, angularLimitsMin=<-10 -30 -10>, angularTarget=<1 0 0>}
      localJointOffset=<2 0 0>
     }]
    relativeSpaceBone {boneName=root}
   }
   n170 = Anim Dynamics - Bone: dyn_pant_leg_left_01!(ComponentPose=@n169, Alpha=1, ExternalForce="0,0,0") -> (Pose) {
    angularDampingOverride=0.9
    angularSpringConstant=1000
    bGravityOverrideInSimSpace=true
    bOverrideAngularDamping=true
    bUseGravityOverride=true
    simulationSpace=BoneRelative
    alphaBoolBlend {blendOption=Linear}
    boundBone {boneName=dyn_pant_leg_left_01}
    componentAppliedLinearAccClamp=<1500 1500 1500>
    componentLinearAccScale=<0.4 0.4 0.4>
    componentLinearVelScale=<0.4 0.4 0.4>
    gravityOverride {z=-980}
    physicsBodyDefinitions [{
      sphereCollisionRadius=0
      boundBone {boneName=dyn_pant_leg_left_01}
      boxExtents=<18 18 18>
      constraintSetup {angularLimitsMax=<10 5 10>, angularLimitsMin=<-10 -30 -10>, angularTarget=<1 0 0>}
      localJointOffset=<2 0 0>
     }]
    relativeSpaceBone {boneName=root}
   }
   n171 = Anim Dynamics - Bone: dyn_pant_leg_left_02!(ComponentPose=@n170, Alpha=1, ExternalForce="0,0,0") -> (Pose) {
    angularDampingOverride=0.9
    angularSpringConstant=1000
    bGravityOverrideInSimSpace=true
    bOverrideAngularDamping=true
    bUseGravityOverride=true
    simulationSpace=BoneRelative
    alphaBoolBlend {blendOption=Linear}
    boundBone {boneName=dyn_pant_leg_left_02}
    componentAppliedLinearAccClamp=<1500 1500 1500>
    componentLinearAccScale=<0.4 0.4 0.4>
    componentLinearVelScale=<0.4 0.4 0.4>
    gravityOverride {z=-980}
    physicsBodyDefinitions [{
      sphereCollisionRadius=0
      boundBone {boneName=dyn_pant_leg_left_02}
      boxExtents=<18 18 18>
      constraintSetup {angularLimitsMax=<10 5 10>, angularLimitsMin=<-10 -30 -10>, angularTarget=<1 0 0>}
      localJointOffset=<2 0 0>
     }]
    relativeSpaceBone {boneName=root}
   }
   n172 = Anim Dynamics - Bone: dyn_pant_leg_left_03!(ComponentPose=@n171, Alpha=1, ExternalForce="0,0,0") -> (Pose) {
    angularDampingOverride=0.9
    angularSpringConstant=1000
    bGravityOverrideInSimSpace=true
    bOverrideAngularDamping=true
    bUseGravityOverride=true
    simulationSpace=BoneRelative
    alphaBoolBlend {blendOption=Linear}
    boundBone {boneName=dyn_pant_leg_left_03}
    componentAppliedLinearAccClamp=<1500 1500 1500>
    componentLinearAccScale=<0.4 0.4 0.4>
    componentLinearVelScale=<0.4 0.4 0.4>
    gravityOverride {z=-980}
    physicsBodyDefinitions [{
      sphereCollisionRadius=0
      boundBone {boneName=dyn_pant_leg_left_03}
      boxExtents=<18 18 18>
      constraintSetup {angularLimitsMax=<10 5 10>, angularLimitsMin=<-10 -30 -10>, angularTarget=<1 0 0>}
      localJointOffset=<2 0 0>
     }]
    relativeSpaceBone {boneName=root}
   }
   n173 = Anim Dynamics - Bone: dyn_pant_leg_right_01!(ComponentPose=@n172, Alpha=1, ExternalForce="0,0,0") -> (Pose) {
    angularDampingOverride=0.9
    angularSpringConstant=1000
    bGravityOverrideInSimSpace=true
    bOverrideAngularDamping=true
    bUseGravityOverride=true
    simulationSpace=BoneRelative
    alphaBoolBlend {blendOption=Linear}
    boundBone {boneName=dyn_pant_leg_right_01}
    componentAppliedLinearAccClamp=<1500 1500 1500>
    componentLinearAccScale=<0.4 0.4 0.4>
    componentLinearVelScale=<0.4 0.4 0.4>
    gravityOverride {z=-980}
    physicsBodyDefinitions [{
      sphereCollisionRadius=0
      boundBone {boneName=dyn_pant_leg_right_01}
      boxExtents=<18 18 18>
      constraintSetup {angularLimitsMax=<10 5 10>, angularLimitsMin=<-10 -30 -10>, angularTarget=<1 0 0>}
      localJointOffset=<2 0 0>
     }]
    relativeSpaceBone {boneName=root}
   }
   n174 = Anim Dynamics - Bone: dyn_pant_leg_right_02!(ComponentPose=@n173, Alpha=1, ExternalForce="0,0,0") -> (Pose) {
    angularDampingOverride=0.9
    angularSpringConstant=1000
    bGravityOverrideInSimSpace=true
    bOverrideAngularDamping=true
    bUseGravityOverride=true
    simulationSpace=BoneRelative
    alphaBoolBlend {blendOption=Linear}
    boundBone {boneName=dyn_pant_leg_right_02}
    componentAppliedLinearAccClamp=<1500 1500 1500>
    componentLinearAccScale=<0.4 0.4 0.4>
    componentLinearVelScale=<0.4 0.4 0.4>
    gravityOverride {z=-980}
    physicsBodyDefinitions [{
      sphereCollisionRadius=0
      boundBone {boneName=dyn_pant_leg_right_02}
      boxExtents=<18 18 18>
      constraintSetup {angularLimitsMax=<10 5 10>, angularLimitsMin=<-10 -30 -10>, angularTarget=<1 0 0>}
      localJointOffset=<2 0 0>
     }]
    relativeSpaceBone {boneName=root}
   }
   n175 = Anim Dynamics - Bone: dyn_pant_leg_right_03!(ComponentPose=@n174, Alpha=1, ExternalForce="0,0,0") -> (Pose) {
    angularDampingOverride=0.9
    angularSpringConstant=1000
    bGravityOverrideInSimSpace=true
    bOverrideAngularDamping=true
    bUseGravityOverride=true
    simulationSpace=BoneRelative
    alphaBoolBlend {blendOption=Linear}
    boundBone {boneName=dyn_pant_leg_right_03}
    componentAppliedLinearAccClamp=<1500 1500 1500>
    componentLinearAccScale=<0.4 0.4 0.4>
    componentLinearVelScale=<0.4 0.4 0.4>
    gravityOverride {z=-980}
    physicsBodyDefinitions [{
      sphereCollisionRadius=0
      boundBone {boneName=dyn_pant_leg_right_03}
      boxExtents=<18 18 18>
      constraintSetup {angularLimitsMax=<10 5 10>, angularLimitsMin=<-10 -30 -10>, angularTarget=<1 0 0>}
      localJointOffset=<2 0 0>
     }]
    relativeSpaceBone {boneName=root}
   }
   n176 = Anim Dynamics - Bone: dyn_pant_leg_front_01!(ComponentPose=@n175, Alpha=1, ExternalForce="0,0,0") -> (Pose) {
    angularDampingOverride=0.9
    angularSpringConstant=1000
    bGravityOverrideInSimSpace=true
    bOverrideAngularDamping=true
    bUseGravityOverride=true
    simulationSpace=BoneRelative
    alphaBoolBlend {blendOption=Linear}
    boundBone {boneName=dyn_pant_leg_front_01}
    componentAppliedLinearAccClamp=<1500 1500 1500>
    componentLinearAccScale=<0.4 0.4 0.4>
    componentLinearVelScale=<0.4 0.4 0.4>
    gravityOverride {z=-980}
    physicsBodyDefinitions [{
      sphereCollisionRadius=0
      boundBone {boneName=dyn_pant_leg_front_01}
      boxExtents=<18 18 18>
      constraintSetup {angularLimitsMax=<10 5 10>, angularLimitsMin=<-10 -30 -10>, angularTarget=<1 0 0>}
      localJointOffset=<2 0 0>
     }]
    relativeSpaceBone {boneName=root}
   }
   n177 = Anim Dynamics - Bone: dyn_pant_leg_front_02!(ComponentPose=@n176, Alpha=1, ExternalForce="0,0,0") -> (Pose) {
    angularDampingOverride=0.9
    angularSpringConstant=1000
    bGravityOverrideInSimSpace=true
    bOverrideAngularDamping=true
    bUseGravityOverride=true
    simulationSpace=BoneRelative
    alphaBoolBlend {blendOption=Linear}
    boundBone {boneName=dyn_pant_leg_front_02}
    componentAppliedLinearAccClamp=<1500 1500 1500>
    componentLinearAccScale=<0.4 0.4 0.4>
    componentLinearVelScale=<0.4 0.4 0.4>
    gravityOverride {z=-980}
    physicsBodyDefinitions [{
      sphereCollisionRadius=0
      boundBone {boneName=dyn_pant_leg_front_02}
      boxExtents=<18 18 18>
      constraintSetup {angularLimitsMax=<10 5 10>, angularLimitsMin=<-10 -30 -10>, angularTarget=<1 0 0>}
      localJointOffset=<2 0 0>
     }]
    relativeSpaceBone {boneName=root}
   }
   n178 = Anim Dynamics - Bone: dyn_pant_leg_front_03!(ComponentPose=@n177, Alpha=1, ExternalForce="0,0,0") -> (Pose) {
    angularDampingOverride=0.9
    angularSpringConstant=1000
    bGravityOverrideInSimSpace=true
    bOverrideAngularDamping=true
    bUseGravityOverride=true
    simulationSpace=BoneRelative
    alphaBoolBlend {blendOption=Linear}
    boundBone {boneName=dyn_pant_leg_front_03}
    componentAppliedLinearAccClamp=<1500 1500 1500>
    componentLinearAccScale=<0.4 0.4 0.4>
    componentLinearVelScale=<0.4 0.4 0.4>
    gravityOverride {z=-980}
    physicsBodyDefinitions [{
      sphereCollisionRadius=0
      boundBone {boneName=dyn_pant_leg_front_03}
      boxExtents=<18 18 18>
      constraintSetup {angularLimitsMax=<10 5 10>, angularLimitsMin=<-10 -30 -10>, angularTarget=<1 0 0>}
      localJointOffset=<2 0 0>
     }]
    relativeSpaceBone {boneName=root}
   }
   n179 = RigidBody(ComponentPose=@n178, OverrideWorldGravity="0,0,-1100") {
    bClampLinearTranslationLimitToRefPose=true
    bOverrideWorldGravity=true
    overridePhysicsAsset=&/M_MED_Cube_Ninja_Physics_Crouching
    simulationSpace=BaseBoneSpace
    alphaBoolBlend {blendOption=Linear}
    baseBoneRef {boneName=root}
    componentAppliedLinearAccClamp=<5000 5000 5000>
    componentLinearAccScale=<0.8 0.8 0.9>
    componentLinearVelScale=<0.8 0.8 0.9>
    overrideWorldGravity {z=-1100}
   }
   n180 = Blend Poses by bool(
    BlendPose_0=Component To Local(ComponentPose=@n179)
    BlendPose_1=Blend Poses by bool(
     BlendPose_0=Component To Local(ComponentPose=@n143)
     BlendPose_1=Blend Poses by bool(BlendPose_0=@n43, BlendPose_1=@n94, BlendTime_0=0.2, BlendTime_1=1.3, bActiveValue=@n95) {blendType=Linear}
     bActiveValue=Get Cloud_Emote_Playing() -> (Cloud_Emote_Playing:bool)
    ) {blendType=Linear}
    BlendTime_0=0.2
    BlendTime_1=1.3
    bActiveValue=Get bIsCrouching()
   ) {blendType=Linear}
   Output Pose(Result=Component To Local(ComponentPose=Local To Component(LocalPose=@n180))) {layerGroup=None}
   Get GravityOverride_tessel()
  ]}
 EventGraph {nodes [
   n1 = Event OnBeginCrouching()
   n2 = Delay(execute=[@n1.then, @n11.then, @n12.then], Duration=0.1)
   ResetDynamics(execute=[@n13.then, @n2.then], InTeleportType=TeleportPhysics)
   > ResetDynamics(InTeleportType=ResetPhysics)
   Event BlueprintUpdateAnimation()
   > n3 = Sequence!(execute) -> (then_0, then_1, then_2:Exec, then_3:Exec, then_4:Exec, then_5:Exec)
   n4 = GetAnimInstance(self=GetOwnerSkeletalMeshComponent())
   Is Valid(exec=@n3.then_1, InputObject=@n4)
   > Set Cloud_Emote_Playing(Cloud_Emote_Playing:bool=Montage_IsPlaying(self=@n4, Montage=/BRCosmetics/Animation/Game/MainPlayer/Emotes/CloudFloat/Emote_CloudFloat_CMM_M)) -> (Output_Get:bool)
   n5 = Branch(execute=@n3.then_2, Condition=Get Cloud_Emote_Playing())
   > n6 = Gravity Override State Function DEPRECATED(CharacterPartAnimInstance=Self-Reference(), GameTimeInSeconds=GetGameTimeInSeconds(), GravityOverrideParams=Get Gravity Override Params())
   > Set GravityOverride_tessel(GravityOverride_tessel:Vector=@n6) -> (Output_Get:Vector)
   Set GravityOverride_tessel(execute=@n5.else, GravityOverride_tessel="0, 0,-1100")
   Set TesselRotator_01(
    execute=@n3.then_3
    TesselRotator_01:Rotator=MakeWaveRotator!(
     Amplitudes:Rotator="-40,50,0"
     Frequencies:Rotator="20,100,0"
     Offsets:Rotator="15,30,0"
     bUseNoise:bool=true
     Min:Vector="0,-15,-180"
     Max:Vector="0,180,30"
    ) -> (ReturnValue:Rotator)
   ) -> (Output_Get:Rotator)
   Set TesselRotator_02(
    execute=@n3.then_4
    TesselRotator_02:Rotator=MakeWaveRotator(Amplitudes="80,40,0", Frequencies="40,25,0", Offsets="0,0,0", bUseNoise=true, Min="0,-180, 0", Max="0,15,0")
   ) -> (Output_Get:Rotator)
   n7 = BreakTransform(InTransform=GetDeltaTransformFromRefPose(self=GetOwnerSkeletalMeshComponent(), BoneName=thigh_r, BaseName=pelvis))
   Set Thigh_Rotate(execute=@n3.then_0, Thigh_Rotate:double=MapRangeClamped(Value=BreakRotator(InRot=@n7.Rotation).Pitch, InRangeA=0, InRangeB=40)) -> (Output_Get:double)
   > n8 = Branch(Condition=Get bIsSkyDiving())
   > n9 = Branch(Condition=Get bIsParachuteOpened())
   Set Override_Gravity(execute=[@n8.else, @n9.then], Override_Gravity:Vector="0, 0,-980") -> (Output_Get:Vector)
   n10 = Branch(execute=@n9.else, Condition="float > float"(A=Get PawnVelocityZ(), B=50) {is=title K2Node_CallFunction})
   > Set Override_Gravity(Override_Gravity="0, 0,200")
   Set Override_Gravity(execute=@n10.else, Override_Gravity="0, 0,-200")
   n11 = Event OnEndPlayingEmote() {is=title K2Node_Event}
   n12 = Event OnEndSkydiving()
   Event BlueprintInitializeAnimation()
   > n13 = Delay(Duration=0.15)
   n14 = GetDeltaTransformFromRefPose(self=GetOwnerSkeletalMeshComponent(), BoneName=thigh_r, BaseName=pelvis)
   MapRangeClamped(Value=BreakRotator(InRot=BreakTransform(InTransform=@n14).Rotation).Pitch, InRangeA=0, InRangeB=40)
   Branch(Condition=OR Boolean(A=false, B=false))
  ]}
]
variables [
 Cloud_Emote_Playing {type=boolean}
 Gravity Override Params {
  type=GravityOverrideParamsStruct
  default {
   emoteOrMelee {gravityOverride=<0 -300 20>, windAmplitude=<50 -25 250>, windFrequency=<0.15 1 0.4>, windOffset=<0 -100 -100>}
   onGround_Moving {gravityOverride=<0 -300 20>}
   onGround_Standing {gravityOverride=<0 -300 20>}
   skydive_DivingDown {gravityOverride=<0 0 0>}
   skydive_FloatingDown {gravityOverride=<0 0 0>}
   skydive_GoingUp {gravityOverride=<0 0 -100>, windOffset=<0 -600 400>}
   skydive_Parachuting {gravityOverride=<0 0 -300>}
   swimming {jointName=root, gravityOverride=<0 0 0>}
  }
 }
 GravityOverride_tessel {type=Vector}
 isJumping {type=number}
 isNotJumping {type=number}
 Override_Gravity {type=Vector}
 TesselRotator_01 {type=Rotator}
 TesselRotator_02 {type=Rotator}
 Thigh_Rotate {type=number}
]
