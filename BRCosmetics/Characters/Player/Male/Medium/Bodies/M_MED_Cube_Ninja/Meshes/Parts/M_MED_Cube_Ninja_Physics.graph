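# Physics asset for the M_MED_Cube_Ninja player cosmetic body, covering the dyn_ hair and earring bones plus a coarse upper body.
# Box and capsule hair bodies form branching chains from the head with mostly tight angular limits, looser at the left-side tips.
# Top hair chains hang off small master bodies angularly locked to the head, while both earrings attach directly.
# Head, neck, upper spine, clavicle and upper arm capsules are linked with free angular motion.
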PhysicsAsset /BRCosmetics/Characters/Player/Male/Medium/Bodies/M_MED_Cube_Ninja/Meshes/Parts/M_MED_Cube_Ninja_Physics
bodies {
 clavicle_l [{shapeType=Capsule, length=9.7525, radius=2.3883, center=<7.4417 0.0002 -1.8052>, rotation=<rot 90 0 0>}]
 clavicle_r [{shapeType=Capsule, length=8.676, radius=3.4647, center=<-9.0591 -0.0002 0>, rotation=<rot 90 0 0>}]
 dyn_earring [{shapeType=Box, extentX=2, extentY=0.8627, extentZ=0.5543, center=<0.7979 0 0>}]
 dyn_earring_left [{shapeType=Box, extentX=2, extentY=0.8627, extentZ=0.5543, center=<0.7979 0 0>}]
 dyn_hair_back_l_01 [{shapeType=Box, extentX=2, extentY=2, extentZ=0.5593, center=<1.1082 0 -0.0001>, rotation=<rot -16.0684 -0.0001 0.0001>}]
 dyn_hair_back_l_02 [{shapeType=Box, extentX=2, extentY=2, extentZ=0.5593, center=<1.1082 0 -0.0001>, rotation=<rot -16.0684 -0.0001 0.0001>}]
 dyn_hair_front_01 [{shapeType=Box, extentX=2.1521, extentY=2, extentZ=0.5593, center=<1.4684 0 -0.1671>, rotation=<rot -4.5863 -0.0001 0>}]
 dyn_hair_front_02 [{shapeType=Capsule, length=3.2126, radius=1.4037, center=<3.0478 0 0>, rotation=<rot 90 0 0>}]
 dyn_hair_front_03 [{shapeType=Capsule, length=3.2721, radius=1.1236, center=<2.8221 0 0>, rotation=<rot 90 0 0>}]
 dyn_hair_front_04 [{shapeType=Capsule, length=1.174, radius=0.6936, center=<1.1127 0 0>, rotation=<rot 90 0 0>}]
 dyn_hair_front_04_02 [{shapeType=Capsule, length=2.0144, radius=0.7909, center=<1.76 0 0.0001>, rotation=<rot 90 0 0>}]
 dyn_hair_l_01 [{shapeType=Capsule, length=3.2656, radius=1.1549, center=<2.8278 0.0001 0>, rotation=<rot 90 0 0>}]
 dyn_hair_l_02 [{shapeType=Capsule, length=3.4423, radius=1.3935, center=<2.8701 0.3323 0.5903>, rotation=<rot 90 0 0>}]
 dyn_hair_l_03_A [{shapeType=Capsule, length=2.7696, radius=1.056, center=<2.3389 -0.0001 0>, rotation=<rot 90 0 0>}]
 dyn_hair_l_03_B [{shapeType=Capsule, length=2.0827, radius=0.9842, center=<1.7713 0 0>, rotation=<rot 90 0 0>}]
 dyn_hair_l_03_C [{shapeType=Capsule, length=2.682, radius=1.2122, center=<2.6469 0 0>, rotation=<rot 90 0 0>}]
 dyn_hair_r_01 [{shapeType=Box, extentX=3.5827, extentY=4.375, extentZ=0.5593, center=<1.9136 -0.0668 0.39>, rotation=<rot 7.092 -5.3055 -14.2804>}]
 dyn_hair_r_01_A [{shapeType=Box, extentX=2.9854, extentY=2.7245, extentZ=1.6019, center=<1.5953 0.0001 -0.0003>, rotation=<rot 0.0001 0 -5.7225>}]
 dyn_hair_r_01_A_02 [{shapeType=Box, extentX=2, extentY=3.0819, extentZ=3.0317, center=<1.1082 0 -0.0001>, rotation=<rot -16.0683 -0.0001 -20.283>}]
 dyn_hair_r_01_B [{shapeType=Box, extentX=2.0995, extentY=2, extentZ=1.93, center=<1.2379 0 -0.0373>, rotation=<rot -16.0684 -0.0001 0.0001>}]
 dyn_hair_r_01_C [{shapeType=Box, extentX=2, extentY=2, extentZ=0.5593, center=<1.1082 0 -0.0001>, rotation=<rot -16.0684 -0.0001 0.0001>}]
 dyn_hair_r_01_C_02 [{shapeType=Box, extentX=2.2449, extentY=1.3483, extentZ=0.8527, center=<1.0685 0 0.0001>}]
 dyn_hair_top_c_01 [{shapeType=Box, extentX=2.873, extentY=3.2377, extentZ=1.7764, center=<1.4686 0 0>}]
 dyn_hair_top_c_02 [{shapeType=Box, extentX=3.0391, extentY=3.1095, extentZ=3.6282, center=<1.5439 0 0>, rotation=<rot -9.9309 0 0>}]
 dyn_hair_top_c_master [{shapeType=Capsule, length=0.505, radius=0.505}]
 dyn_hair_top_r_01 [{shapeType=Box, extentX=3.0423, extentY=3.4162, extentZ=0.9989, center=<1.6697 0 -0.0379>, rotation=<rot -3.8511 -0.0001 -7.1643>}]
 dyn_hair_top_r_02 [{shapeType=Box, extentX=2, extentY=3.567, extentZ=2.5204, center=<1.3512 0 0>, rotation=<rot -15.7223 0 0>}]
 dyn_hair_top_r_03 [{shapeType=Box, extentX=2.1795, extentY=3.73, extentZ=3.8381, center=<1.145 -0.0001 0>}]
 dyn_hair_top_r_master [{shapeType=Capsule, length=0.505, radius=0.505}]
 head [{shapeType=Capsule, length=15.7578, radius=8.7211, center=<-2.2506 1.8258 0>, rotation=<rot 90 0 0>}]
 neck_01 [{shapeType=Capsule, length=9.6895, radius=5.6779, rotation=<rot 90 0 0>}]
 spine_05 [{shapeType=Capsule, length=18.1113, radius=9.9815, rotation=<rot 0 -90 0>}]
 upperarm_l [{shapeType=Capsule, length=11.6606, radius=3.8708, center=<12.6361 -0.0002 0>, rotation=<rot 90 0 0>}]
 upperarm_r [{shapeType=Capsule, length=12.8609, radius=5.7455, center=<-11.3199 -0.0004 0>, rotation=<rot 90 0 0>}]
}
constraints [
 {bone1Name=dyn_hair_r_01_C, bone2Name=head, swing1LimitDegrees=5, swing2LimitDegrees=10, twistLimitDegrees=2}
 {bone1Name=dyn_hair_r_01_C_02, bone2Name=dyn_hair_r_01_C, swing1LimitDegrees=5, swing2LimitDegrees=10, twistLimitDegrees=2}
 {bone1Name=dyn_hair_top_r_01, bone2Name=dyn_hair_top_r_master, swing1LimitDegrees=2, swing2LimitDegrees=10, twistLimitDegrees=2}
 {bone1Name=dyn_hair_top_r_02, bone2Name=dyn_hair_top_r_01, swing1LimitDegrees=2, swing2LimitDegrees=10, twistLimitDegrees=2}
 {bone1Name=dyn_hair_top_r_03, bone2Name=dyn_hair_top_r_02, swing1LimitDegrees=2, swing2LimitDegrees=10, twistLimitDegrees=2}
 {bone1Name=dyn_hair_top_c_01, bone2Name=dyn_hair_top_c_master, swing1LimitDegrees=5, swing2LimitDegrees=10, twistLimitDegrees=2}
 {bone1Name=dyn_hair_top_c_02, bone2Name=dyn_hair_top_c_01, swing1LimitDegrees=5, swing2LimitDegrees=10, twistLimitDegrees=2}
 {bone1Name=dyn_hair_back_l_01, bone2Name=head, swing1LimitDegrees=5, swing2LimitDegrees=10, twistLimitDegrees=2}
 {bone1Name=dyn_hair_back_l_02, bone2Name=dyn_hair_back_l_01, swing1LimitDegrees=5, swing2LimitDegrees=8, twistLimitDegrees=2}
 {bone1Name=dyn_hair_r_01, bone2Name=head, swing1LimitDegrees=2, swing2LimitDegrees=3, twistLimitDegrees=2}
 {bone1Name=dyn_hair_r_01_B, bone2Name=dyn_hair_r_01, swing1LimitDegrees=2, swing2LimitDegrees=7, twistLimitDegrees=2}
 {bone1Name=dyn_hair_r_01_A, bone2Name=dyn_hair_r_01, swing1LimitDegrees=5, swing2LimitDegrees=10, twistLimitDegrees=2}
 {bone1Name=dyn_hair_r_01_A_02, bone2Name=dyn_hair_r_01_A, swing1LimitDegrees=3, swing2LimitDegrees=10, twistLimitDegrees=2}
 {bone1Name=dyn_hair_l_01, bone2Name=head, swing1LimitDegrees=4, swing2LimitDegrees=10, twistLimitDegrees=2}
 {bone1Name=dyn_hair_l_02, bone2Name=dyn_hair_l_01, swing1LimitDegrees=6, swing2LimitDegrees=10, twistLimitDegrees=2}
 {bone1Name=dyn_hair_l_03_C, bone2Name=dyn_hair_l_02, swing1LimitDegrees=26, swing2LimitDegrees=33, twistLimitDegrees=10}
 {bone1Name=dyn_hair_l_03_B, bone2Name=dyn_hair_l_02, swing1LimitDegrees=26, swing2LimitDegrees=33, twistLimitDegrees=10}
 {bone1Name=dyn_hair_l_03_A, bone2Name=dyn_hair_l_02, swing1LimitDegrees=26, swing2LimitDegrees=33, twistLimitDegrees=10}
 {bone1Name=dyn_hair_front_01, bone2Name=head, swing1LimitDegrees=5, swing2LimitDegrees=8, twistLimitDegrees=2}
 {bone1Name=dyn_hair_front_02, bone2Name=dyn_hair_front_01, swing1LimitDegrees=3, swing2LimitDegrees=20, twistLimitDegrees=2}
 {bone1Name=dyn_hair_front_03, bone2Name=dyn_hair_front_02, swing1LimitDegrees=4, swing2LimitDegrees=20, twistLimitDegrees=2.5}
 {bone1Name=dyn_hair_front_04_02, bone2Name=dyn_hair_front_03, swing1LimitDegrees=6, swing2LimitDegrees=15, twistLimitDegrees=10}
 {bone1Name=dyn_hair_front_04, bone2Name=dyn_hair_front_03, swing1LimitDegrees=6, swing2LimitDegrees=15, twistLimitDegrees=5}
 {bone1Name=dyn_earring, bone2Name=head, swing1LimitDegrees=5, swing2LimitDegrees=20, twistLimitDegrees=2}
 {
  bone1Name=head
  bone2Name=neck_01
  swing1LimitDegrees=45
  swing1Motion=Free
  swing2LimitDegrees=45
  swing2Motion=Free
  twistLimitDegrees=45
  twistMotion=Free
 }
 {
  bone1Name=neck_01
  bone2Name=spine_05
  swing1LimitDegrees=45
  swing1Motion=Free
  swing2LimitDegrees=45
  swing2Motion=Free
  twistLimitDegrees=45
  twistMotion=Free
 }
 {
  bone1Name=upperarm_r
  bone2Name=clavicle_r
  swing1LimitDegrees=45
  swing1Motion=Free
  swing2LimitDegrees=45
  swing2Motion=Free
  twistLimitDegrees=45
  twistMotion=Free
 }
 {
  bone1Name=clavicle_l
  bone2Name=spine_05
  swing1LimitDegrees=45
  swing1Motion=Free
  swing2LimitDegrees=45
  swing2Motion=Free
  twistLimitDegrees=45
  twistMotion=Free
 }
 {
  bone1Name=upperarm_l
  bone2Name=clavicle_l
  swing1LimitDegrees=45
  swing1Motion=Free
  swing2LimitDegrees=45
  swing2Motion=Free
  twistLimitDegrees=45
  twistMotion=Free
 }
 {
  bone1Name=clavicle_r
  bone2Name=spine_05
  swing1LimitDegrees=45
  swing1Motion=Free
  swing2LimitDegrees=45
  swing2Motion=Free
  twistLimitDegrees=45
  twistMotion=Free
 }
 {
  bone1Name=dyn_hair_top_r_master
  bone2Name=head
  swing1LimitDegrees=45
  swing1Motion=Locked
  swing2LimitDegrees=45
  swing2Motion=Locked
  twistLimitDegrees=45
  twistMotion=Locked
 }
 {
  bone1Name=dyn_hair_top_c_master
  bone2Name=head
  swing1LimitDegrees=45
  swing1Motion=Locked
  swing2LimitDegrees=45
  swing2Motion=Locked
  twistLimitDegrees=45
  twistMotion=Locked
 }
 {bone1Name=dyn_earring_left, bone2Name=head, swing1LimitDegrees=5, swing2LimitDegrees=20, twistLimitDegrees=2}
]
