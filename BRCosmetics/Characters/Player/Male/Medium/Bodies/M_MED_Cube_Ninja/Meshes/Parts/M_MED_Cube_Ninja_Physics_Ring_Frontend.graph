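# Physics asset for the frontend variant of the Cube Ninja's ring part, a neck hoop on a medium male player body.
# The dyn_neck_hoop bone carries two capsules and hangs off neck_01 through an angle-limited constraint.
# Capsules on the head, neck, upper spine, clavicles and upper arms form the surrounding body shapes.
# The constraints between those body bones leave swing and twist free, so only the hoop's joint has angular limits.
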PhysicsAsset /BRCosmetics/Characters/Player/Male/Medium/Bodies/M_MED_Cube_Ninja/Meshes/Parts/M_MED_Cube_Ninja_Physics_Ring_Frontend
bodies {
 clavicle_l [{shapeType=Capsule, length=9.7525, radius=2.3883, center=<7.4417 0.0002 -1.8052>, rotation=<rot 90 0 0>}]
 clavicle_r [{shapeType=Capsule, length=8.676, radius=3.4647, center=<-9.0591 -0.0002 0>, rotation=<rot 90 0 0>}]
 dyn_neck_hoop [
  {shapeType=Capsule, length=19.4088, radius=3.6628, center=<0 16.6807 2.0671>, rotation=<rot 0 89.9999 96.0221>}
  {shapeType=Capsule, length=20.8781, radius=3.7934, center=<-0.0002 6.983 2.1973>, rotation=<rot 0 90 90>}
 ]
 head [{shapeType=Capsule, length=15.7578, radius=8.7211, center=<-2.2506 1.8258 0>, rotation=<rot 90 0 0>}]
 neck_01 [{shapeType=Capsule, length=9.6895, radius=5.6779, rotation=<rot 90 0 0>}]
 spine_05 [
  {shapeType=Capsule, length=18.5198, radius=9.5729, rotation=<rot 0 -90 0>}
  {shapeType=Capsule, length=27.8613, radius=9.9815, center=<10 -13.7064 -0.0001>, rotation=<rot 0 -90 0>}
 ]
 upperarm_l [{shapeType=Capsule, length=11.6606, radius=3.8708, center=<12.6361 -0.0002 0>, rotation=<rot 90 0 0>}]
 upperarm_r [{shapeType=Capsule, length=19.8319, radius=5.86, center=<-11.3199 -0.0004 0>, rotation=<rot 90 0 0>}]
}
constraints [
 {bone1Name=dyn_neck_hoop, bone2Name=neck_01, swing1LimitDegrees=10, swing2LimitDegrees=20, twistLimitDegrees=50}
 {
  bone1Name=head
  bone2Name=neck_01
  swing1LimitDegrees=45
  swing1Motion=Free
  swing2LimitDegrees=45
  swing2Motion=Free
  twistLimitDegrees=45
  twistMotion=Free
 }
 {
  bone1Name=neck_01
  bone2Name=spine_05
  swing1LimitDegrees=45
  swing1Motion=Free
  swing2LimitDegrees=45
  swing2Motion=Free
  twistLimitDegrees=45
  twistMotion=Free
 }
 {
  bone1Name=upperarm_r
  bone2Name=clavicle_r
  swing1LimitDegrees=45
  swing1Motion=Free
  swing2LimitDegrees=45
  swing2Motion=Free
  twistLimitDegrees=45
  twistMotion=Free
 }
 {
  bone1Name=clavicle_l
  bone2Name=spine_05
  swing1LimitDegrees=45
  swing1Motion=Free
  swing2LimitDegrees=45
  swing2Motion=Free
  twistLimitDegrees=45
  twistMotion=Free
 }
 {
  bone1Name=upperarm_l
  bone2Name=clavicle_l
  swing1LimitDegrees=45
  swing1Motion=Free
  swing2LimitDegrees=45
  swing2Motion=Free
  twistLimitDegrees=45
  twistMotion=Free
 }
 {
  bone1Name=clavicle_r
  bone2Name=spine_05
  swing1LimitDegrees=45
  swing1Motion=Free
  swing2LimitDegrees=45
  swing2Motion=Free
  twistLimitDegrees=45
  twistMotion=Free
 }
]
