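# Post-process style cosmetic AnimBP for the Cube Ninja body, copying the parent mesh pose and adding secondary motion to hair, earrings and neck hoop.
# bIsInFrontEnd selects between a RigidBody-led front-end chain and a gameplay chain of per-bone AnimDynamics, ring RigidBody and constraints.
# A Tar emote montage check reweights hoop physics and neck constraints.
# Events like initialization, crouch, weapon change and skydive end reset dynamics, mostly after short delays.
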AnimBlueprint /BRCosmetics/Characters/Player/Male/Medium/Bodies/M_MED_Cube_Ninja/Meshes/Parts/M_MED_Cube_Ninja_AnimBP
target_skeleton=&/M_MED_Cube_Ninja_Skeleton
graphs [
 AnimGraph {nodes [
   n1 = Anim Dynamics - Bone: dyn_hair_front_02(ComponentPose=Local To Component(LocalPose=Copy Pose From Mesh() {bUseAttachedParent=true}), Alpha=1, ExternalForce="0,0,0") {
    angularDampingOverride=0.7
    angularSpringConstant=1000
    bGravityOverrideInSimSpace=true
    bOverrideAngularDamping=true
    bUseGravityOverride=true
    simulationSpace=BoneRelative
    alphaBoolBlend {blendOption=Linear}
    boundBone {boneName=dyn_hair_front_02}
    componentAppliedLinearAccClamp=<1500 1500 1500>
    componentLinearAccScale=<0.4 0.4 0.4>
    componentLinearVelScale=<0.4 0.4 0.4>
    gravityOverride {z=-980}
    physicsBodyDefinitions [{
      sphereCollisionRadius=0
      boundBone {boneName=dyn_hair_front_02}
      boxExtents=<15 15 15>
      constraintSetup {angularLimitsMax=<10 10 10>, angularLimitsMin=<-10 -10 -10>, angularTarget=<1 0 0>}
      localJointOffset=<2.5 0 0>
     }]
    relativeSpaceBone {boneName=root}
   }
   n2 = Anim Dynamics - Bone: dyn_hair_front_03(ComponentPose=@n1, Alpha=1, ExternalForce="0,0,0") {
    angularDampingOverride=0.7
    angularSpringConstant=1000
    bGravityOverrideInSimSpace=true
    bOverrideAngularDamping=true
    bUseGravityOverride=true
    simulationSpace=BoneRelative
    alphaBoolBlend {blendOption=Linear}
    boundBone {boneName=dyn_hair_front_03}
    componentAppliedLinearAccClamp=<1500 1500 1500>
    componentLinearAccScale=<0.4 0.4 0.4>
    componentLinearVelScale=<0.4 0.4 0.4>
    gravityOverride {z=-980}
    physicsBodyDefinitions [{
      sphereCollisionRadius=0
      boundBone {boneName=dyn_hair_front_03}
      boxExtents=<15 15 15>
      constraintSetup {angularLimitsMax=<10 10 10>, angularLimitsMin=<-10 -10 -10>, angularTarget=<1 0 0>}
      localJointOffset=<2.5 0 0>
     }]
    relativeSpaceBone {boneName=root}
   }
   n3 = Anim Dynamics - Bone: dyn_hair_front_04_02(ComponentPose=@n2, Alpha=1, ExternalForce="0,0,0") {
    angularDampingOverride=0.7
    angularSpringConstant=1000
    bGravityOverrideInSimSpace=true
    bOverrideAngularDamping=true
    bUseGravityOverride=true
    simulationSpace=BoneRelative
    alphaBoolBlend {blendOption=Linear}
    boundBone {boneName=dyn_hair_front_04_02}
    componentAppliedLinearAccClamp=<1500 1500 1500>
    componentLinearAccScale=<0.4 0.4 0.4>
    componentLinearVelScale=<0.4 0.4 0.4>
    gravityOverride {z=-980}
    physicsBodyDefinitions [{
      sphereCollisionRadius=0
      boundBone {boneName=dyn_hair_front_04_02}
      boxExtents=<15 15 15>
      constraintSetup {angularLimitsMax=<10 10 10>, angularLimitsMin=<-10 -10 -10>, angularTarget=<1 0 0>}
      localJointOffset=<2.5 0 0>
     }]
    relativeSpaceBone {boneName=root}
   }
   n4 = Anim Dynamics - Bone: dyn_hair_front_04(ComponentPose=@n3, Alpha=1, ExternalForce="0,0,0") {
    angularDampingOverride=0.7
    angularSpringConstant=1000
    bGravityOverrideInSimSpace=true
    bOverrideAngularDamping=true
    bUseGravityOverride=true
    simulationSpace=BoneRelative
    alphaBoolBlend {blendOption=Linear}
    boundBone {boneName=dyn_hair_front_04}
    componentAppliedLinearAccClamp=<1500 1500 1500>
    componentLinearAccScale=<0.4 0.4 0.4>
    componentLinearVelScale=<0.4 0.4 0.4>
    gravityOverride {z=-980}
    physicsBodyDefinitions [{
      sphereCollisionRadius=0
      boundBone {boneName=dyn_hair_front_04}
      boxExtents=<15 15 15>
      constraintSetup {angularLimitsMax=<10 10 10>, angularLimitsMin=<-10 -10 -10>, angularTarget=<1 0 0>}
      localJointOffset=<2.5 0 0>
     }]
    relativeSpaceBone {boneName=root}
   }
   n5 = Anim Dynamics - Bone: dyn_hair_l_01(ComponentPose=@n4, Alpha=1, ExternalForce="0,0,0") {
    angularDampingOverride=0.7
    angularSpringConstant=1000
    bGravityOverrideInSimSpace=true
    bOverrideAngularDamping=true
    bUseGravityOverride=true
    simulationSpace=BoneRelative
    alphaBoolBlend {blendOption=Linear}
    boundBone {boneName=dyn_hair_l_01}
    componentAppliedLinearAccClamp=<1500 1500 1500>
    componentLinearAccScale=<0.4 0.4 0.4>
    componentLinearVelScale=<0.4 0.4 0.4>
    gravityOverride {z=-980}
    physicsBodyDefinitions [{
      sphereCollisionRadius=0
      boundBone {boneName=dyn_hair_l_01}
      boxExtents=<15 15 15>
      constraintSetup {angularLimitsMax=<10 10 10>, angularLimitsMin=<-10 -10 -10>, angularTarget=<1 0 0>}
      localJointOffset=<2.5 0 0>
     }]
    relativeSpaceBone {boneName=root}
   }
   n6 = Anim Dynamics - Bone: dyn_hair_l_02(ComponentPose=@n5, Alpha=1, ExternalForce="0,0,0") {
    angularDampingOverride=0.7
    angularSpringConstant=1000
    bGravityOverrideInSimSpace=true
    bOverrideAngularDamping=true
    bUseGravityOverride=true
    simulationSpace=BoneRelative
    alphaBoolBlend {blendOption=Linear}
    boundBone {boneName=dyn_hair_l_02}
    componentAppliedLinearAccClamp=<1500 1500 1500>
    componentLinearAccScale=<0.4 0.4 0.4>
    componentLinearVelScale=<0.4 0.4 0.4>
    gravityOverride {z=-980}
    physicsBodyDefinitions [{
      sphereCollisionRadius=0
      boundBone {boneName=dyn_hair_l_02}
      boxExtents=<15 15 15>
      constraintSetup {angularLimitsMax=<10 10 10>, angularLimitsMin=<-10 -10 -10>, angularTarget=<1 0 0>}
      localJointOffset=<2.5 0 0>
     }]
    relativeSpaceBone {boneName=root}
   }
   n7 = Anim Dynamics - Bone: dyn_hair_l_03_C(ComponentPose=@n6, Alpha=1, ExternalForce="0,0,0") {
    angularDampingOverride=0.7
    angularSpringConstant=1000
    bGravityOverrideInSimSpace=true
    bOverrideAngularDamping=true
    bUseGravityOverride=true
    simulationSpace=BoneRelative
    alphaBoolBlend {blendOption=Linear}
    boundBone {boneName=dyn_hair_l_03_C}
    componentAppliedLinearAccClamp=<1500 1500 1500>
    componentLinearAccScale=<0.4 0.4 0.4>
    componentLinearVelScale=<0.4 0.4 0.4>
    gravityOverride {z=-980}
    physicsBodyDefinitions [{
      sphereCollisionRadius=0
      boundBone {boneName=dyn_hair_l_03_C}
      boxExtents=<15 15 15>
      constraintSetup {angularLimitsMax=<10 10 10>, angularLimitsMin=<-10 -10 -10>, angularTarget=<1 0 0>}
      localJointOffset=<2.5 0 0>
     }]
    relativeSpaceBone {boneName=root}
   }
   n8 = Anim Dynamics - Bone: dyn_hair_l_03_B(ComponentPose=@n7, Alpha=1, ExternalForce="0,0,0") {
    angularDampingOverride=0.7
    angularSpringConstant=1000
    bGravityOverrideInSimSpace=true
    bOverrideAngularDamping=true
    bUseGravityOverride=true
    simulationSpace=BoneRelative
    alphaBoolBlend {blendOption=Linear}
    boundBone {boneName=dyn_hair_l_03_B}
    componentAppliedLinearAccClamp=<1500 1500 1500>
    componentLinearAccScale=<0.4 0.4 0.4>
    componentLinearVelScale=<0.4 0.4 0.4>
    gravityOverride {z=-980}
    physicsBodyDefinitions [{
      sphereCollisionRadius=0
      boundBone {boneName=dyn_hair_l_03_B}
      boxExtents=<15 15 15>
      constraintSetup {angularLimitsMax=<10 10 10>, angularLimitsMin=<-10 -10 -10>, angularTarget=<1 0 0>}
      localJointOffset=<2.5 0 0>
     }]
    relativeSpaceBone {boneName=root}
   }
   n9 = Anim Dynamics - Bone: dyn_hair_l_03_A(ComponentPose=@n8, Alpha=1, ExternalForce="0,0,0") {
    angularDampingOverride=0.7
    angularSpringConstant=1000
    bGravityOverrideInSimSpace=true
    bOverrideAngularDamping=true
    bUseGravityOverride=true
    simulationSpace=BoneRelative
    alphaBoolBlend {blendOption=Linear}
    boundBone {boneName=dyn_hair_l_03_A}
    componentAppliedLinearAccClamp=<1500 1500 1500>
    componentLinearAccScale=<0.4 0.4 0.4>
    componentLinearVelScale=<0.4 0.4 0.4>
    gravityOverride {z=-980}
    physicsBodyDefinitions [{
      sphereCollisionRadius=0
      boundBone {boneName=dyn_hair_l_03_A}
      boxExtents=<15 15 15>
      constraintSetup {angularLimitsMax=<10 10 10>, angularLimitsMin=<-10 -10 -10>, angularTarget=<1 0 0>}
      localJointOffset=<2.5 0 0>
     }]
    relativeSpaceBone {boneName=root}
   }
   n10 = Anim Dynamics - Bone: dyn_hair_back_l_02(ComponentPose=@n9, Alpha=1, ExternalForce="0,0,0") {
    angularDampingOverride=0.7
    angularSpringConstant=1000
    bGravityOverrideInSimSpace=true
    bOverrideAngularDamping=true
    bUseGravityOverride=true
    simulationSpace=BoneRelative
    alphaBoolBlend {blendOption=Linear}
    boundBone {boneName=dyn_hair_back_l_02}
    componentAppliedLinearAccClamp=<1500 1500 1500>
    componentLinearAccScale=<0.4 0.4 0.4>
    componentLinearVelScale=<0.4 0.4 0.4>
    gravityOverride {z=-980}
    physicsBodyDefinitions [{
      sphereCollisionRadius=0
      boundBone {boneName=dyn_hair_back_l_02}
      boxExtents=<15 15 15>
      constraintSetup {angularLimitsMax=<7 7 7>, angularLimitsMin=<-7 -7 -7>, angularTarget=<1 0 0>}
      localJointOffset=<2.5 0 0>
     }]
    relativeSpaceBone {boneName=root}
   }
   n11 = Anim Dynamics - Bone: dyn_hair_back_l_01(ComponentPose=@n10, Alpha=1, ExternalForce="0,0,0") {
    angularDampingOverride=0.7
    angularSpringConstant=1000
    bGravityOverrideInSimSpace=true
    bOverrideAngularDamping=true
    bUseGravityOverride=true
    simulationSpace=BoneRelative
    alphaBoolBlend {blendOption=Linear}
    boundBone {boneName=dyn_hair_back_l_01}
    componentAppliedLinearAccClamp=<1500 1500 1500>
    componentLinearAccScale=<0.4 0.4 0.4>
    componentLinearVelScale=<0.4 0.4 0.4>
    gravityOverride {z=-980}
    physicsBodyDefinitions [{
      sphereCollisionRadius=0
      boundBone {boneName=dyn_hair_back_l_01}
      boxExtents=<15 15 15>
      constraintSetup {angularLimitsMax=<5 2 5>, angularLimitsMin=<-5 -2 -5>, angularTarget=<1 0 0>}
      localJointOffset=<2.5 0 0>
     }]
    relativeSpaceBone {boneName=root}
   }
   n12 = Anim Dynamics - Bone: dyn_hair_back_lb_01(ComponentPose=@n11, Alpha=1, ExternalForce="0,0,0") {
    angularDampingOverride=0.7
    angularSpringConstant=1000
    bGravityOverrideInSimSpace=true
    bOverrideAngularDamping=true
    bUseGravityOverride=true
    simulationSpace=BoneRelative
    alphaBoolBlend {blendOption=Linear}
    boundBone {boneName=dyn_hair_back_lb_01}
    componentAppliedLinearAccClamp=<1500 1500 1500>
    componentLinearAccScale=<0.4 0.4 0.4>
    componentLinearVelScale=<0.4 0.4 0.4>
    gravityOverride {z=-980}
    physicsBodyDefinitions [{
      sphereCollisionRadius=0
      boundBone {boneName=dyn_hair_back_lb_01}
      boxExtents=<15 15 15>
      constraintSetup {angularLimitsMax=<5 0 0>, angularLimitsMin=<-5 -14 -10>, angularTarget=<1 0 0>}
      localJointOffset=<2.5 1 0>
     }]
    relativeSpaceBone {boneName=root}
   }
   n13 = Anim Dynamics - Bone: dyn_hair_back_lb_02(ComponentPose=@n12, Alpha=1, ExternalForce="0,0,0") {
    angularDampingOverride=0.7
    angularSpringConstant=1000
    bGravityOverrideInSimSpace=true
    bOverrideAngularDamping=true
    bUseGravityOverride=true
    simulationSpace=BoneRelative
    alphaBoolBlend {blendOption=Linear}
    boundBone {boneName=dyn_hair_back_lb_02}
    componentAppliedLinearAccClamp=<1500 1500 1500>
    componentLinearAccScale=<0.4 0.4 0.4>
    componentLinearVelScale=<0.4 0.4 0.4>
    gravityOverride {z=-980}
    physicsBodyDefinitions [{
      sphereCollisionRadius=0
      boundBone {boneName=dyn_hair_back_lb_02}
      boxExtents=<15 15 15>
      constraintSetup {angularLimitsMax=<5 0 0>, angularLimitsMin=<-5 -14 -10>, angularTarget=<1 0 0>}
      localJointOffset=<2.5 1 0>
     }]
    relativeSpaceBone {boneName=root}
   }
   n14 = Anim Dynamics - Bone: dyn_hair_top_c_02(ComponentPose=@n13, Alpha=1, ExternalForce="0,0,0") {
    angularDampingOverride=0.7
    angularSpringConstant=1000
    bGravityOverrideInSimSpace=true
    bOverrideAngularDamping=true
    bUseGravityOverride=true
    simulationSpace=BoneRelative
    alphaBoolBlend {blendOption=Linear}
    boundBone {boneName=dyn_hair_top_c_02}
    componentAppliedLinearAccClamp=<1500 1500 1500>
    componentLinearAccScale=<0.4 0.4 0.4>
    componentLinearVelScale=<0.4 0.4 0.4>
    gravityOverride {z=-980}
    physicsBodyDefinitions [{
      sphereCollisionRadius=0
      boundBone {boneName=dyn_hair_top_c_02}
      boxExtents=<15 15 15>
      constraintSetup {angularLimitsMax=<10 10 10>, angularLimitsMin=<-10 -10 -10>, angularTarget=<1 0 0>}
      localJointOffset=<2.5 0 0>
     }]
    relativeSpaceBone {boneName=root}
   }
   n15 = Anim Dynamics - Bone: dyn_hair_top_c_03(ComponentPose=@n14, Alpha=1, ExternalForce="0,0,0") {
    angularDampingOverride=0.7
    angularSpringConstant=1000
    bGravityOverrideInSimSpace=true
    bOverrideAngularDamping=true
    bUseGravityOverride=true
    simulationSpace=BoneRelative
    alphaBoolBlend {blendOption=Linear}
    boundBone {boneName=dyn_hair_top_c_03}
    componentAppliedLinearAccClamp=<1500 1500 1500>
    componentLinearAccScale=<0.4 0.4 0.4>
    componentLinearVelScale=<0.4 0.4 0.4>
    gravityOverride {z=-980}
    physicsBodyDefinitions [{
      sphereCollisionRadius=0
      boundBone {boneName=dyn_hair_top_c_03}
      boxExtents=<15 15 15>
      constraintSetup {angularLimitsMax=<10 10 10>, angularLimitsMin=<-10 -10 -10>, angularTarget=<1 0 0>}
      localJointOffset=<2.5 0 0>
     }]
    relativeSpaceBone {boneName=root}
   }
   n16 = Anim Dynamics - Bone: dyn_hair_top_r_02(ComponentPose=@n15, Alpha=1, ExternalForce="0,0,0") {
    angularDampingOverride=0.7
    angularSpringConstant=1000
    bGravityOverrideInSimSpace=true
    bOverrideAngularDamping=true
    bUseGravityOverride=true
    simulationSpace=BoneRelative
    alphaBoolBlend {blendOption=Linear}
    boundBone {boneName=dyn_hair_top_r_02}
    componentAppliedLinearAccClamp=<1500 1500 1500>
    componentLinearAccScale=<0.4 0.4 0.4>
    componentLinearVelScale=<0.4 0.4 0.4>
    gravityOverride {z=-980}
    physicsBodyDefinitions [{
      sphereCollisionRadius=0
      boundBone {boneName=dyn_hair_top_r_02}
      boxExtents=<15 15 15>
      constraintSetup {angularLimitsMax=<10 10 10>, angularLimitsMin=<-10 -10 -10>, angularTarget=<1 0 0>}
      localJointOffset=<2.5 0 0>
     }]
    relativeSpaceBone {boneName=root}
   }
   n17 = Anim Dynamics - Bone: dyn_hair_top_r_03(ComponentPose=@n16, Alpha=1, ExternalForce="0,0,0") {
    angularDampingOverride=0.7
    angularSpringConstant=1000
    bGravityOverrideInSimSpace=true
    bOverrideAngularDamping=true
    bUseGravityOverride=true
    simulationSpace=BoneRelative
    alphaBoolBlend {blendOption=Linear}
    boundBone {boneName=dyn_hair_top_r_03}
    componentAppliedLinearAccClamp=<1500 1500 1500>
    componentLinearAccScale=<0.4 0.4 0.4>
    componentLinearVelScale=<0.4 0.4 0.4>
    gravityOverride {z=-980}
    physicsBodyDefinitions [{
      sphereCollisionRadius=0
      boundBone {boneName=dyn_hair_top_r_03}
      boxExtents=<15 15 15>
      constraintSetup {angularLimitsMax=<10 10 10>, angularLimitsMin=<-10 -10 -10>, angularTarget=<1 0 0>}
      localJointOffset=<2.5 0 0>
     }]
    relativeSpaceBone {boneName=root}
   }
   n18 = Anim Dynamics - Bone: dyn_hair_r_01(ComponentPose=@n17, Alpha=1, ExternalForce="0,0,0") {
    angularDampingOverride=0.7
    angularSpringConstant=1000
    bGravityOverrideInSimSpace=true
    bOverrideAngularDamping=true
    bUseGravityOverride=true
    simulationSpace=BoneRelative
    alphaBoolBlend {blendOption=Linear}
    boundBone {boneName=dyn_hair_r_01}
    componentAppliedLinearAccClamp=<1500 1500 1500>
    componentLinearAccScale=<0.4 0.4 0.4>
    componentLinearVelScale=<0.4 0.4 0.4>
    gravityOverride {z=-980}
    physicsBodyDefinitions [{
      sphereCollisionRadius=0
      boundBone {boneName=dyn_hair_r_01}
      boxExtents=<15 15 15>
      constraintSetup {angularLimitsMax=<5 5 5>, angularLimitsMin=<-5 -5 -5>, angularTarget=<1 0 0>}
      localJointOffset=<2.5 0 0>
     }]
    relativeSpaceBone {boneName=root}
   }
   n19 = Anim Dynamics - Bone: dyn_hair_r_01_A(ComponentPose=@n18, Alpha=1, ExternalForce="0,0,0") {
    angularDampingOverride=0.7
    angularSpringConstant=1000
    bGravityOverrideInSimSpace=true
    bOverrideAngularDamping=true
    bUseGravityOverride=true
    simulationSpace=BoneRelative
    alphaBoolBlend {blendOption=Linear}
    boundBone {boneName=dyn_hair_r_01_A}
    componentAppliedLinearAccClamp=<1500 1500 1500>
    componentLinearAccScale=<0.4 0.4 0.4>
    componentLinearVelScale=<0.4 0.4 0.4>
    gravityOverride {z=-980}
    physicsBodyDefinitions [{
      sphereCollisionRadius=0
      boundBone {boneName=dyn_hair_r_01_A}
      boxExtents=<15 15 15>
      constraintSetup {angularLimitsMax=<10 10 10>, angularLimitsMin=<-10 -10 -10>, angularTarget=<1 0 0>}
      localJointOffset=<2.5 0 0>
     }]
    relativeSpaceBone {boneName=root}
   }
   n20 = Anim Dynamics - Bone: dyn_hair_r_01_B(ComponentPose=@n19, Alpha=1, ExternalForce="0,0,0") {
    angularDampingOverride=0.7
    angularSpringConstant=1000
    bGravityOverrideInSimSpace=true
    bOverrideAngularDamping=true
    bUseGravityOverride=true
    simulationSpace=BoneRelative
    alphaBoolBlend {blendOption=Linear}
    boundBone {boneName=dyn_hair_r_01_B}
    componentAppliedLinearAccClamp=<1500 1500 1500>
    componentLinearAccScale=<0.4 0.4 0.4>
    componentLinearVelScale=<0.4 0.4 0.4>
    gravityOverride {z=-980}
    physicsBodyDefinitions [{
      sphereCollisionRadius=0
      boundBone {boneName=dyn_hair_r_01_B}
      boxExtents=<15 15 15>
      constraintSetup {angularLimitsMax=<10 10 10>, angularLimitsMin=<-10 -10 -10>, angularTarget=<1 0 0>}
      localJointOffset=<2.5 0 0>
     }]
    relativeSpaceBone {boneName=root}
   }
   n21 = Anim Dynamics - Bone: dyn_hair_r_01_C(ComponentPose=@n20, Alpha=1, ExternalForce="0,0,0") {
    angularDampingOverride=0.7
    angularSpringConstant=1000
    bGravityOverrideInSimSpace=true
    bOverrideAngularDamping=true
    bUseGravityOverride=true
    simulationSpace=BoneRelative
    alphaBoolBlend {blendOption=Linear}
    boundBone {boneName=dyn_hair_r_01_C}
    componentAppliedLinearAccClamp=<1500 1500 1500>
    componentLinearAccScale=<0.4 0.4 0.4>
    componentLinearVelScale=<0.4 0.4 0.4>
    gravityOverride {z=-980}
    physicsBodyDefinitions [{
      sphereCollisionRadius=0
      boundBone {boneName=dyn_hair_r_01_C}
      boxExtents=<15 15 15>
      constraintSetup {angularLimitsMax=<10 10 10>, angularLimitsMin=<-10 -10 -10>, angularTarget=<1 0 0>}
      localJointOffset=<2.5 0 0>
     }]
    relativeSpaceBone {boneName=root}
   }
   n22 = Anim Dynamics - Bone: dyn_earring(ComponentPose=@n21, Alpha=1, ExternalForce="0,0,0") {
    angularDampingOverride=0.9
    angularSpringConstant=500
    bAngularSpring=true
    bGravityOverrideInSimSpace=true
    bOverrideAngularDamping=true
    bUseGravityOverride=true
    simulationSpace=BoneRelative
    alphaBoolBlend {blendOption=Linear}
    boundBone {boneName=dyn_earring}
    componentAppliedLinearAccClamp=<1500 1500 1500>
    componentLinearAccScale=<0.4 0.4 0.4>
    componentLinearVelScale=<0.4 0.4 0.4>
    gravityOverride {z=-980}
    physicsBodyDefinitions [{
      sphereCollisionRadius=0
      boundBone {boneName=dyn_earring}
      boxExtents=<12 12 12>
      constraintSetup {angularLimitsMax=<2 8 8>, angularLimitsMin=<-2 -8 -8>, angularTarget=<1 0 0>}
      localJointOffset=<2.5 0 0>
     }]
    relativeSpaceBone {boneName=root}
   }
   n23 = Anim Dynamics - Bone: dyn_earring_left(ComponentPose=@n22, Alpha=1, ExternalForce="0,0,0") {
    angularDampingOverride=0.9
    angularSpringConstant=500
    bAngularSpring=true
    bGravityOverrideInSimSpace=true
    bOverrideAngularDamping=true
    bUseGravityOverride=true
    simulationSpace=BoneRelative
    alphaBoolBlend {blendOption=Linear}
    boundBone {boneName=dyn_earring_left}
    componentAppliedLinearAccClamp=<1500 1500 1500>
    componentLinearAccScale=<0.4 0.4 0.4>
    componentLinearVelScale=<0.4 0.4 0.4>
    gravityOverride {z=-980}
    physicsBodyDefinitions [{
      sphereCollisionRadius=0
      boundBone {boneName=dyn_earring_left}
      boxExtents=<12 12 12>
      constraintSetup {angularLimitsMax=<2 8 8>, angularLimitsMin=<-2 -8 -8>, angularTarget=<1 0 0>}
      localJointOffset=<2.5 0 0>
     }]
    relativeSpaceBone {boneName=root}
   }
   n24 = "Constraint - dyn_neck_hoop_master (P)"(ComponentPose=@n23, Alpha=1, ConstraintWeights_0=1) {
    alphaBoolBlend {blendOption=Linear}
    boneToModify {boneName=dyn_neck_hoop_master}
    constraintSetup [{targetBone {boneName=spine_05}}]
    constraintWeights [1]
   }
   n25 = RigidBody(ComponentPose=@n24, Alpha=Get quick_bite_neck_control() -> (quick_bite_neck_control:double)) {
    bClampLinearTranslationLimitToRefPose=true
    bOverrideWorldGravity=true
    overridePhysicsAsset=&/M_MED_Cube_Ninja_Physics_Ring
    simulationSpace=BaseBoneSpace
    alphaBoolBlend {blendOption=Linear}
    baseBoneRef {boneName=neck_01}
    componentAppliedLinearAccClamp=<5000 5000 5000>
    componentLinearAccScale=<0.6 0.6 0.6>
    componentLinearVelScale=<0.6 0.6 0.6>
    overrideWorldGravity {z=-980}
   }
   n26 = "Constraint - dyn_neck_hoop (R)"(ComponentPose=@n25, Alpha=0.6, ConstraintWeights_0=1) {
    alphaBoolBlend {blendInTime=0.1, blendOption=Linear, blendOutTime=0.1}
    alphaScaleBiasClamp {bClampResult=true, clampMax=0.65}
    boneToModify {boneName=dyn_neck_hoop}
    constraintSetup [{transformType=Rotation, targetBone {boneName=neck_01}}]
    constraintWeights [1]
   }
   n27 = Get quickBiteEmote() -> (quickBiteEmote:double)
   n28 = "Constraint - dyn_neck_hoop_base (P)"(ComponentPose=@n26, Alpha=@n27, ConstraintWeights_0=1) {
    alphaBoolBlend {blendOption=Linear}
    boneToModify {boneName=dyn_neck_hoop_base}
    constraintSetup [{targetBone {boneName=neck_02}}]
    constraintWeights [1]
   }
   n29 = "Constraint - dyn_neck_hoop (P)"(ComponentPose=@n28, Alpha=@n27, ConstraintWeights_0=1) {
    alphaBoolBlend {blendOption=Linear}
    boneToModify {boneName=dyn_neck_hoop}
    constraintSetup [{targetBone {boneName=neck_02}}]
    constraintWeights [1]
   }
   n30 = "Constraint - dyn_hair_l_01 (P)"(ComponentPose=@n29, Alpha=0.65, ConstraintWeights_0=1) {
    alphaBoolBlend {blendOption=Linear}
    boneToModify {boneName=dyn_hair_l_01}
    constraintSetup [{targetBone {boneName=dyn_hair_back_l_master}}]
    constraintWeights [1]
   }
   n31 = "Constraint - dyn_hair_l_02 (P)"(ComponentPose=@n30, Alpha=0.2, ConstraintWeights_0=1) {
    alphaBoolBlend {blendOption=Linear}
    boneToModify {boneName=dyn_hair_l_02}
    constraintSetup [{targetBone {boneName=dyn_hair_back_l_master}}]
    constraintWeights [1]
   }
   n32 = RigidBody(ComponentPose=Local To Component(LocalPose=Copy Pose From Mesh() {bUseAttachedParent=true})) {
    bClampLinearTranslationLimitToRefPose=true
    bOverrideWorldGravity=true
    overridePhysicsAsset=&/M_MED_Cube_Ninja_Physics
    simulationSpace=BaseBoneSpace
    alphaBoolBlend {blendOption=Linear}
    baseBoneRef {boneName=root}
    componentAppliedLinearAccClamp=<1250 1250 1250>
    componentLinearAccScale=<0.4 0.4 0.5>
    overrideWorldGravity {z=-980}
   }
   n33 = RigidBody(ComponentPose=@n32, Alpha=Get quick_bite_neck_control()) {
    bClampLinearTranslationLimitToRefPose=true
    bOverrideWorldGravity=true
    overridePhysicsAsset=&/M_MED_Cube_Ninja_Physics_Ring_Frontend
    simulationSpace=BaseBoneSpace
    alphaBoolBlend {blendOption=Linear}
    baseBoneRef {boneName=root}
    componentAppliedLinearAccClamp=<300 300 300>
    componentLinearAccScale=<1.1 1.1 1.1>
    componentLinearVelScale=<1 1 1.1>
    overrideWorldGravity {z=-1100}
   }
   n34 = Get quickBiteEmote()
   n35 = "Constraint - dyn_neck_hoop_base (P)"(ComponentPose=@n33, Alpha=@n34, ConstraintWeights_0=1) {
    alphaBoolBlend {blendOption=Linear}
    boneToModify {boneName=dyn_neck_hoop_base}
    constraintSetup [{targetBone {boneName=neck_02}}]
    constraintWeights [1]
   }
   n36 = "Constraint - dyn_neck_hoop (P)"(ComponentPose=@n35, Alpha=@n34, ConstraintWeights_0=1) {
    alphaBoolBlend {blendOption=Linear}
    boneToModify {boneName=dyn_neck_hoop}
    constraintSetup [{targetBone {boneName=neck_02}}]
    constraintWeights [1]
   }
   n37 = "Constraint - dyn_hair_l_01 (P)"(ComponentPose=@n36, Alpha=0.65, ConstraintWeights_0=1) {
    alphaBoolBlend {blendOption=Linear}
    boneToModify {boneName=dyn_hair_l_01}
    constraintSetup [{targetBone {boneName=dyn_hair_back_l_master}}]
    constraintWeights [1]
   }
   n38 = "Constraint - dyn_hair_l_02 (P)"(ComponentPose=@n37, Alpha=0.2, ConstraintWeights_0=1) {
    alphaBoolBlend {blendOption=Linear}
    boneToModify {boneName=dyn_hair_l_02}
    constraintSetup [{targetBone {boneName=dyn_hair_back_l_master}}]
    constraintWeights [1]
   }
   n39 = Anim Dynamics - Bone: dyn_hair_back_lb_01(ComponentPose=@n38, Alpha=1, ExternalForce="0,0,0") {
    angularDampingOverride=0.7
    angularSpringConstant=1000
    bGravityOverrideInSimSpace=true
    bOverrideAngularDamping=true
    bUseGravityOverride=true
    simulationSpace=BoneRelative
    alphaBoolBlend {blendOption=Linear}
    boundBone {boneName=dyn_hair_back_lb_01}
    componentAppliedLinearAccClamp=<1500 1500 1500>
    componentLinearAccScale=<0.4 0.4 0.4>
    componentLinearVelScale=<0.4 0.4 0.4>
    gravityOverride {z=-980}
    physicsBodyDefinitions [{
      sphereCollisionRadius=0
      boundBone {boneName=dyn_hair_back_lb_01}
      boxExtents=<15 15 15>
      constraintSetup {angularLimitsMax=<5 0 0>, angularLimitsMin=<-5 -14 -10>, angularTarget=<1 0 0>}
      localJointOffset=<2.5 1 0>
     }]
    relativeSpaceBone {boneName=root}
   }
   n40 = Anim Dynamics - Bone: dyn_hair_back_lb_02(ComponentPose=@n39, Alpha=1, ExternalForce="0,0,0") {
    angularDampingOverride=0.7
    angularSpringConstant=1000
    bGravityOverrideInSimSpace=true
    bOverrideAngularDamping=true
    bUseGravityOverride=true
    simulationSpace=BoneRelative
    alphaBoolBlend {blendOption=Linear}
    boundBone {boneName=dyn_hair_back_lb_02}
    componentAppliedLinearAccClamp=<1500 1500 1500>
    componentLinearAccScale=<0.4 0.4 0.4>
    componentLinearVelScale=<0.4 0.4 0.4>
    gravityOverride {z=-980}
    physicsBodyDefinitions [{
      sphereCollisionRadius=0
      boundBone {boneName=dyn_hair_back_lb_02}
      boxExtents=<15 15 15>
      constraintSetup {angularLimitsMax=<5 0 0>, angularLimitsMin=<-5 -14 -10>, angularTarget=<1 0 0>}
      localJointOffset=<2.5 1 0>
     }]
    relativeSpaceBone {boneName=root}
   }
   n41 = Anim Dynamics - Bone: dyn_neck_hoop_base(ComponentPose=@n40, Alpha=0.5, ExternalForce="0,0,0") {
    angularDampingOverride=0.99
    angularSpringConstant=2
    bGravityOverrideInSimSpace=true
    bOverrideAngularDamping=true
    bOverrideLinearDamping=true
    gravityScale=0.6
    linearDampingOverride=0.7
    simulationSpace=BoneRelative
    alphaBoolBlend {blendOption=Linear}
    boundBone {boneName=dyn_neck_hoop_base}
    componentLinearAccScale=<0.53 0.53 0.53>
    componentLinearVelScale=<0.75 0.75 0.75>
    physicsBodyDefinitions [{
      sphereCollisionRadius=0
      boundBone {boneName=dyn_neck_hoop_base}
      boxExtents=<0.2 0.2 0.2>
      constraintSetup {twistAxis=AxisY, angularLimitsMax=<0 95 0>, angularLimitsMin=<0 85 0>, angularTarget=<0 0 0>, linearAxesMax=<0 0 3>}
      localJointOffset=<0 0 -0.1>
     }]
    relativeSpaceBone {boneName=root}
   }
   Output Pose(Result=Blend Poses by bool(BlendPose_0=Component To Local(ComponentPose=@n41), BlendPose_1=Component To Local(ComponentPose=@n31), bActiveValue=Get bIsInFrontEnd()) {blendType=Linear}) {layerGroup=None}
  ]}
 EventGraph {nodes [
   n1 = Event OnBeginCrouching()
   n2 = Delay(execute=[@n1.then, @n5.then, @n6.then], Duration=0.1)
   ResetDynamics(execute=[@n2.then, @n7.then], InTeleportType=TeleportPhysics)
   > ResetDynamics(InTeleportType=ResetPhysics)
   Event OnWeaponChanged()
   > ResetDynamics(InTeleportType=TeleportPhysics)
   > ResetDynamics(InTeleportType=ResetPhysics)
   Event OnEndPlayingEmote() {is=title K2Node_Event}
   > Delay(Duration=0.1)
   > ResetDynamics(InTeleportType=TeleportPhysics)
   > ResetDynamics(InTeleportType=ResetPhysics)
   Event BlueprintUpdateAnimation()
   n3 = GetAnimInstance(self=GetOwnerSkeletalMeshComponent())
   > Is Valid(InputObject=@n3)
   n4 = Montage_IsPlaying(self=@n3, Montage=/BRCosmetics/Animation/Game/MainPlayer/Emotes/Tar/Emote_Tar_CMM_M)
   > Set quickBiteEmote(quickBiteEmote:double=SelectFloat(A=0.5, B=0, bPickA=@n4)) -> (Output_Get:double)
   > Set quick_bite_neck_control(quick_bite_neck_control:double=SelectFloat(A=0.75, B=1, bPickA=@n4)) -> (Output_Get:double)
   n5 = Event OnEndSkydiving()
   n6 = Event OnCharacterCustomizationLoaded()
   Event BlueprintInitializeAnimation()
   > n7 = Delay(Duration=0.15)
   n8 = Branch()
   > Set quickBiteEmote(quickBiteEmote=0.5)
   > n9 = Set quick_bite_neck_control(quick_bite_neck_control=0.2)
   n10 = Do Once Then Do!(execute:Exec=@n8.else, Reset:Exec=@n9.then, StartClosed:bool) -> (DoOnce:Exec, ThenDo:Exec) {is=title K2Node_MacroInstance}
   > ResetDynamics(InTeleportType=TeleportPhysics)
   > n11 = ResetDynamics(InTeleportType=ResetPhysics)
   Set quickBiteEmote(execute=[@n11.then, @n10.ThenDo], quickBiteEmote=0)
   > Set quick_bite_neck_control(quick_bite_neck_control=1)
   PrintString!(execute, InString="To String (Float)"(InDouble=Get quick_bite_neck_control()), bPrintToScreen, bPrintToLog, TextColor, Duration, Key) -> (then)
  ]}
]
variables [
 quick_bite_neck_control {type=number}
 quickBiteEmote {type=number}
]
